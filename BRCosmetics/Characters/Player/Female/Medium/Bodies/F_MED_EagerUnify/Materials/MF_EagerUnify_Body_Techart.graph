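# Material function for the EagerUnify female body that overrides emissive on incoming material attributes to add an emote logo glow.
# View direction in tangent space offsets UVs for parallax, and three parameter-offset SDF circles build soft outer and inner glows plus a hard glow mask.
# The emissive combines an overlay blend and an added term, suppressed by the Tech Art Mask's blue channel and scaled by the EagerUnifyMaterialCurve parameter.
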MaterialFunction /BRCosmetics/Characters/Player/Female/Medium/Bodies/F_MED_EagerUnify/Materials/MF_EagerUnify_Body_Techart
graph {nodes [
  n1 = GetMaterialAttributes(None=FunctionInput() {inputName=Result, inputType=FunctionInput_MaterialAttributes}) {attributeGetTypes [#1]}
  FunctionOutput(None=SetMaterialAttributes!(MaterialAttributes=@n1.MaterialAttributes, EmissiveColor=NamedRerouteUsage()) {attributeSetTypes [#2]}) {outputName=SetMaterialAttributes}
  n2 = Subtract(A=CameraPositionWS(), B=WorldPosition() {worldPositionShaderOffset=WPT_ExcludeAllShaderOffsets}.XYZ)
  n3 = ComponentMask(None=Transform(None=Normalize(VectorInput=@n2)) {transformSourceType=TRANSFORMSOURCE_World, transformType=TRANSFORM_Tangent}) {g=true, r=true}
  n4 = ScalarParameter() {group=Emote Logo Glow, parameterName=Parallax Intensity}
  NamedRerouteDeclaration(None=Subtract(A=TextureCoordinate(), B=Multiply(A=@n3, B=@n4))) {name=OffsetUV, nodeColor=<rgba 1 0.6797 0 1>}
  NamedRerouteDeclaration(None=@n1.BaseColor) {name=BaseColor, nodeColor=<rgba 1 0.9844 0 1>}
  n5 = Multiply(A=MaterialFunctionCall(Color=FunctionInput() {inputName=Tech Art Mask, inputType=FunctionInput_Scalar, sortPriority=10}) {materialFunction=&SplitComponents}.B) {constB=3}
  n6 = NamedRerouteUsage()
  n7 = NamedRerouteUsage()
  n8 = Multiply(
   A=OneMinus(None=Saturate(None=@n5))
   B=Add(
    A=MaterialFunctionCall(Base=@n6, Blend=@n7) {materialFunction=&Blend_Overlay}.Result
    B=NamedRerouteUsage()
   )
  ) {constA=1}
  n9 = ScalarParameter() {parameterName=EagerUnifyMaterialCurve}
  NamedRerouteDeclaration(None=Multiply(A=Multiply(A=@n8, B=@n9), B=NamedRerouteUsage())) {name=OutEmissive, nodeColor=<rgba 0.6641 0 1 1>}
  n10 = VectorParameter() {group=Emote Logo Glow, parameterName=Circle 1 Mask Offset}
  NamedRerouteDeclaration(None=Add(A=NamedRerouteUsage(), B=ComponentMask(None=@n10.RGB) {g=true, r=true})) {name=Circle1UV, nodeColor=<rgba 0 0.4375 1 1>}
  n11 = ScalarParameter() {defaultValue=0.1, group=Emote Logo Glow, parameterName=Cicle Mask Softness}
  n12 = NamedRerouteUsage()
  n13 = Saturate(None=MaterialFunctionCall!(UVs=@n12, Radius=ScalarParameter() {defaultValue=0.1, group=Emote Logo Glow, parameterName=Cicle Mask Scale}) {materialFunction=/FortUILibrary/MaterialFunctions/Shapes/SDFs/MF_SDF_Circle}.SDF)
  n14 = ScalarParameter() {defaultValue=0.1, group=Emote Logo Glow, parameterName=Cicle Mask Softness}
  n15 = NamedRerouteUsage()
  n16 = Saturate(None=MaterialFunctionCall(UVs=@n15, Radius=ScalarParameter() {defaultValue=0.1, group=Emote Logo Glow, parameterName=Cicle Mask Scale}) {materialFunction=&MF_SDF_Circle}.SDF)
  n17 = Add(A=OneMinus(None=SmoothStep(Max=@n11, Value=@n13) {constMax=0.05}), B=OneMinus(None=SmoothStep(Max=@n14, Value=@n16) {constMax=0.05}))
  n18 = NamedRerouteUsage()
  n19 = MaterialFunctionCall(UVs=@n18, Radius=ScalarParameter() {defaultValue=0.1, group=Emote Logo Glow, parameterName=Cicle Mask Scale}) {materialFunction=&MF_SDF_Circle}
  n20 = OneMinus(None=SmoothStep(Max=ScalarParameter() {defaultValue=0.1, group=Emote Logo Glow, parameterName=Cicle Mask Softness}, Value=Saturate(None=@n19.SDF)) {constMax=0.05})
  NamedRerouteDeclaration(None=Multiply(A=Add(A=@n17, B=@n20)) {constB=1.5}) {name=OuterGlow, nodeColor=<rgba 1 0.0234 0 1>}
  n21 = ScalarParameter() {defaultValue=0.01, group=Emote Logo Glow, parameterName=Glow Softness}
  n22 = Saturate(None=MaterialFunctionCall(UVs=@n12, Radius=ScalarParameter() {defaultValue=0.01, group=Emote Logo Glow, parameterName=Glow Scale}) {materialFunction=&MF_SDF_Circle}.SDF)
  n23 = ScalarParameter() {defaultValue=0.01, group=Emote Logo Glow, parameterName=Glow Softness}
  n24 = Saturate(None=MaterialFunctionCall(UVs=@n15, Radius=ScalarParameter() {defaultValue=0.01, group=Emote Logo Glow, parameterName=Glow Scale}) {materialFunction=&MF_SDF_Circle}.SDF)
  n25 = Add(A=OneMinus(None=SmoothStep(Max=@n21, Value=@n22)), B=OneMinus(None=SmoothStep(Max=@n23, Value=@n24)))
  n26 = MaterialFunctionCall(UVs=@n18, Radius=ScalarParameter() {defaultValue=0.01, group=Emote Logo Glow, parameterName=Glow Scale}) {materialFunction=&MF_SDF_Circle}
  n27 = OneMinus(None=SmoothStep(Max=ScalarParameter() {defaultValue=0.01, group=Emote Logo Glow, parameterName=Glow Softness}, Value=Saturate(None=@n26.SDF)))
  NamedRerouteDeclaration(None=Multiply(A=Add(A=@n25, B=@n27)) {constB=0.6}) {name=InnerGlow, nodeColor=<rgba 0 0.2969 1 1>}
  n28 = VectorParameter() {group=Emote Logo Glow, parameterName=Circle 2 Mask Offset}
  NamedRerouteDeclaration(None=Add(A=NamedRerouteUsage(), B=ComponentMask(None=@n28.RGB) {g=true, r=true})) {name=Circle2UV, nodeColor=<rgba 0.7578 1 0 1>}
  n29 = VectorParameter() {group=Emote Logo Glow, parameterName=Circle 3 Mask Offset}
  NamedRerouteDeclaration(None=Add(A=NamedRerouteUsage(), B=ComponentMask(None=@n29.RGB) {g=true, r=true})) {name=Circle3UV, nodeColor=<rgba 0.9219 0 1 1>}
  n30 = Constant() {r=0.05}
  n31 = MaterialFunctionCall(UVs=NamedRerouteUsage(), Radius=@n30) {materialFunction=&MF_SDF_Circle}
  n32 = Saturate(None=If(A=OneMinus(None=@n31.SDF), B=Constant() {r=0.99}, "A > B"=Constant() {r=1}, "A < B"=Constant()))
  n33 = MaterialFunctionCall(UVs=NamedRerouteUsage(), Radius=@n30) {materialFunction=&MF_SDF_Circle}
  n34 = Saturate(None=If(A=OneMinus(None=@n33.SDF), B=Constant() {r=0.99}, "A > B"=Constant() {r=1}, "A < B"=Constant()))
  n35 = NamedRerouteUsage()
  n36 = If(
   A=OneMinus(None=MaterialFunctionCall(UVs=@n35, Radius=@n30) {materialFunction=&MF_SDF_Circle}.SDF)
   B=Constant() {r=0.99}
   "A > B"=Constant() {r=1}
   "A < B"=Constant()
  )
  NamedRerouteDeclaration(None=Add(A=Add(A=@n32, B=@n34), B=Saturate(None=@n36))) {name=GlowMask, nodeColor=<rgba 0 1 0.5547 1>}
 ]}
