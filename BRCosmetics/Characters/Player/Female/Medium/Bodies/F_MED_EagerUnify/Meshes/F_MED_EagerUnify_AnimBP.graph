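# Animation Blueprint for the EagerUnify female medium body cosmetic, adding secondary motion and jaw matching to the copied parent-mesh pose.
# The copied pose runs through a Control Rig, then jaw and chin bones take transforms read each update from the head mesh.
# A bool blend selects between jacket RigidBody plus zipper/collar AnimDynamics and an all-AnimDynamics chain.
# The event graph also computes gravity override and movement/idle values and sets an alternative skin weight profile.
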AnimBlueprint /BRCosmetics/Characters/Player/Female/Medium/Bodies/F_MED_EagerUnify/Meshes/F_MED_EagerUnify_AnimBP
target_skeleton=&/F_MED_EagerUnify_Skeleton
graphs [
 AnimGraph {nodes [
   Output Pose(Result=Use cached pose 'Dyn_CP'()) {layerGroup=None}
   n1 = Control Rig!(Source=Copy Pose From Mesh() {bCopyCurves=true, bUseAttachedParent=true}, Alpha, isIdleFloat:double=0) -> (Pose) {
    alphaBoolBlend {blendOption=Linear}
    controlRigAssetReference {blueprintRigClass=&/F_MED_EagerUnify_CtrlRig.F_MED_EagerUnify_CtrlRig_C}
   }
   Save cached pose 'Input_CP'(Pose=@n1) {cacheName=Input_CP}
   n2 = RigidBody!(ComponentPose=Local To Component(LocalPose=Use cached pose 'Facial_CP'()), Alpha=0.75, OverrideWorldGravity:Vector="0,0,-980") -> (Pose) {
    bClampLinearTranslationLimitToRefPose=true
    bOverrideWorldGravity=true
    overridePhysicsAsset=&/F_MED_EagerUnify_Jacket_Physics
    simulationSpace=BaseBoneSpace
    alphaBoolBlend {blendOption=Linear}
    baseBoneRef {boneName=pelvis}
    componentAppliedLinearAccClamp=<5000 5000 5000>
    componentLinearAccScale=<0.8 0.8 0.35>
    componentLinearVelScale=<0.8 0.8 0.35>
    simSpaceSettings {maxAngularAcceleration=100, maxAngularVelocity=9, maxLinearAcceleration=100, maxLinearVelocity=300, velocityScaleZ=0, worldAlpha=1}
   }
   n3 = Anim Dynamics - Bone: dyn_zipper!(ComponentPose=@n2, Alpha=1, GravityOverride:Vector="0,0,-980") -> (Pose) {
    angularDampingOverride=1
    angularSpringConstant=20000
    bOverrideAngularDamping=true
    bUseGravityOverride=true
    gravityScale=0.35
    simulationSpace=RootRelative
    boundBone {boneName=dyn_zipper}
    componentAppliedLinearAccClamp=<3000 3000 3000>
    componentLinearAccScale=<0.6 0.6 0.6>
    componentLinearVelScale=<0.6 0.6 0.6>
    gravityOverride {z=-980}
    physicsBodyDefinitions [{
      boundBone {boneName=dyn_zipper}
      boxExtents=<10 16 8>
      constraintSetup {angularLimitsMax=<0 5 15>, angularLimitsMin=<0 -5 -70>, angularTarget=<1 0 0>}
      localJointOffset=<4 0 0>
     }]
   }
   n4 = Anim Dynamics - Bone: dyn_collar_l!(ComponentPose=@n3, GravityOverride:Vector="0,0,-980") -> (Pose) {
    angularDampingOverride=1
    angularSpringConstant=3500
    bAngularSpring=true
    bOverrideAngularDamping=true
    bUseGravityOverride=true
    gravityScale=0.35
    alphaBoolBlend {blendOption=Linear}
    boundBone {boneName=dyn_collar_l}
    componentAppliedLinearAccClamp=<10000 10000 10000>
    componentLinearAccScale=<1 1 1>
    componentLinearVelScale=<1 1 1>
    gravityOverride {z=-980}
    physicsBodyDefinitions [{
      boundBone {boneName=dyn_collar_l}
      boxExtents=<10 10 10>
      constraintSetup {angularLimitsMax=<0 0 30>, angularLimitsMin=<0 0 -10>, angularTarget=<1 0 0>}
      localJointOffset=<2 0 0>
     }]
   }
   n5 = Anim Dynamics - Bone: dyn_collar_r!(ComponentPose=@n4, GravityOverride:Vector="0,0,-980") -> (Pose) {
    angularDampingOverride=1
    angularSpringConstant=3500
    bAngularSpring=true
    bOverrideAngularDamping=true
    bUseGravityOverride=true
    gravityScale=0.35
    alphaBoolBlend {blendOption=Linear}
    boundBone {boneName=dyn_collar_r}
    componentAppliedLinearAccClamp=<10000 10000 10000>
    componentLinearAccScale=<1 1 1>
    componentLinearVelScale=<1 1 1>
    gravityOverride {z=-980}
    physicsBodyDefinitions [{
      boundBone {boneName=dyn_collar_r}
      boxExtents=<10 10 10>
      constraintSetup {angularLimitsMax=<0 0 30>, angularLimitsMin=<0 0 -10>, angularTarget=<1 0 0>}
      localJointOffset=<2 0 0>
     }]
   }
   n6 = Anim Dynamics - Bone: dyn_front_Jacket_l!(ComponentPose=Local To Component(LocalPose=Use cached pose 'Facial_CP'()), Alpha=0.5, GravityOverride:Vector="-980,0,0") -> (Pose) {
    angularDampingOverride=1
    angularSpringConstant=100
    bAngularSpring=true
    bOverrideAngularDamping=true
    bUseGravityOverride=true
    gravityScale=0.35
    simulationSpace=World
    boundBone {boneName=dyn_front_Jacket_l}
    componentAppliedLinearAccClamp=<6000 6000 6000>
    componentLinearAccScale=<0.5 0.5 0.5>
    componentLinearVelScale=<0.6 0.6 0.6>
    gravityOverride {x=-980}
    physicsBodyDefinitions [{
      boundBone {boneName=dyn_front_Jacket_l}
      boxExtents=<10 16 8>
      constraintSetup {angularLimitsMax=<0 1 0>, angularLimitsMin=<0 -20 0>, angularTarget=<1 0 0>}
      localJointOffset=<4 0 0>
     }]
    relativeSpaceBone {boneName=neck_01}
    simSpaceSettings {maxAngularAcceleration=15, maxAngularVelocity=7, simSpaceAngularAlpha=0.75}
   }
   n7 = Anim Dynamics - Bone: dyn_front_Jacket_r!(ComponentPose=@n6, Alpha=1, GravityOverride:Vector="-980,0,0") -> (Pose) {
    angularDampingOverride=1
    angularSpringConstant=100
    bAngularSpring=true
    bOverrideAngularDamping=true
    bUseGravityOverride=true
    gravityScale=0.35
    simulationSpace=RootRelative
    boundBone {boneName=dyn_front_Jacket_r}
    componentAppliedLinearAccClamp=<6000 6000 6000>
    componentLinearAccScale=<0.5 0.5 0.5>
    componentLinearVelScale=<0.6 0.6 0.6>
    gravityOverride {x=-980}
    physicsBodyDefinitions [{
      boundBone {boneName=dyn_front_Jacket_r}
      boxExtents=<10 16 8>
      constraintSetup {angularLimitsMax=<0 1 0>, angularLimitsMin=<0 -20 0>, angularTarget=<1 0 0>}
      localJointOffset=<4 0 0>
     }]
    relativeSpaceBone {boneName=neck_01}
    simSpaceSettings {maxAngularAcceleration=15, maxAngularVelocity=7, simSpaceAngularAlpha=0.75}
   }
   n8 = Anim Dynamics - Bone: dyn_side_Jacket_l!(ComponentPose=@n7, Alpha=1, GravityOverride:Vector="-980,0,0") -> (Pose) {
    angularDampingOverride=1
    angularSpringConstant=100
    bAngularSpring=true
    bOverrideAngularDamping=true
    bUseGravityOverride=true
    gravityScale=0.35
    simulationSpace=RootRelative
    boundBone {boneName=dyn_side_Jacket_l}
    componentAppliedLinearAccClamp=<6000 6000 6000>
    componentLinearAccScale=<0.5 0.5 0.5>
    componentLinearVelScale=<0.6 0.6 0.6>
    gravityOverride {x=-980}
    physicsBodyDefinitions [{
      boundBone {boneName=dyn_side_Jacket_l}
      boxExtents=<10 16 8>
      constraintSetup {angularLimitsMax=<0 1 0>, angularLimitsMin=<0 -15 0>, angularTarget=<1 0 0>}
      localJointOffset=<4 0 0>
     }]
    relativeSpaceBone {boneName=neck_01}
    simSpaceSettings {maxAngularAcceleration=15, maxAngularVelocity=7, simSpaceAngularAlpha=0.75}
   }
   n9 = Anim Dynamics - Bone: dyn_side_Jacket_r!(ComponentPose=@n8, Alpha=1, GravityOverride:Vector="-980,0,0") -> (Pose) {
    angularDampingOverride=1
    angularSpringConstant=100
    bAngularSpring=true
    bOverrideAngularDamping=true
    bUseGravityOverride=true
    gravityScale=0.35
    simulationSpace=RootRelative
    boundBone {boneName=dyn_side_Jacket_r}
    componentAppliedLinearAccClamp=<6000 6000 6000>
    componentLinearAccScale=<0.5 0.5 0.5>
    componentLinearVelScale=<0.6 0.6 0.6>
    gravityOverride {x=-980}
    physicsBodyDefinitions [{
      boundBone {boneName=dyn_side_Jacket_r}
      boxExtents=<10 16 8>
      constraintSetup {angularLimitsMax=<0 1 0>, angularLimitsMin=<0 -15 0>, angularTarget=<1 0 0>}
      localJointOffset=<4 0 0>
     }]
    relativeSpaceBone {boneName=neck_01}
    simSpaceSettings {maxAngularAcceleration=15, maxAngularVelocity=7, simSpaceAngularAlpha=0.75}
   }
   n10 = Anim Dynamics - Bone: dyn_back_Jacket_l!(ComponentPose=@n9, Alpha=1, GravityOverride:Vector="-980,0,0") -> (Pose) {
    angularDampingOverride=1
    angularSpringConstant=100
    bAngularSpring=true
    bOverrideAngularDamping=true
    bUseGravityOverride=true
    gravityScale=0.35
    simulationSpace=RootRelative
    boundBone {boneName=dyn_back_Jacket_l}
    componentAppliedLinearAccClamp=<6000 6000 6000>
    componentLinearAccScale=<0.5 0.5 0.5>
    componentLinearVelScale=<0.6 0.6 0.6>
    gravityOverride {x=-980}
    physicsBodyDefinitions [{
      boundBone {boneName=dyn_back_Jacket_l}
      boxExtents=<10 16 8>
      constraintSetup {angularLimitsMax=<0 1 0>, angularLimitsMin=<0 -10 0>, angularTarget=<1 0 0>}
      localJointOffset=<4 0 0>
     }]
    relativeSpaceBone {boneName=neck_01}
    simSpaceSettings {maxAngularAcceleration=15, maxAngularVelocity=7, simSpaceAngularAlpha=0.75}
   }
   n11 = Anim Dynamics - Bone: dyn_zipper(ComponentPose=@n10, Alpha=1, GravityOverride="-980,0,0") {
    angularDampingOverride=1
    angularSpringConstant=100
    bAngularSpring=true
    bOverrideAngularDamping=true
    bUseGravityOverride=true
    gravityScale=0.35
    simulationSpace=RootRelative
    boundBone {boneName=dyn_zipper}
    componentAppliedLinearAccClamp=<6000 6000 6000>
    componentLinearAccScale=<0.5 0.5 0.5>
    componentLinearVelScale=<0.6 0.6 0.6>
    gravityOverride {x=-980}
    physicsBodyDefinitions [{
      boundBone {boneName=dyn_zipper}
      boxExtents=<10 16 8>
      constraintSetup {angularLimitsMax=<0 5 15>, angularLimitsMin=<0 -5 -70>, angularTarget=<1 0 0>}
      localJointOffset=<4 0 0>
     }]
    relativeSpaceBone {boneName=neck_01}
    simSpaceSettings {maxAngularAcceleration=15, maxAngularVelocity=7, simSpaceAngularAlpha=0.75}
   }
   n12 = Anim Dynamics - Bone: dyn_collar_l(ComponentPose=@n11, GravityOverride="0,0,-980") {
    angularDampingOverride=1
    angularSpringConstant=3500
    bAngularSpring=true
    bOverrideAngularDamping=true
    bUseGravityOverride=true
    gravityScale=0.35
    alphaBoolBlend {blendOption=Linear}
    boundBone {boneName=dyn_collar_l}
    componentAppliedLinearAccClamp=<10000 10000 10000>
    componentLinearAccScale=<1 1 1>
    componentLinearVelScale=<1 1 1>
    gravityOverride {z=-980}
    physicsBodyDefinitions [{
      boundBone {boneName=dyn_collar_l}
      boxExtents=<10 10 10>
      constraintSetup {angularLimitsMax=<0 0 30>, angularLimitsMin=<0 0 -10>, angularTarget=<1 0 0>}
      localJointOffset=<2 0 0>
     }]
   }
   n13 = Anim Dynamics - Bone: dyn_collar_r(ComponentPose=@n12, GravityOverride="0,0,-980") {
    angularDampingOverride=1
    angularSpringConstant=3500
    bAngularSpring=true
    bOverrideAngularDamping=true
    bUseGravityOverride=true
    gravityScale=0.35
    alphaBoolBlend {blendOption=Linear}
    boundBone {boneName=dyn_collar_r}
    componentAppliedLinearAccClamp=<10000 10000 10000>
    componentLinearAccScale=<1 1 1>
    componentLinearVelScale=<1 1 1>
    gravityOverride {z=-980}
    physicsBodyDefinitions [{
      boundBone {boneName=dyn_collar_r}
      boxExtents=<10 10 10>
      constraintSetup {angularLimitsMax=<0 0 30>, angularLimitsMin=<0 0 -10>, angularTarget=<1 0 0>}
      localJointOffset=<2 0 0>
     }]
   }
   Save cached pose 'Dyn_CP'(Pose=Blend Poses by bool(BlendPose_0=Component To Local(ComponentPose=@n13), BlendPose_1=Component To Local(ComponentPose=@n5)) {blendType=Linear}) {cacheName=Dyn_CP}
   n14 = "Transform (Modify) Bone - Bone: FACIAL_C_Jaw"!(
    ComponentPose=Local To Component(LocalPose=Use cached pose 'Input_CP'())
    bAlphaBoolEnabled:bool=True
    Translation=Get FACIAL_C_Jaw_Tranform() -> (FACIAL_C_Jaw_Tranform:Vector)
    Rotation=Get FACIAL_C_Jaw_Rotation() -> (FACIAL_C_Jaw_Rotation:Rotator)
   ) -> (Pose) {
    alphaInputType=Bool
    rotationMode=BMM_Replace
    rotationSpace=BCS_ParentBoneSpace
    translationMode=BMM_Replace
    translationSpace=BCS_ParentBoneSpace
    alphaBoolBlend {blendOption=Linear}
    boneToModify {boneName=FACIAL_C_Jaw}
   }
   n15 = "Transform (Modify) Bone - Bone: FACIAL_C_Chin"!(
    ComponentPose=@n14
    bAlphaBoolEnabled:bool=True
    Translation=Get FACIAL_C_Chin_Translation() -> (FACIAL_C_Chin_Translation:Vector)
    Rotation=Get FACIAL_C_Chin_Rotation() -> (FACIAL_C_Chin_Rotation:Rotator)
   ) -> (Pose) {
    alphaInputType=Bool
    rotationMode=BMM_Replace
    rotationSpace=BCS_ParentBoneSpace
    translationMode=BMM_Replace
    translationSpace=BCS_ParentBoneSpace
    alphaBoolBlend {blendOption=Linear}
    boneToModify {boneName=FACIAL_C_Chin}
   }
   n16 = "Transform (Modify) Bone - Bone: FACIAL_L_Jawline"!(
    ComponentPose=@n15
    bAlphaBoolEnabled:bool=True
    Translation=Get FACIAL_L_Jawline_Translation() -> (FACIAL_L_Jawline_Translation:Vector)
    Rotation=Get FACIAL_L_Jawline_Rotation() -> (FACIAL_L_Jawline_Rotation:Rotator)
   ) -> (Pose) {
    alphaInputType=Bool
    rotationMode=BMM_Replace
    rotationSpace=BCS_ParentBoneSpace
    translationMode=BMM_Replace
    translationSpace=BCS_ParentBoneSpace
    alphaBoolBlend {blendOption=Linear}
    boneToModify {boneName=FACIAL_L_Jawline}
   }
   n17 = "Transform (Modify) Bone - Bone: FACIAL_C_Jawline"!(
    ComponentPose=@n16
    bAlphaBoolEnabled:bool=True
    Translation=Get FACIAL_C_Jawline_Translation() -> (FACIAL_C_Jawline_Translation:Vector)
    Rotation=Get FACIAL_C_Jawline_Rotation() -> (FACIAL_C_Jawline_Rotation:Rotator)
   ) -> (Pose) {
    alphaInputType=Bool
    rotationMode=BMM_Replace
    rotationSpace=BCS_ParentBoneSpace
    translationMode=BMM_Replace
    translationSpace=BCS_ParentBoneSpace
    alphaBoolBlend {blendOption=Linear}
    boneToModify {boneName=FACIAL_C_Jawline}
   }
   n18 = "Transform (Modify) Bone - Bone: FACIAL_R_Jawline"!(
    ComponentPose=@n17
    bAlphaBoolEnabled:bool=True
    Translation=Get FACIAL_R_Jawline_Translation() -> (FACIAL_R_Jawline_Translation:Vector)
    Rotation=Get FACIAL_R_Jawline_Rotation() -> (FACIAL_R_Jawline_Rotation:Rotator)
   ) -> (Pose) {
    alphaInputType=Bool
    rotationMode=BMM_Replace
    rotationSpace=BCS_ParentBoneSpace
    translationMode=BMM_Replace
    translationSpace=BCS_ParentBoneSpace
    alphaBoolBlend {blendOption=Linear}
    boneToModify {boneName=FACIAL_R_Jawline}
   }
   n19 = "Transform (Modify) Bone - Bone: FACIAL_L_UnderChin"!(
    ComponentPose=@n18
    bAlphaBoolEnabled:bool=True
    Translation=Get FACIAL_L_UnderChin_Translation() -> (FACIAL_L_UnderChin_Translation:Vector)
    Rotation=Get FACIAL_L_UnderChin_Rotation() -> (FACIAL_L_UnderChin_Rotation:Rotator)
   ) -> (Pose) {
    alphaInputType=Bool
    rotationMode=BMM_Replace
    rotationSpace=BCS_ParentBoneSpace
    translationMode=BMM_Replace
    translationSpace=BCS_ParentBoneSpace
    alphaBoolBlend {blendOption=Linear}
    boneToModify {boneName=FACIAL_L_UnderChin}
   }
   n20 = "Transform (Modify) Bone - Bone: FACIAL_C_UnderChin"!(
    ComponentPose=@n19
    bAlphaBoolEnabled:bool=True
    Translation=Get FACIAL_C_UnderChin_Translation() -> (FACIAL_C_UnderChin_Translation:Vector)
    Rotation=Get FACIAL_C_UnderChin_Rotation() -> (FACIAL_C_UnderChin_Rotation:Rotator)
   ) -> (Pose) {
    alphaInputType=Bool
    rotationMode=BMM_Replace
    rotationSpace=BCS_ParentBoneSpace
    translationMode=BMM_Replace
    translationSpace=BCS_ParentBoneSpace
    alphaBoolBlend {blendInTime=0.25, blendOption=Linear, blendOutTime=0.25}
    boneToModify {boneName=FACIAL_C_UnderChin}
   }
   n21 = "Transform (Modify) Bone - Bone: FACIAL_R_UnderChin"!(
    ComponentPose=@n20
    bAlphaBoolEnabled:bool=True
    Translation=Get FACIAL_R_UnderChin_Translation() -> (FACIAL_R_UnderChin_Translation:Vector)
    Rotation=Get FACIAL_R_UnderChin_Rotation() -> (FACIAL_R_UnderChin_Rotation:Rotator)
   ) -> (Pose) {
    alphaInputType=Bool
    rotationMode=BMM_Replace
    rotationSpace=BCS_ParentBoneSpace
    translationMode=BMM_Replace
    translationSpace=BCS_ParentBoneSpace
    alphaBoolBlend {blendOption=Linear}
    boneToModify {boneName=FACIAL_R_UnderChin}
   }
   Save cached pose 'Facial_CP'(Pose=Component To Local(ComponentPose=@n21)) {cacheName=Facial_CP}
  ]}
 EventGraph {nodes [
   Event BlueprintInitializeAnimation()
   n1 = TryGetPawnOwner()
   > Is Valid(InputObject=@n1)
   > n2 = Cast To FortPlayerPawnAthena(Object=@n1)
   > Set PlayerPawn(PlayerPawn:Fort Player Pawn Athena Object Reference=@n2) -> (Output_Get:Fort Player Pawn Athena Object Reference)
   > SetSkinWeightProfile!(
    execute
    self:Skinned Mesh Component Object Reference=Get SkelMesh()
    InProfileName:Name=Alternative
    InLayer:ESkinWeightProfileLayer Enum=Secondary
   ) -> (then, ReturnValue:bool)
   Event BlueprintUpdateAnimation()
   > n3 = Sequence!(execute) -> (then_0, then_1, then_2:Exec, then_3:Exec, then_4:Exec)
   > n4 = "Gravity Override State Function (Parameter Asset) (Nativized)"(
    CharacterPartAnimInstance=Self-Reference()
    GameTimeInSeconds=GetGameTimeInSeconds()
    Parameters=&/F_MED_EagerUnify_Parameters
   )
   > Set GravityOverride(GravityOverride=@n4)
   n5 = Get HeadMesh() -> (HeadMesh:Skeletal Mesh Component Budgeted Object Reference)
   Is Valid(exec=@n3.then_2, InputObject=@n5)
   n6 = GetSocketBoneName!(self:Skinned Mesh Component Object Reference=@n5, InSocketName:Name=FACIAL_C_Jaw) -> (ReturnValue:Name=None)
   > n7 = Set isUsingFacialBones(isUsingFacialBones:bool="Not Equal Exactly (String)"(A="To String (Name)"(InName=@n6), B=None)) -> (Output_Get:bool)
   > Branch(Condition=@n7)
   > K2Node_Composite_0 = JawLineSetup_2!(execute, HeadMesh:Scene Component Object Reference=Get HeadMesh()) {is=title K2Node_Composite}
   n8 = PlayerMovement(execute=@n3.then_3, AnimInstance=Self-Reference(), XDotProductCoefficient=1, YDotProductCoefficient=1, ZDotProductCoefficient=1)
   n9 = BreakVector(InVec=MakeVector(X=@n8, Y=@n8.YMovement))
   > Set StrafeLeftFloat(StrafeLeftFloat:double="Absolute (Float)"(A="Clamp (Float)"(Value=@n9, Min=-1, Max=0))) -> (Output_Get:double)
   > Set strafeRightFloat(strafeRightFloat:double="Clamp (Float)"(Value=@n9, Min=0, Max=1)) -> (Output_Get:double)
   > Set moveForward(moveForward:double="Clamp (Float)"(Value=@n9.Y, Min=0, Max=1)) -> (Output_Get:double)
   > Set moveBackward(moveBackward:double="Absolute (Float)"(A="Clamp (Float)"(Value=@n9.Y, Min=-1, Max=0))) -> (Output_Get:double)
   > n10 = Branch(Condition=OR Boolean!(A=Get bIsSkyDiving(), B=Get bIsCrouching(), C:bool=Get bIsPlayingEmote()) -> (ReturnValue))
   > Set isIdle(isIdle:double=0) -> (Output_Get:double)
   > n11 = Set isIdleInFE(isIdleInFE:double=0) -> (Output_Get:double)
   n12 = Get bIsPlayingEmote()
   Set isIdleFrontEnd(
    execute=[@n14.then, @n11.then]
    isIdleFrontEnd:bool=AND Boolean(A=Get bIsInFrontEnd(), B=InvertBool!(bValue:bool=@n12) -> (ReturnValue:bool))
   ) -> (Output_Get:bool)
   n13 = float + float!(
    A=Get StrafeLeftFloat() -> (StrafeLeftFloat:double)
    B=Get strafeRightFloat() -> (strafeRightFloat:double)
    C:double=Get moveForward() -> (moveForward:double)
    D:double=Get moveBackward() -> (moveBackward:double)
   ) -> (ReturnValue)
   n14 = Set isIdle(execute=@n10.else, isIdle=Select(Option 0:double=1, Option 1:double=0, Index="float > float"(A=@n13)) -> (ReturnValue:double))
   n15 = Event OnCharacterCustomizationLoaded()
   > Set HeadMesh(HeadMesh:Skeletal Mesh Component Budgeted Object Reference=Get Array Item with EFort Custom Part Type(Array=@n15.CustomMeshParts, PartType=Head)) -> (Output_Get:Skeletal Mesh Component Budgeted Object Reference)
   > n16 = Set SkelMesh(SkelMesh=GetPartSkeletalMeshComponent())
   > Get Dynamic Material Instance(SkeletalMeshComponent=@n16)
   GetOrCreateDynamicOverlayMaterial(MeshComponent=@n16, MaterialName=None)
  ]}
 JawLineSetup_2 {
  path=EventGraph.K2Node_Composite_0.JawLineSetup_2
  nodes [
   n1 = Inputs!() -> (execute, HeadMesh:Scene Component Object Reference)
   n2 = BreakTransform(InTransform=GetSocketTransform(self=@n1, InSocketName=FACIAL_C_UnderChin, TransformSpace=RTS_ParentBoneSpace))
   > Set FACIAL_C_UnderChin_Translation(FACIAL_C_UnderChin_Translation:Vector=@n2) -> (Output_Get:Vector)
   > Set FACIAL_C_UnderChin_Rotation(FACIAL_C_UnderChin_Rotation:Rotator=@n2.Rotation) -> (Output_Get:Rotator)
   n3 = BreakTransform(InTransform=GetSocketTransform(self=@n1, InSocketName=FACIAL_C_Chin, TransformSpace=RTS_ParentBoneSpace))
   > Set FACIAL_C_Chin_Translation(FACIAL_C_Chin_Translation:Vector=@n3) -> (Output_Get:Vector)
   > Set FACIAL_C_Chin_Rotation(FACIAL_C_Chin_Rotation:Rotator=@n3.Rotation) -> (Output_Get:Rotator)
   n4 = BreakTransform(InTransform=GetSocketTransform(self=@n1, InSocketName=FACIAL_L_UnderChin, TransformSpace=RTS_ParentBoneSpace))
   > Set FACIAL_L_UnderChin_Translation(FACIAL_L_UnderChin_Translation:Vector=@n4) -> (Output_Get:Vector)
   > Set FACIAL_L_UnderChin_Rotation(FACIAL_L_UnderChin_Rotation:Rotator=@n4.Rotation) -> (Output_Get:Rotator)
   n5 = BreakTransform(InTransform=GetSocketTransform(self=@n1, InSocketName=FACIAL_R_UnderChin, TransformSpace=RTS_ParentBoneSpace))
   > Set FACIAL_R_UnderChin_Translation(FACIAL_R_UnderChin_Translation:Vector=@n5) -> (Output_Get:Vector)
   > Set FACIAL_R_UnderChin_Rotation(FACIAL_R_UnderChin_Rotation:Rotator=@n5.Rotation) -> (Output_Get:Rotator)
   n6 = BreakTransform(InTransform=GetSocketTransform(self=@n1, InSocketName=FACIAL_C_Jaw, TransformSpace=RTS_ParentBoneSpace))
   > Set FACIAL_C_Jaw_Tranform(FACIAL_C_Jaw_Tranform:Vector=@n6) -> (Output_Get:Vector)
   > Set FACIAL_C_Jaw_Rotation(FACIAL_C_Jaw_Rotation:Rotator=@n6.Rotation) -> (Output_Get:Rotator)
   n7 = BreakTransform(InTransform=GetSocketTransform(self=@n1, InSocketName=FACIAL_C_Jawline, TransformSpace=RTS_ParentBoneSpace))
   > Set FACIAL_C_Jawline_Translation(FACIAL_C_Jawline_Translation:Vector=@n7) -> (Output_Get:Vector)
   > Set FACIAL_C_Jawline_Rotation(FACIAL_C_Jawline_Rotation:Rotator=@n7.Rotation) -> (Output_Get:Rotator)
   n8 = BreakTransform(InTransform=GetSocketTransform(self=@n1, InSocketName=FACIAL_L_Jawline, TransformSpace=RTS_ParentBoneSpace))
   > Set FACIAL_L_Jawline_Translation(FACIAL_L_Jawline_Translation:Vector=@n8) -> (Output_Get:Vector)
   > Set FACIAL_L_Jawline_Rotation(FACIAL_L_Jawline_Rotation:Rotator=@n8.Rotation) -> (Output_Get:Rotator)
   n9 = BreakTransform(InTransform=GetSocketTransform(self=@n1, InSocketName=FACIAL_R_Jawline, TransformSpace=RTS_ParentBoneSpace))
   > Set FACIAL_R_Jawline_Translation(FACIAL_R_Jawline_Translation:Vector=@n9) -> (Output_Get:Vector)
   > Set FACIAL_R_Jawline_Rotation(FACIAL_R_Jawline_Rotation:Rotator=@n9.Rotation) -> (Output_Get:Rotator)
   Outputs()
  ]
 }
]
variables [
 FaceAccMesh {type=SkeletalMeshComponentBudgeted}
 FACIAL_C_Chin_Rotation {type=Rotator}
 FACIAL_C_Chin_Translation {type=Vector}
 FACIAL_C_Jaw_Rotation {type=Rotator}
 FACIAL_C_Jaw_Tranform {type=Vector}
 FACIAL_C_Jawline_Rotation {type=Rotator}
 FACIAL_C_Jawline_Translation {type=Vector}
 FACIAL_C_UnderChin_Rotation {type=Rotator}
 FACIAL_C_UnderChin_Translation {type=Vector}
 FACIAL_L_Jawline_Rotation {type=Rotator}
 FACIAL_L_Jawline_Translation {type=Vector}
 FACIAL_L_UnderChin_Rotation {type=Rotator}
 FACIAL_L_UnderChin_Translation {type=Vector}
 FACIAL_R_Jawline_Rotation {type=Rotator}
 FACIAL_R_Jawline_Translation {type=Vector}
 FACIAL_R_UnderChin_Rotation {type=Rotator}
 FACIAL_R_UnderChin_Translation {type=Vector}
 GravityOverride {type=Vector}
 HeadMesh {type=SkeletalMeshComponentBudgeted}
 isIdle {type=number}
 IsIdleFrontEnd {type=boolean}
 isIdleInFE {type=number}
 isUsingFacialBones {type=boolean}
 moveBackward {type=number}
 moveForward {type=number}
 PlayerPawn {type=FortPlayerPawnAthena}
 SkelMesh {type=SkeletalMeshComponent}
 StrafeLeftFloat {type=number}
 strafeRightFloat {type=number}
]
