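# Control Rig for the EagerUnify female face accessory that drives its long hair and front strand to follow the body.
# Weighted parent constraints attach control nulls to head, spine, clavicle, arm and pelvis bones.
# Rotation-delta alphas, spherical pose readers and pose variables like crouching weight corrective offsets on those nulls.
# B-spline IK fits bones to controls, using shorter chains when rigid bodies are enabled without back bling.
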ControlRigBlueprint /BRCosmetics/Characters/Player/Female/Medium/Bodies/F_MED_EagerUnify/Meshes/Parts/F_MED_EagerUnify_FaceAcc_CtrlRig
graphs [
 CollapseNode_ContainedGraph {nodes [
   n1 = RigVMFunctionEntryNode()
   Constraints = RigVMCollapseNode(ExecutePin=@n1.ExecuteContext)
   RigVMFunction_Sequence = RigVMAggregateNode(ExecuteContext=@Constraints.ExecutePin)
   RigVMFunction_Sequence_1 = RigVMAggregateNode(ExecuteContext=@RigVMFunction_Sequence.A)
   GetAlphaFromQuatDelta_6 = RigVMFunctionReferenceNode(
    Rotation Order=ZYX
    child=(Type=Bone,Name="clavicle_l")
    debugRotation=false
    onlyTwist=false
    outputMinMax=(X=0,Y=1)
    parent=(Type=Null,Name="clavicle_l_null")
    premult=1
    rotationMinMax=(X=0,Y=-45)
    triggerAxis=Z
    twistAxis=(X=1,Y=0,Z=0)
    ExecuteContext=@RigVMFunction_Sequence_1.C
   )
   > &RigUnit_ModifyTransforms(
    ItemToModify="((Item=(Type=Null,Name=\"dyn_hair_3_l_ctrl_null\"),Transform=(Rotation=(X=0,Y=0,Z=0,W=-1),Translation=(X=-1,Y=-1,Z=2),Scale3D=(X=1,Y=1,Z=1))),(Item=(Type=Null,Name=\"dyn_hair_4_l_ctrl_null\"),Transform=(Rotation=(X=0,Y=0,Z=0,W=-1),Translation=(X=0,Y=1,Z=4),Sca ...(truncated)"
    WeightMinimum=0
    Weight=@GetAlphaFromQuatDelta_6.alpha
   )
   > &RigUnit_ModifyTransforms(
    ItemToModify="((Item=(Type=Null,Name=\"dyn_hair_3_c_ctrl_null\"),Transform=(Rotation=(X=0,Y=0,Z=0,W=1),Translation=(X=0,Y=-3,Z=0),Scale3D=(X=1,Y=1,Z=1))),(Item=(Type=Null,Name=\"dyn_hair_4_c_ctrl_null\"),Transform=(Rotation=(X=0,Y=0,Z=0,W=1),Translation=(X=0,Y=-1,Z=0),Scale ...(truncated)"
    WeightMinimum=0
    Weight=@GetAlphaFromQuatDelta_6.alpha
   )
   > GetAlphaFromQuatDelta_7 = RigVMFunctionReferenceNode(
    Rotation Order=ZYX
    child=(Type=Bone,Name="clavicle_r")
    debugRotation=false
    onlyTwist=false
    outputMinMax=(X=0,Y=1)
    parent=(Type=Null,Name="clavicle_r_null")
    premult=1
    rotationMinMax=(X=0,Y=-45)
    triggerAxis=Z
    twistAxis=(X=1,Y=0,Z=0)
   )
   > &RigUnit_ModifyTransforms(
    ItemToModify="((Item=(Type=Null,Name=\"dyn_hair_3_r_ctrl_null\"),Transform=(Translation=(X=-2,Y=-3,Z=6),Scale3D=(X=1,Y=1,Z=1))),(Item=(Type=Null,Name=\"dyn_hair_4_r_ctrl_null\"),Transform=(Rotation=(X=0,Y=0,Z=0,W=1),Translation=(X=0,Y=-4,Z=4),Scale3D=(X=1,Y=1,Z=1))),(Item=( ...(truncated)"
    WeightMinimum=0
    Weight=@GetAlphaFromQuatDelta_7.alpha
   )
   > &RigUnit_ModifyTransforms(
    ItemToModify="((Item=(Type=Null,Name=\"dyn_hair_4_c_ctrl_null\"),Transform=(Rotation=(X=0,Y=0,Z=0,W=1),Translation=(X=0,Y=2,Z=0),Scale3D=(X=1,Y=1,Z=1))),(Item=(Type=Null,Name=\"dyn_hair_3_c_ctrl_null\"),Transform=(Rotation=(X=0,Y=0,Z=0,W=1),Translation=(X=0,Y=1,Z=0),Scale3D ...(truncated)"
    WeightMinimum=0
    Weight=@GetAlphaFromQuatDelta_7.alpha
   )
   > GetAlphaFromQuatDelta_8 = RigVMFunctionReferenceNode(
    Rotation Order=ZYX
    child=(Type=Bone,Name="clavicle_l")
    debugRotation=false
    onlyTwist=false
    outputMinMax=(X=0,Y=1)
    parent=(Type=Null,Name="clavicle_l_null")
    premult=1
    rotationMinMax=(X=0,Y=60)
    triggerAxis=Z
    twistAxis=(X=1,Y=0,Z=0)
   )
   > RigVMVariableNode(Value=@GetAlphaFromQuatDelta_8.alpha) {variable=ClavFrontDelta_l}
   > &RigUnit_ModifyTransforms(
    ItemToModify="((Item=(Type=Null,Name=\"dyn_hair_3_l_ctrl_null\"),Transform=(Rotation=(X=0,Y=0,Z=0,W=-1),Translation=(X=0,Y=-3,Z=-1),Scale3D=(X=1,Y=1,Z=1))),(Item=(Type=Null,Name=\"dyn_hair_4_l_ctrl_null\"),Transform=(Rotation=(X=0,Y=0,Z=0,W=-1),Translation=(X=0,Y=-5,Z=-1),S ...(truncated)"
    WeightMinimum=0
    Weight=@GetAlphaFromQuatDelta_8.alpha
   )
   > GetAlphaFromQuatDelta_9 = RigVMFunctionReferenceNode(
    Rotation Order=ZYX
    child=(Type=Bone,Name="clavicle_r")
    debugRotation=false
    onlyTwist=false
    outputMinMax=(X=0,Y=1)
    parent=(Type=Null,Name="clavicle_r_null")
    premult=1
    rotationMinMax=(X=0,Y=60)
    triggerAxis=Z
    twistAxis=(X=1,Y=0,Z=0)
   )
   > &RigUnit_ModifyTransforms(
    ItemToModify="((Item=(Type=Null,Name=\"dyn_hair_3_r_ctrl_null\"),Transform=(Rotation=(X=0,Y=0,Z=0,W=-1),Translation=(X=0,Y=4,Z=0),Scale3D=(X=1,Y=1,Z=1))),(Item=(Type=Null,Name=\"dyn_hair_4_r_ctrl_null\"),Transform=(Rotation=(X=0,Y=0,Z=0,W=-1),Translation=(X=0,Y=6,Z=-1),Scal ...(truncated)"
    WeightMinimum=0
    Weight=@GetAlphaFromQuatDelta_9.alpha
   )
   > GetAlphaFromQuatDelta_10 = RigVMFunctionReferenceNode(
    Rotation Order=ZYX
    child=(Type=Bone,Name="clavicle_l")
    debugRotation=false
    onlyTwist=false
    outputMinMax=(X=0,Y=1)
    parent=(Type=Null,Name="clavicle_l_null")
    premult=1
    rotationMinMax=(X=0,Y=-60)
    triggerAxis=Y
    twistAxis=(X=1,Y=0,Z=0)
   )
   > RigVMVariableNode(Value=@GetAlphaFromQuatDelta_10.alpha) {variable=ClavUpDelta_l}
   > &RigUnit_ModifyTransforms(
    ItemToModify="((Item=(Type=Null,Name=\"dyn_hair_3_l_ctrl_null\"),Transform=(Rotation=(X=0,Y=0,Z=0,W=-1),Translation=(X=0,Y=4,Z=3),Scale3D=(X=1,Y=1,Z=1))),(Item=(Type=Null,Name=\"dyn_hair_4_l_ctrl_null\"),Transform=(Rotation=(X=0,Y=0,Z=0,W=-1),Translation=(X=0,Y=-1,Z=-0.5),S ...(truncated)"
    WeightMinimum=0
    Weight=@GetAlphaFromQuatDelta_10.alpha
   )
   > GetAlphaFromQuatDelta_11 = RigVMFunctionReferenceNode(
    Rotation Order=ZYX
    child=(Type=Bone,Name="clavicle_r")
    debugRotation=false
    onlyTwist=false
    outputMinMax=(X=0,Y=1)
    parent=(Type=Null,Name="clavicle_r_null")
    premult=1
    rotationMinMax=(X=0,Y=-60)
    triggerAxis=Y
    twistAxis=(X=1,Y=0,Z=0)
   )
   > &RigUnit_ModifyTransforms(
    ItemToModify="((Item=(Type=Null,Name=\"dyn_hair_3_r_ctrl_null\"),Transform=(Rotation=(X=0,Y=0,Z=0,W=-1),Translation=(X=0,Y=-4,Z=3),Scale3D=(X=1,Y=1,Z=1))),(Item=(Type=Null,Name=\"dyn_hair_4_r_ctrl_null\"),Transform=(Rotation=(X=0,Y=0,Z=0,W=-1),Translation=(X=0,Y=1,Z=0),Scal ...(truncated)"
    WeightMinimum=0
    Weight=@GetAlphaFromQuatDelta_11.alpha
   )
   n2 = RigVMVariableNode() {variable=ClavUpDelta_l}
   > &RigUnit_ModifyTransforms(
    ItemToModify="((Item=(Type=Null,Name=\"dyn_frontHair_A_03_l_ctrl_null\"),Transform=(Rotation=(X=0,Y=0,Z=0,W=1),Translation=(X=0,Y=5,Z=0),Scale3D=(X=1,Y=1,Z=1))),(Item=(Type=Null,Name=\"dyn_frontHair_A_04_l_ctrl_null\"),Transform=(Rotation=(X=0,Y=0,Z=0,W=1),Translation=(X=0, ...(truncated)"
    WeightMinimum=0
    Weight=&RigVMFunction_MathDoubleSub(B=&RigVMFunction_MathDoubleMul(B=0.25, A=@n2.Value).Result, A=RigVMVariableNode() {variable=ClavFrontDelta_l}.Value).Result
   )
   n3 = RigVMVariableNode() {variable=ClavUpDelta_l}
   > &RigUnit_ModifyTransforms(
    ItemToModify="((Item=(Type=Null,Name=\"dyn_frontHair_A_02_l_ctrl_null\"),Transform=(Rotation=(X=0,Y=0,Z=0,W=1),Translation=(X=-2,Y=8,Z=0),Scale3D=(X=1,Y=1,Z=1))),(Item=(Type=Null,Name=\"dyn_frontHair_A_03_l_ctrl_null\"),Transform=(Rotation=(X=0,Y=0,Z=0,W=1),Translation=(X=0 ...(truncated)"
    WeightMinimum=0
    Weight=&RigVMFunction_MathDoubleMul(
     B=&RigVMFunction_MathDoubleRemap(SourceMaximum=0.8, SourceMinimum=0, TargetMaximum=1, TargetMinimum=0, Value=@n3.Value).Result
     A=RigVMVariableNode() {variable=ClavFrontDelta_l}.Value
    ).Result
   )
   GetAlphaFromQuatDelta_12 = RigVMFunctionReferenceNode(
    Rotation Order=ZYX
    child=(Type=Bone,Name="upperarm_l")
    debugRotation=false
    onlyTwist=false
    outputMinMax=(X=0,Y=1)
    parent=(Type=Null,Name="upperarm_l_null")
    premult=1
    rotationMinMax=(X=0,Y=90)
    triggerAxis=Z
    twistAxis=(X=1,Y=0,Z=0)
    ExecuteContext=@RigVMFunction_Sequence_1.E
   )
   RigVMFunction_AnimEvalRichCurve = &RigVMFunction_AnimEvalRichCurve(
    Curve=(EditorCurveData=(Keys=((InterpMode=RCIM_Cubic,TangentMode=RCTM_User,ArriveTangent=0.0663,LeaveTangent=0.0663),(InterpMode=RCIM_Cubic,TangentMode=RCTM_User,Time=1,Value=1,ArriveTangent=2.5239,LeaveTangent=2.5239))))
    SourceMinimum=0
    TargetMinimum=0
    Value=@GetAlphaFromQuatDelta_12.alpha
   )
   > &RigUnit_ModifyTransforms(
    ItemToModify="((Item=(Type=Null,Name=\"dyn_frontHair_A_02_l_ctrl_null\"),Transform=(Rotation=(X=0,Y=0,Z=0,W=-1),Translation=(X=0,Y=0.5835,Z=-0.0299),Scale3D=(X=1,Y=1,Z=1))),(Item=(Type=Null,Name=\"dyn_frontHair_A_03_l_ctrl_null\"),Transform=(Rotation=(X=0,Y=0,Z=0,W=1),Trans ...(truncated)"
    WeightMinimum=0
    Weight=@RigVMFunction_AnimEvalRichCurve.Result
   )
   > GetAlphaFromQuatDelta = RigVMFunctionReferenceNode(
    Rotation Order=ZYX
    child=(Type=Bone,Name="head")
    debugRotation=false
    outputMinMax=(X=0,Y=1)
    parent=(Type=Null,Name="head_null")
    premult=1
    rotationMinMax=(X=0,Y=20)
    triggerAxis=Z
    twistAxis=(X=1,Y=0,Z=0)
   )
   > &RigUnit_ModifyTransforms(
    ItemToModify=((Item=(Type=Null,Name="dyn_frontHair_A_01_l_ctrl_null"),Transform=(Rotation=(X=0,Y=0,Z=0,W=1),Translation=(X=0,Y=1,Z=0),Scale3D=(X=1,Y=1,Z=1))))
    WeightMinimum=0
    Weight=@GetAlphaFromQuatDelta.alpha
   )
   > GetAlphaFromQuatDelta_13 = RigVMFunctionReferenceNode(
    Rotation Order=ZYX
    child=(Type=Bone,Name="head")
    debugRotation=false
    onlyTwist=true
    outputMinMax=(X=0,Y=1)
    parent=(Type=Null,Name="head_null")
    premult=1
    rotationMinMax=(X=-20,Y=-80)
    triggerAxis=X
    twistAxis=(X=1,Y=0,Z=0)
   )
   n4 = &RigVMFunction_AlphaInterp(InterpSpeedDecreasing=10, InterpSpeedIncreasing=10, Scale=1, bClampResult=False, Value=@GetAlphaFromQuatDelta_13.alpha)
   > &RigUnit_ModifyTransforms(
    ItemToModify="((Item=(Type=Null,Name=\"dyn_frontHair_A_01_l_ctrl_null\"),Transform=(Rotation=(X=0,Y=0,Z=0,W=-1),Translation=(X=0,Y=0,Z=1),Scale3D=(X=1,Y=1,Z=1))),(Item=(Type=Null,Name=\"dyn_frontHair_A_02_l_ctrl_null\"),Transform=(Rotation=(X=0,Y=0,Z=0,W=-1),Translation=(X= ...(truncated)"
    WeightMinimum=0
    Weight=@n4.Result
   )
   GetAlphaFromQuatDelta_4 = RigVMFunctionReferenceNode(
    Rotation Order=ZYX
    child=(Type=Bone,Name="spine_05")
    debugRotation=false
    onlyTwist=True
    outputMinMax=(X=0,Y=1)
    parent=(Type=Bone,Name="spine_03")
    premult=1
    rotationMinMax=(X=0,Y=35)
    triggerAxis=X
    twistAxis=(X=1,Y=0,Z=0)
    ExecuteContext=@RigVMFunction_Sequence_1.B
   )
   > RigVMVariableNode(Value=@GetAlphaFromQuatDelta_4.alpha) {variable=TorsoTwistRight}
   > &RigUnit_ModifyTransforms(
    ItemToModify="((Item=(Type=Null,Name=\"dyn_hair_4_r_ctrl_null\"),Transform=(Rotation=(X=0,Y=0,Z=0,W=1),Translation=(X=0,Y=-2.112,Z=2),Scale3D=(X=1,Y=1,Z=1))),(Item=(Type=Null,Name=\"dyn_hair_5_r_ctrl_null\"),Transform=(Rotation=(X=0,Y=0,Z=0,W=-1),Translation=(X=0,Y=0,Z=2.60 ...(truncated)"
    WeightMinimum=0
    Weight=@GetAlphaFromQuatDelta_4.alpha
   )
   > &RigUnit_ModifyTransforms(
    ItemToModify="((Item=(Type=Null,Name=\"dyn_hair_3_r_ctrl_null\"),Transform=(Rotation=(X=0,Y=0,Z=0,W=1),Translation=(X=0,Y=0,Z=3),Scale3D=(X=1,Y=1,Z=1))),(Item=(Type=Null,Name=\"dyn_hair_2_r_ctrl_null\"),Transform=(Rotation=(X=0,Y=0,Z=0,W=1),Translation=(X=0,Y=0,Z=2.0512),Sc ...(truncated)"
    WeightMinimum=0
    Weight=@GetAlphaFromQuatDelta_4.alpha
   )
   > &RigUnit_ModifyTransforms(
    ItemToModify="((Item=(Type=Null,Name=\"dyn_frontHair_A_04_l_ctrl_null\"),Transform=(Rotation=(X=0,Y=0,Z=0,W=1),Translation=(X=0,Y=2,Z=0),Scale3D=(X=1,Y=1,Z=1))),(Item=(Type=Null,Name=\"dyn_frontHair_A_05_l_ctrl_null\"),Transform=(Rotation=(X=0,Y=0,Z=0,W=1),Translation=(X=0, ...(truncated)"
    WeightMinimum=0
    Weight=@GetAlphaFromQuatDelta_4.alpha
   )
   > GetAlphaFromQuatDelta_5 = RigVMFunctionReferenceNode(
    Rotation Order=ZYX
    child=(Type=Bone,Name="spine_05")
    onlyTwist=True
    outputMinMax=(X=0,Y=1)
    parent=(Type=Bone,Name="spine_03")
    premult=0
    rotationMinMax=(X=0,Y=-35)
    triggerAxis=X
    twistAxis=(X=1,Y=0,Z=0)
   )
   > &RigUnit_ModifyTransforms(
    ItemToModify="((Item=(Type=Null,Name=\"dyn_hair_4_l_ctrl_null\"),Transform=(Rotation=(X=0,Y=0,Z=0,W=1),Translation=(X=0,Y=-2.112,Z=3.5606),Scale3D=(X=1,Y=1,Z=1))),(Item=(Type=Null,Name=\"dyn_hair_5_l_ctrl_null\"),Transform=(Rotation=(X=0,Y=0,Z=0,W=-1),Translation=(X=0,Y=0,Z ...(truncated)"
    WeightMinimum=0
    Weight=@GetAlphaFromQuatDelta_5.alpha
   )
   n5 = &RigUnit_SphericalPoseReader(
    ActiveRegionScaleFactors=(PositiveWidth=0.5,NegativeWidth=0.5,PositiveHeight=0.5,NegativeHeight=0.5)
    ActiveRegionSize=0.1
    Debug=(bDrawDebug=false,bDraw2D=False,bDrawLocalAxes=False,DebugScale=25,DebugSegments=20,DebugThickness=0.25)
    DriverAxis=(X=-1,Y=0,Z=0)
    DriverItem=(Type=Bone,Name="upperarm_r")
    FalloffRegionScaleFactors=(PositiveWidth=1,NegativeWidth=1,PositiveHeight=1,NegativeHeight=1)
    FalloffSize=0.2
    FlipHeightScaling=False
    FlipWidthScaling=False
    OptionalParentItem=(Type=Bone,Name="clavicle_r")
    RotationOffset=(X=120,Y=0,Z=-140)
    ExecutePin=@RigVMFunction_Sequence_1.D
   )
   AlphaInterp = &RigVMFunction_AlphaInterp(InterpSpeedDecreasing=8, InterpSpeedIncreasing=14, Scale=1, bClampResult=False, Value=@n5.OutputParam)
   > RigVMVariableNode(Value=@AlphaInterp.Result) {variable=isPistolASDUp}
   > &RigUnit_ModifyTransforms(
    ItemToModify="((Item=(Type=Null,Name=\"dyn_hair_2_r_ctrl_null\"),Transform=(Rotation=(X=0,Y=0,Z=0,W=1),Translation=(X=0,Y=0,Z=0),Scale3D=(X=1,Y=1,Z=1))),(Item=(Type=Null,Name=\"dyn_hair_2_r_ctrl_null\"),Transform=(Rotation=(X=0,Y=0,Z=0,W=1),Translation=(X=0,Y=0,Z=0),Scale3D ...(truncated)"
    WeightMinimum=0
    Weight=@AlphaInterp.Result
   )
   n6 = RigVMVariableNode() {variable=isPistolASDUp}
   n7 = &RigVMFunction_AlphaInterp(
    InterpSpeedDecreasing=8
    InterpSpeedIncreasing=10
    Scale=1
    bClampResult=False
    Value=&RigVMFunction_MathDoubleMul(
     A=&RigVMFunction_MathDoubleRemap(SourceMaximum=0.65, SourceMinimum=0.1, TargetMaximum=1, TargetMinimum=0, Value=@n6.Value).Result
     B=RigVMVariableNode() {variable=Crouching}.Value
    ).Result
   )
   > &RigUnit_ModifyTransforms(
    ItemToModify="((Item=(Type=Null,Name=\"dyn_hair_3_r_ctrl_null\"),Transform=(Rotation=(X=0,Y=0,Z=0,W=1),Translation=(X=0,Y=-1,Z=-2),Scale3D=(X=1,Y=1,Z=1))),(Item=(Type=Bone,Name=\"dyn_hair2Curve_r\"),Transform=(Rotation=(X=0,Y=0,Z=0,W=1),Translation=(X=-1,Y=-2,Z=-1),Scale3D= ...(truncated)"
    WeightMinimum=0
    Weight=@n7.Result
   )
   > n8 = &RigUnit_SphericalPoseReader(
    ActiveRegionScaleFactors=(PositiveWidth=1,NegativeWidth=1,PositiveHeight=1,NegativeHeight=1)
    ActiveRegionSize=0.035
    Debug=(bDrawDebug=false,bDraw2D=False,bDrawLocalAxes=False,DebugScale=25,DebugSegments=20,DebugThickness=0.25)
    DriverAxis=(X=1,Y=0,Z=0)
    DriverItem=(Type=Bone,Name="head")
    FalloffRegionScaleFactors=(PositiveWidth=1,NegativeWidth=1,PositiveHeight=1,NegativeHeight=1)
    FalloffSize=0.2
    FlipHeightScaling=False
    FlipWidthScaling=False
    OptionalParentItem=(Type=Bone,Name="spine_05")
    RotationOffset=(X=-48,Y=0,Z=45)
   )
   > RigVMVariableNode(Value=@n8.OutputParam) {variable=HeadADSBigGun}
   > &RigUnit_ModifyTransforms(
    ItemToModify="((Item=(Type=Null,Name=\"dyn_hair_1_r_ctrl_null\"),Transform=(Rotation=(X=0,Y=0,Z=0,W=1),Translation=(X=-4,Y=1,Z=0),Scale3D=(X=1,Y=1,Z=1))),(Item=(Type=Null,Name=\"dyn_hair_2_r_ctrl_null\"),Transform=(Rotation=(X=0,Y=0,Z=0,W=1),Translation=(X=0,Y=-4,Z=2),Scale ...(truncated)"
    WeightMinimum=0
    Weight=@n8.OutputParam
   )
   > &RigUnit_ModifyTransforms(
    ItemToModify="((Item=(Type=Null,Name=\"dyn_hair_1_c_ctrl_null\"),Transform=(Rotation=(X=0,Y=0,Z=0,W=1),Translation=(X=0,Y=-1,Z=0),Scale3D=(X=1,Y=1,Z=1))),(Item=(Type=Null,Name=\"dyn_hair_2_c_ctrl_null\"),Transform=(Rotation=(X=0,Y=0,Z=0,W=1),Translation=(X=0,Y=-2,Z=0),Scale ...(truncated)"
    WeightMinimum=0
    Weight=@n8.OutputParam
   )
   > &RigUnit_ModifyTransforms(
    ItemToModify="((Item=(Type=Null,Name=\"dyn_hair_2_l_ctrl_null\"),Transform=(Rotation=(X=0,Y=0,Z=0,W=1),Translation=(X=1,Y=-3,Z=0),Scale3D=(X=1,Y=1,Z=1))),(Item=(Type=Null,Name=\"dyn_hair_3_l_ctrl_null\"),Transform=(Rotation=(X=0,Y=0,Z=0,W=1),Translation=(X=0,Y=-1,Z=0),Scale ...(truncated)"
    WeightMinimum=0
    Weight=@n8.OutputParam
   )
   > &RigUnit_ModifyTransforms(
    ItemToModify="((Item=(Type=Null,Name=\"dyn_hair_3_r_ctrl_null\"),Transform=(Rotation=(X=0,Y=0,Z=0,W=1),Translation=(X=0,Y=-2,Z=-3),Scale3D=(X=1,Y=1,Z=1))),(Item=(Type=Null,Name=\"dyn_hair_4_r_ctrl_null\"),Transform=(Rotation=(X=0,Y=0,Z=0,W=1),Translation=(X=0,Y=-2,Z=0),Scal ...(truncated)"
    WeightMinimum=0
    Weight=&RigVMFunction_MathDoubleMul(A=RigVMVariableNode() {variable=HeadADSBigGun}.Value, B=RigVMVariableNode() {variable=CrouchingInPlace}.Value).Result
   )
   &RigUnit_ModifyTransforms(
    ItemToModify=((Item=(Type=Bone,Name="hair_0_c"),Transform=(Rotation=(X=0,Y=0,Z=0,W=1),Translation=(X=-8,Y=-1,Z=6),Scale3D=(X=1,Y=1,Z=1))))
    WeightMinimum=0
    Weight=RigVMVariableNode() {variable=headLookingUp}.Value
    ExecutePin=@RigVMFunction_Sequence_1.A
   )
   > GetAlphaFromQuatDelta_1 = RigVMFunctionReferenceNode(
    Rotation Order=ZYX
    child=(Type=Bone,Name="head")
    debugRotation=false
    onlyTwist=True
    outputMinMax=(X=0,Y=1)
    parent=(Type=Bone,Name="neck_01")
    premult=1
    rotationMinMax=(X=0,Y=-60)
    triggerAxis=X
    twistAxis=(X=1,Y=0,Z=0)
   )
   > &RigUnit_ModifyTransforms(
    ItemToModify="((Item=(Type=Null,Name=\"dyn_hair_2_r_ctrl_null\"),Transform=(Rotation=(X=0,Y=0,Z=0,W=1),Translation=(X=0,Y=0,Z=3),Scale3D=(X=1,Y=1,Z=1))),(Item=(Type=Null,Name=\"dyn_hair_3_r_ctrl_null\"),Transform=(Rotation=(X=0,Y=0,Z=0,W=-1),Translation=(X=0,Y=0,Z=3),Scale3 ...(truncated)"
    WeightMinimum=0
    Weight=@GetAlphaFromQuatDelta_1.alpha
   )
   > GetAlphaFromQuatDelta_2 = RigVMFunctionReferenceNode(
    Rotation Order=ZYX
    child=(Type=Bone,Name="head")
    onlyTwist=True
    outputMinMax=(X=0,Y=1)
    parent=(Type=Bone,Name="neck_01")
    premult=1
    rotationMinMax=(X=0,Y=60)
    triggerAxis=X
    twistAxis=(X=1,Y=0,Z=0)
   )
   > &RigUnit_ModifyTransforms(
    ItemToModify="((Item=(Type=Null,Name=\"dyn_hair_1_r_ctrl_null\"),Transform=(Rotation=(X=0,Y=0,Z=0,W=1),Translation=(X=-0.2341,Y=-3.6072,Z=-0.9853),Scale3D=(X=1,Y=1,Z=1))),(Item=(Type=Null,Name=\"dyn_hair_2_r_ctrl_null\"),Transform=(Rotation=(X=0,Y=0,Z=0,W=-1),Translation=(X ...(truncated)"
    WeightMinimum=0
    Weight=@GetAlphaFromQuatDelta_2.alpha
   )
   > GetAlphaFromQuatDelta_3 = RigVMFunctionReferenceNode(
    Rotation Order=ZYX
    child=(Type=Bone,Name="head")
    debugRotation=false
    onlyTwist=False
    outputMinMax=(X=0,Y=1)
    parent=(Type=Null,Name="head_null")
    premult=1
    rotationMinMax=(X=-60,Y=-80)
    triggerAxis=Z
    twistAxis=(X=1,Y=0,Z=0)
   )
   > RigVMVariableNode(Value=@GetAlphaFromQuatDelta_3.alpha) {variable=head rz n60n80 alpha}
   > &RigUnit_ModifyTransforms(
    ItemToModify="((Item=(Type=Null,Name=\"dyn_hair_3_r_ctrl_null\"),Transform=(Rotation=(X=0,Y=0,Z=0,W=-1),Translation=(X=2,Y=-1,Z=0),Scale3D=(X=1,Y=1,Z=1))),(Item=(Type=Null,Name=\"dyn_hair_2_l_ctrl_null\"),Transform=(Rotation=(X=0,Y=0,Z=0,W=-1),Translation=(X=7.6214,Y=0,Z=0) ...(truncated)"
    WeightMinimum=0
    Weight=@GetAlphaFromQuatDelta_3.alpha
   )
   > &RigUnit_ModifyTransforms(
    ItemToModify="((Item=(Type=Null,Name=\"dyn_hair_1_r_ctrl_null\"),Transform=(Rotation=(X=0,Y=0,Z=0,W=1),Translation=(X=-2,Y=0,Z=0),Scale3D=(X=1,Y=1,Z=1))),(Item=(Type=Null,Name=\"dyn_hair_2_r_ctrl_null\"),Transform=(Rotation=(X=0,Y=0,Z=0,W=1),Translation=(X=2,Y=0,Z=-2),Scale ...(truncated)"
    WeightMinimum=0
    Weight=@GetAlphaFromQuatDelta_3.alpha
   )
   > &RigUnit_ModifyTransforms(
    ItemToModify="((Item=(Type=Null,Name=\"dyn_hair_1_r_ctrl_null\"),Transform=(Rotation=(X=0,Y=0,Z=0,W=1),Translation=(X=-1,Y=-2,Z=1),Scale3D=(X=1,Y=1,Z=1))),(Item=(Type=Null,Name=\"dyn_hair_2_r_ctrl_null\"),Transform=(Rotation=(X=0,Y=0,Z=0,W=1),Translation=(X=3,Y=-1,Z=0),Scal ...(truncated)"
    WeightMinimum=0
    Weight=&RigVMFunction_MathDoubleRemap(SourceMaximum=0.5, SourceMinimum=0, TargetMaximum=0, TargetMinimum=1, Value=RigVMVariableNode() {variable=headTiltingSideways}.Value).Result
   )
   > &RigUnit_ModifyTransforms(
    ItemToModify="((Item=(Type=Null,Name=\"dyn_hair_1_r_ctrl_null\"),Transform=(Rotation=(X=0,Y=0,Z=0,W=1),Translation=(X=-5,Y=2,Z=0),Scale3D=(X=1,Y=1,Z=1))),(Item=(Type=Null,Name=\"dyn_hair_3_r_ctrl_null\"),Transform=(Rotation=(X=0,Y=0,Z=0,W=1),Translation=(X=0,Y=-1,Z=1),Scale ...(truncated)"
    WeightMinimum=0
    Weight=RigVMVariableNode() {variable=HeadLookingUp_Crouching}.Value
   )
   RigVMFunctionReturnNode(ExecuteContext=@RigVMFunction_Sequence.D)
   IkTwists = RigVMCollapseNode(ExecuteContext=@RigVMFunction_Sequence.B)
   RigVMFunction_MathBoolNand2 = /Script/RigVM.RigVMFunction_MathBoolNand2(
    B=&RigVMFunction_MathBoolNot(Value=RigVMVariableNode() {variable=pub isWearingBackBling}.Value).Result
    A=RigVMVariableNode() {variable=pub isRigidBodies Enabled}.Value
   ) {is=struct RigVMUnitNode, template="Nand::Execute(in A,in B,out Result)"}
   RigVMFunction_ControlFlowBranch = &RigVMFunction_ControlFlowBranch(ExecuteContext=@IkTwists.ExecuteContext, Condition=@RigVMFunction_MathBoolNand2.Result)
   Post_IK_Fixes = RigVMCollapseNode(ExecuteContext=@RigVMFunction_ControlFlowBranch.Completed)
   controls_ik_c_x7 = RigVMDispatchNode(Value="((Type=Control,Name=\"dyn_hair_1_c_ctrl\"),(Type=Control,Name=\"dyn_hair_2_c_ctrl\"),(Type=Control,Name=\"dyn_hair_3_c_ctrl\"),(Type=Control,Name=\"dyn_hair_4_c_ctrl\"),(Type=Control,Name=\"dyn_hair_5_c_ctrl\"),(Type=Control,Name=\"dyn_hair_6_c_ctrl\"),(Type=Control,N ...(truncated)") {template="DISPATCH_RigVMDispatch_Constant(io Value)"}
   joints_ik_c_f_x7 = RigVMDispatchNode(Value=((Type=Bone,Name="dyn_hair_1_c"),(Type=Bone,Name="dyn_hair_2_c"),(Type=Bone,Name="dyn_hair_3_c"),(Type=Bone,Name="dyn_hair_4_c"),(Type=Bone,Name="dyn_hair_5_c"),(Type=Bone,Name="dyn_hair_6_c"),(Type=Bone,Name="dyn_hair_7_c"))) {template="DISPATCH_RigVMDispatch_Constant(io Value)"}
   SplineIK_6 = RigVMFunctionReferenceNode(
    Primary Axis=(X=1,Y=0,Z=0)
    Samples Per Segment=1
    Secondary Spline Direction=(X=0,Y=0,Z=0)
    Spline Mode=BSpline
    Stretch=False
    Up Axis=(X=0,Y=0,Z=0)
    debug=false
    Controls=@controls_ik_c_x7.Value
    Bones=@joints_ik_c_f_x7.Value
    ExecuteContext=@RigVMFunction_ControlFlowBranch.True
   )
   controls_ik_l_f_x7 = RigVMDispatchNode(Value="((Type=Control,Name=\"dyn_hair_1_l_ctrl\"),(Type=Control,Name=\"dyn_hair_2_l_ctrl\"),(Type=Control,Name=\"dyn_hair_3_l_ctrl\"),(Type=Control,Name=\"dyn_hair_4_l_ctrl\"),(Type=Control,Name=\"dyn_hair_5_l_ctrl\"),(Type=Control,Name=\"dyn_hair_6_l_ctrl\"),(Type=Control,N ...(truncated)") {template="DISPATCH_RigVMDispatch_Constant(io Value)"}
   joints_ik_l_f_x7 = RigVMDispatchNode(Value=((Type=Bone,Name="dyn_hair_1_l"),(Type=Bone,Name="dyn_hair_2_l"),(Type=Bone,Name="dyn_hair_3_l"),(Type=Bone,Name="dyn_hair_4_l"),(Type=Bone,Name="dyn_hair_5_l"),(Type=Bone,Name="dyn_hair_6_l"),(Type=Bone,Name="dyn_hair_7_l"))) {template="DISPATCH_RigVMDispatch_Constant(io Value)"}
   > SplineIK_7 = RigVMFunctionReferenceNode(
    Primary Axis=(X=1,Y=0,Z=0)
    Samples Per Segment=1
    Secondary Spline Direction=(X=0,Y=0,Z=0)
    Spline Mode=BSpline
    Stretch=False
    Up Axis=(X=0,Y=0,Z=0)
    debug=false
    Controls=@controls_ik_l_f_x7.Value
    Bones=@joints_ik_l_f_x7.Value
   )
   controls_ik_r_x7 = RigVMDispatchNode(Value="((Type=Control,Name=\"dyn_hair_1_r_ctrl\"),(Type=Control,Name=\"dyn_hair_2_r_ctrl\"),(Type=Control,Name=\"dyn_hair_3_r_ctrl\"),(Type=Control,Name=\"dyn_hair_4_r_ctrl\"),(Type=Control,Name=\"dyn_hair_5_r_ctrl\"),(Type=Control,Name=\"dyn_hair_6_r_ctrl\"),(Type=Control,N ...(truncated)") {template="DISPATCH_RigVMDispatch_Constant(io Value)"}
   joints_ik_r_f_x7 = RigVMDispatchNode(Value=((Type=Bone,Name="dyn_hair_1_r"),(Type=Bone,Name="dyn_hair_2_r"),(Type=Bone,Name="dyn_hair_3_r"),(Type=Bone,Name="dyn_hair_4_r"),(Type=Bone,Name="dyn_hair_5_r"),(Type=Bone,Name="dyn_hair_6_r"),(Type=Bone,Name="dyn_hair_7_r"))) {template="DISPATCH_RigVMDispatch_Constant(io Value)"}
   > SplineIK_8 = RigVMFunctionReferenceNode(
    Primary Axis=(X=1,Y=0,Z=0)
    Samples Per Segment=1
    Secondary Spline Direction=(X=0,Y=0,Z=0)
    Spline Mode=BSpline
    Stretch=False
    Up Axis=(X=0,Y=0,Z=0)
    debug=false
    Controls=@controls_ik_r_x7.Value
    Bones=@joints_ik_r_f_x7.Value
   )
   ItemArray_3 = RigVMDispatchNode(Value="((Type=Bone,Name=\"dyn_frontHair_A_01_l\"),(Type=Bone,Name=\"dyn_frontHair_A_02_l\"),(Type=Bone,Name=\"dyn_frontHair_A_03_l\"),(Type=Bone,Name=\"dyn_frontHair_A_04_l\"),(Type=Bone,Name=\"dyn_frontHair_A_05_l\"),(Type=Bone,Name=\"dyn_frontHair_A_06_l\"),(Type=Bone,Name ...(truncated)") {template="DISPATCH_RigVMDispatch_Constant(io Value)"}
   ItemArray_4 = RigVMDispatchNode(Value="((Type=Control,Name=\"dyn_frontHair_A_01_l_ctrl\"),(Type=Control,Name=\"dyn_frontHair_A_02_l_ctrl\"),(Type=Control,Name=\"dyn_frontHair_A_03_l_ctrl\"),(Type=Control,Name=\"dyn_frontHair_A_04_l_ctrl\"),(Type=Control,Name=\"dyn_frontHair_A_05_l_ctrl\"),(Type=Control,N ...(truncated)") {template="DISPATCH_RigVMDispatch_Constant(io Value)"}
   > SplineIK_3 = RigVMFunctionReferenceNode(
    Primary Axis=(X=1,Y=0,Z=0)
    Samples Per Segment=1
    Secondary Spline Direction=(X=0,Y=0,Z=0)
    Spline Mode=BSpline
    Stretch=False
    Up Axis=(X=0,Y=0,Z=0)
    debug=false
    Bones=@ItemArray_3.Value
    Controls=@ItemArray_4.Value
   )
   controls_ik_c_x4 = RigVMDispatchNode(Value=((Type=Control,Name="dyn_hair_1_c_ctrl"),(Type=Control,Name="dyn_hair_2_c_ctrl"),(Type=Control,Name="dyn_hair_3_c_ctrl"),(Type=Control,Name="dyn_hair_4_c_ctrl"),(Type=Control,Name="dyn_hair_5_c_ctrl"))) {template="DISPATCH_RigVMDispatch_Constant(io Value)"}
   joints_ik_c_x4 = RigVMDispatchNode(Value=((Type=Bone,Name="dyn_hair_1_c"),(Type=Bone,Name="dyn_hair_2_c"),(Type=Bone,Name="dyn_hair_3_c"),(Type=Bone,Name="dyn_hair_4_c"),(Type=Bone,Name="dyn_hair_5_c"))) {template="DISPATCH_RigVMDispatch_Constant(io Value)"}
   SplineIK = RigVMFunctionReferenceNode(
    Primary Axis=(X=1,Y=0,Z=0)
    Samples Per Segment=1
    Secondary Spline Direction=(X=0,Y=0,Z=0)
    Spline Mode=BSpline
    Stretch=False
    Up Axis=(X=0,Y=0,Z=0)
    debug=false
    Controls=@controls_ik_c_x4.Value
    Bones=@joints_ik_c_x4.Value
    ExecuteContext=@RigVMFunction_ControlFlowBranch.False
   )
   controls_ik_l_x4 = RigVMDispatchNode(Value=((Type=Control,Name="dyn_hair_1_l_ctrl"),(Type=Control,Name="dyn_hair_2_l_ctrl"),(Type=Control,Name="dyn_hair_3_l_ctrl"),(Type=Control,Name="dyn_hair_4_l_ctrl"),(Type=Control,Name="dyn_hair_5_l_ctrl"))) {template="DISPATCH_RigVMDispatch_Constant(io Value)"}
   joints_ik_l_x4 = RigVMDispatchNode(Value=((Type=Bone,Name="dyn_hair_1_l"),(Type=Bone,Name="dyn_hair_2_l"),(Type=Bone,Name="dyn_hair_3_l"),(Type=Bone,Name="dyn_hair_4_l"),(Type=Bone,Name="dyn_hair_5_l"))) {template="DISPATCH_RigVMDispatch_Constant(io Value)"}
   > SplineIK_1 = RigVMFunctionReferenceNode(
    Primary Axis=(X=1,Y=0,Z=0)
    Samples Per Segment=1
    Secondary Spline Direction=(X=0,Y=0,Z=0)
    Spline Mode=BSpline
    Stretch=False
    Up Axis=(X=0,Y=0,Z=0)
    debug=false
    Controls=@controls_ik_l_x4.Value
    Bones=@joints_ik_l_x4.Value
   )
   controls_ik_r_x4 = RigVMDispatchNode(Value=((Type=Control,Name="dyn_hair_1_r_ctrl"),(Type=Control,Name="dyn_hair_2_r_ctrl"),(Type=Control,Name="dyn_hair_3_r_ctrl"),(Type=Control,Name="dyn_hair_4_r_ctrl"),(Type=Control,Name="dyn_hair_5_r_ctrl"))) {template="DISPATCH_RigVMDispatch_Constant(io Value)"}
   joints_ik_r_x4 = RigVMDispatchNode(Value=((Type=Bone,Name="dyn_hair_1_r"),(Type=Bone,Name="dyn_hair_2_r"),(Type=Bone,Name="dyn_hair_3_r"),(Type=Bone,Name="dyn_hair_4_r"),(Type=Bone,Name="dyn_hair_5_r"))) {template="DISPATCH_RigVMDispatch_Constant(io Value)"}
   > SplineIK_2 = RigVMFunctionReferenceNode(
    Primary Axis=(X=1,Y=0,Z=0)
    Samples Per Segment=1
    Secondary Spline Direction=(X=0,Y=0,Z=0)
    Spline Mode=BSpline
    Stretch=False
    Up Axis=(X=0,Y=0,Z=0)
    debug=false
    Controls=@controls_ik_r_x4.Value
    Bones=@joints_ik_r_x4.Value
   )
   ItemArray_1 = RigVMDispatchNode(Value=((Type=Control,Name="dyn_frontHair_A_01_l_ctrl"),(Type=Control,Name="dyn_frontHair_A_02_l_ctrl"),(Type=Control,Name="dyn_frontHair_A_03_l_ctrl"),(Type=Control,Name="dyn_frontHair_A_04_l_ctrl"))) {template="DISPATCH_RigVMDispatch_Constant(io Value)"}
   ItemArray_2 = RigVMDispatchNode(Value=((Type=Bone,Name="dyn_frontHair_A_01_l"),(Type=Bone,Name="dyn_frontHair_A_02_l"),(Type=Bone,Name="dyn_frontHair_A_03_l"),(Type=Bone,Name="dyn_frontHair_A_04_l"))) {template="DISPATCH_RigVMDispatch_Constant(io Value)"}
   > SplineIK_4 = RigVMFunctionReferenceNode(
    Primary Axis=(X=1,Y=0,Z=0)
    Samples Per Segment=1
    Secondary Spline Direction=(X=0,Y=0,Z=0)
    Spline Mode=BSpline
    Stretch=False
    Up Axis=(X=0,Y=0,Z=0)
    debug=false
    Controls=@ItemArray_1.Value
    Bones=@ItemArray_2.Value
   )
   n9 = &RigVMFunction_MathDoubleRemap(SourceMaximum=0.5, SourceMinimum=0, TargetMaximum=0, TargetMinimum=1, Value=RigVMVariableNode() {variable=headTiltingSideways}.Value)
   n10 = RigVMVariableNode() {variable=headLookingUp}
   Subtract = &RigVMFunction_MathDoubleSub(A=1, B=RigVMVariableNode() {variable=Crouching}.Value)
   n11 = &RigVMFunction_MathDoubleRemap(
    SourceMaximum=0.5
    SourceMinimum=0
    TargetMaximum=1
    TargetMinimum=0
    Value=&RigVMFunction_MathDoubleMul(A=&RigVMFunction_MathDoubleMul(B=@n9.Result, A=@n10.Value).Result, B=@Subtract.Result).Result
   )
   &RigUnit_ModifyTransforms(
    ExecutePin=()
    ItemToModify="((Item=(Type=Null,Name=\"dyn_hair_2_r_ctrl_null\"),Transform=(Rotation=(X=0,Y=0,Z=0,W=1),Translation=(X=3,Y=-5,Z=0),Scale3D=(X=1,Y=1,Z=1))),(Item=(Type=Null,Name=\"dyn_hair_3_r_ctrl_null\"),Transform=(Rotation=(X=0,Y=0,Z=0,W=1),Translation=(X=0,Y=-3,Z=0),Scale ...(truncated)"
    WeightMinimum=0
    Weight=@n11.Result
   )
   > &RigUnit_ModifyTransforms(
    ItemToModify="((Item=(Type=Null,Name=\"dyn_hair_2_l_ctrl_null\"),Transform=(Rotation=(X=0,Y=0,Z=0,W=1),Translation=(X=0,Y=-6,Z=0),Scale3D=(X=1,Y=1,Z=1))),(Item=(Type=Null,Name=\"dyn_hair_3_l_ctrl_null\"),Transform=(Rotation=(X=0,Y=0,Z=0,W=1),Translation=(X=0,Y=-3,Z=0),Scale ...(truncated)"
    WeightMinimum=0
    Weight=@n11.Result
   )
   Backup = RigVMCollapseNode()
  ]}
 CollapseNode_ContainedGraph {nodes [
   n1 = RigVMFunctionEntryNode()
   RigVMFunction_Sequence = RigVMAggregateNode(ExecuteContext=@n1.ExecutePin)
   &RigUnit_ParentConstraint(
    AdvancedSettings=(RotationOrderForFilter=XZY)
    Child=(Type=Null,Name="dyn_hair_1_l_ctrl_null")
    Parents=((Item=(Type=Bone,Name="head"),Weight=1.0))
    Weight=1
    ExecutePin=@RigVMFunction_Sequence.A
   )
   > &RigUnit_ParentConstraint(
    AdvancedSettings=(RotationOrderForFilter=XZY)
    Child=(Type=Null,Name="dyn_hair_2_l_ctrl_null")
    Parents="((Item=(Type=Bone,Name=\"head\"),Weight=0.0033),(Item=(Type=Bone,Name=\"spine_03\"),Weight=0.0348),(Item=(Type=Bone,Name=\"clavicle_l\"),Weight=0.0784),(Item=(Type=Bone,Name=\"neck_02\"),Weight=0.0873),(Item=(Type=Bone,Name=\"neck_01\"),Weight=0.113),(Item=(Type=Bon ...(truncated)"
    Weight=1
   )
   > &RigUnit_ParentConstraint(
    AdvancedSettings=(RotationOrderForFilter=XZY)
    Child=(Type=Null,Name="dyn_hair_3_l_ctrl_null")
    Parents="((Item=(Type=Bone,Name=\"clavicle_r\"),Weight=0.005),(Item=(Type=Bone,Name=\"spine_04\"),Weight=0.0127),(Item=(Type=Bone,Name=\"upperarm_twist_01_l\"),Weight=0.025),(Item=(Type=Bone,Name=\"upperarm_l\"),Weight=0.0369),(Item=(Type=Bone,Name=\"spine_05\"),Weight=0.272 ...(truncated)"
    Weight=1
   )
   > &RigUnit_ParentConstraint(
    AdvancedSettings=(RotationOrderForFilter=XZY)
    Child=(Type=Null,Name="dyn_hair_4_l_ctrl_null")
    Parents="((Item=(Type=Bone,Name=\"clavicle_r\"),Weight=0.0029),(Item=(Type=Bone,Name=\"upperarm_l\"),Weight=0.0273),(Item=(Type=Bone,Name=\"upperarm_twist_01_l\"),Weight=0.0532),(Item=(Type=Bone,Name=\"spine_04\"),Weight=0.0923),(Item=(Type=Bone,Name=\"spine_05\"),Weight=0.2 ...(truncated)"
    Weight=1
   )
   > &RigUnit_ParentConstraint(
    AdvancedSettings=(RotationOrderForFilter=XZY)
    Child=(Type=Null,Name="dyn_hair_5_l_ctrl_null")
    Parents="((Item=(Type=Bone,Name=\"spine_01\"),Weight=0.0028),(Item=(Type=Bone,Name=\"upperarm_l\"),Weight=0.003),(Item=(Type=Bone,Name=\"clavicle_r\"),Weight=0.0182),(Item=(Type=Bone,Name=\"spine_02\"),Weight=0.0352),(Item=(Type=Bone,Name=\"spine_03\"),Weight=0.0363),(Item=( ...(truncated)"
    Weight=1
   )
   > &RigUnit_ParentConstraint(
    AdvancedSettings=(RotationOrderForFilter=XZY)
    Child=(Type=Null,Name="dyn_hair_6_l_ctrl_null")
    Parents="((Item=(Type=Bone,Name=\"clavicle_r\"),Weight=0.0127),(Item=(Type=Bone,Name=\"upperarm_twist_01_l\"),Weight=0.0314),(Item=(Type=Bone,Name=\"spine_01\"),Weight=0.0455),(Item=(Type=Bone,Name=\"spine_03\"),Weight=0.1206),(Item=(Type=Bone,Name=\"spine_05\"),Weight=0.121 ...(truncated)"
    Weight=1
   )
   > &RigUnit_ParentConstraint(
    AdvancedSettings=(RotationOrderForFilter=XZY)
    Child=(Type=Null,Name="dyn_hair_7_l_ctrl_null")
    Parents="((Item=(Type=Bone,Name=\"clavicle_l\"),Weight=0.0149),(Item=(Type=Bone,Name=\"spine_05\"),Weight=0.0158),(Item=(Type=Bone,Name=\"spine_01\"),Weight=0.098),(Item=(Type=Bone,Name=\"pelvis\"),Weight=0.1073),(Item=(Type=Bone,Name=\"spine_03\"),Weight=0.1308),(Item=(Type ...(truncated)"
    Weight=1
   )
   > &RigUnit_ParentConstraint(
    AdvancedSettings=(InterpolationType=Average,RotationOrderForFilter=XZY)
    Child=(Type=Null,Name="dyn_hair_8_l_ctrl_null")
    Parents=((Item=(Type=Bone,Name="spine_01"),Weight=0.01),(Item=(Type=Bone,Name="spine_02"),Weight=0.03),(Item=(Type=Bone,Name="pelvis"),Weight=0.96))
    Weight=1
   )
   &RigUnit_ParentConstraint(
    AdvancedSettings=(RotationOrderForFilter=XZY)
    Child=(Type=Null,Name="dyn_hair_1_c_ctrl_null")
    Parents=((Item=(Type=Bone,Name="head"),Weight=1.0))
    Weight=1
    ExecutePin=@RigVMFunction_Sequence.B
   )
   > &RigUnit_ParentConstraint(
    AdvancedSettings=(RotationOrderForFilter=XZY)
    Child=(Type=Null,Name="dyn_hair_2_c_ctrl_null")
    Parents=((Item=(Type=Bone,Name="spine_03"),Weight=0.01),(Item=(Type=Bone,Name="head"),Weight=0.1328),(Item=(Type=Bone,Name="neck_01"),Weight=0.1685),(Item=(Type=Bone,Name="spine_05"),Weight=0.3003),(Item=(Type=Bone,Name="neck_02"),Weight=0.3884))
    Weight=1
   )
   > &RigUnit_ParentConstraint(
    AdvancedSettings=(RotationOrderForFilter=XZY)
    Child=(Type=Null,Name="dyn_hair_3_c_ctrl_null")
    Parents=((Item=(Type=Bone,Name="clavicle_r"),Weight=0.0785),(Item=(Type=Bone,Name="clavicle_l"),Weight=0.0846),(Item=(Type=Bone,Name="spine_05"),Weight=0.8369))
    Weight=1
   )
   > &RigUnit_ParentConstraint(
    AdvancedSettings=(RotationOrderForFilter=XZY)
    Child=(Type=Null,Name="dyn_hair_4_c_ctrl_null")
    Parents=((Item=(Type=Bone,Name="spine_04"),Weight=0.0489),(Item=(Type=Bone,Name="clavicle_r"),Weight=0.0961),(Item=(Type=Bone,Name="clavicle_l"),Weight=0.2595),(Item=(Type=Bone,Name="spine_05"),Weight=0.5955))
    Weight=1
   )
   > &RigUnit_ParentConstraint(
    AdvancedSettings=(RotationOrderForFilter=XZY)
    Child=(Type=Null,Name="dyn_hair_5_c_ctrl_null")
    Parents="((Item=(Type=Bone,Name=\"spine_01\"),Weight=0.0129),(Item=(Type=Bone,Name=\"spine_02\"),Weight=0.0432),(Item=(Type=Bone,Name=\"spine_03\"),Weight=0.0466),(Item=(Type=Bone,Name=\"clavicle_r\"),Weight=0.1191),(Item=(Type=Bone,Name=\"clavicle_l\"),Weight=0.1208),(Item= ...(truncated)"
    Weight=1
   )
   > &RigUnit_ParentConstraint(
    AdvancedSettings=(RotationOrderForFilter=XZY)
    Child=(Type=Null,Name="dyn_hair_6_c_ctrl_null")
    Parents="((Item=(Type=Bone,Name=\"pelvis\"),Weight=0.0121),(Item=(Type=Bone,Name=\"spine_01\"),Weight=0.0532),(Item=(Type=Bone,Name=\"clavicle_r\"),Weight=0.0604),(Item=(Type=Bone,Name=\"clavicle_l\"),Weight=0.0627),(Item=(Type=Bone,Name=\"spine_03\"),Weight=0.1376),(Item=(T ...(truncated)"
    Weight=1
   )
   > &RigUnit_ParentConstraint(
    AdvancedSettings=(RotationOrderForFilter=XZY)
    Child=(Type=Null,Name="dyn_hair_7_c_ctrl_null")
    Parents="((Item=(Type=Bone,Name=\"clavicle_r\"),Weight=0.0016),(Item=(Type=Bone,Name=\"clavicle_l\"),Weight=0.0016),(Item=(Type=Bone,Name=\"spine_05\"),Weight=0.0197),(Item=(Type=Bone,Name=\"spine_01\"),Weight=0.1058),(Item=(Type=Bone,Name=\"pelvis\"),Weight=0.1237),(Item=(T ...(truncated)"
    Weight=1
   )
   > &RigUnit_ParentConstraint(
    AdvancedSettings=(InterpolationType=Average,RotationOrderForFilter=XZY)
    Child=(Type=Null,Name="dyn_hair_8_c_ctrl_null")
    Parents=((Item=(Type=Bone,Name="spine_01"),Weight=0.01),(Item=(Type=Bone,Name="spine_02"),Weight=0.03),(Item=(Type=Bone,Name="pelvis"),Weight=0.96))
    Weight=1
   )
   &RigUnit_ParentConstraint(
    AdvancedSettings=(RotationOrderForFilter=XZY)
    Child=(Type=Null,Name="dyn_hair_1_r_ctrl_null")
    Parents=((Item=(Type=Bone,Name="head"),Weight=1.0))
    Weight=1
    ExecutePin=@RigVMFunction_Sequence.C
   )
   > &RigUnit_ParentConstraint(
    AdvancedSettings=(RotationOrderForFilter=XZY)
    Child=(Type=Null,Name="dyn_hair_2_r_ctrl_null")
    Parents=((Item=(Type=Bone,Name="head"),Weight=0.0012),(Item=(Type=Bone,Name="neck_02"),Weight=0.0931),(Item=(Type=Bone,Name="clavicle_r"),Weight=0.1522),(Item=(Type=Bone,Name="neck_01"),Weight=0.1684),(Item=(Type=Bone,Name="spine_05"),Weight=0.5852))
    Weight=1
   )
   > &RigUnit_ParentConstraint(
    AdvancedSettings=(RotationOrderForFilter=XZY)
    Child=(Type=Null,Name="dyn_hair_3_r_ctrl_null")
    Parents="((Item=(Type=Bone,Name=\"clavicle_l\"),Weight=0.0033),(Item=(Type=Bone,Name=\"spine_04\"),Weight=0.013),(Item=(Type=Bone,Name=\"upperarm_twist_01_r\"),Weight=0.0238),(Item=(Type=Bone,Name=\"upperarm_r\"),Weight=0.0341),(Item=(Type=Bone,Name=\"spine_05\"),Weight=0.27 ...(truncated)"
    Weight=1
   )
   > &RigUnit_ParentConstraint(
    AdvancedSettings=(RotationOrderForFilter=XZY)
    Child=(Type=Null,Name="dyn_hair_4_r_ctrl_null")
    Parents="((Item=(Type=Bone,Name=\"clavicle_l\"),Weight=0.0011),(Item=(Type=Bone,Name=\"upperarm_r\"),Weight=0.026),(Item=(Type=Bone,Name=\"upperarm_twist_01_r\"),Weight=0.0519),(Item=(Type=Bone,Name=\"spine_04\"),Weight=0.0933),(Item=(Type=Bone,Name=\"spine_05\"),Weight=0.25 ...(truncated)"
    Weight=1
   )
   > &RigUnit_ParentConstraint(
    AdvancedSettings=(RotationOrderForFilter=XZY)
    Child=(Type=Null,Name="dyn_hair_5_r_ctrl_null")
    Parents="((Item=(Type=Bone,Name=\"spine_01\"),Weight=0.0025),(Item=(Type=Bone,Name=\"upperarm_r\"),Weight=0.0041),(Item=(Type=Bone,Name=\"clavicle_l\"),Weight=0.0181),(Item=(Type=Bone,Name=\"spine_02\"),Weight=0.0337),(Item=(Type=Bone,Name=\"spine_03\"),Weight=0.0351),(Item= ...(truncated)"
    Weight=1
   )
   > &RigUnit_ParentConstraint(
    AdvancedSettings=(RotationOrderForFilter=XZY)
    Child=(Type=Null,Name="dyn_hair_6_r_ctrl_null")
    Parents="((Item=(Type=Bone,Name=\"clavicle_l\"),Weight=0.0132),(Item=(Type=Bone,Name=\"upperarm_twist_01_r\"),Weight=0.0329),(Item=(Type=Bone,Name=\"spine_01\"),Weight=0.0441),(Item=(Type=Bone,Name=\"spine_03\"),Weight=0.1179),(Item=(Type=Bone,Name=\"spine_05\"),Weight=0.125 ...(truncated)"
    Weight=1
   )
   > &RigUnit_ParentConstraint(
    AdvancedSettings=(RotationOrderForFilter=XZY)
    Child=(Type=Null,Name="dyn_hair_7_r_ctrl_null")
    Parents="((Item=(Type=Bone,Name=\"clavicle_r\"),Weight=0.0149),(Item=(Type=Bone,Name=\"spine_05\"),Weight=0.0158),(Item=(Type=Bone,Name=\"spine_01\"),Weight=0.098),(Item=(Type=Bone,Name=\"pelvis\"),Weight=0.1072),(Item=(Type=Bone,Name=\"spine_03\"),Weight=0.1309),(Item=(Type ...(truncated)"
    Weight=1
   )
   > &RigUnit_ParentConstraint(
    AdvancedSettings=(InterpolationType=Average,RotationOrderForFilter=XZY)
    Child=(Type=Null,Name="dyn_hair_8_r_ctrl_null")
    Parents=((Item=(Type=Bone,Name="spine_01"),Weight=0.01),(Item=(Type=Bone,Name="spine_02"),Weight=0.03),(Item=(Type=Bone,Name="pelvis"),Weight=0.96))
    Weight=1
   )
   RigVMFunctionReturnNode(ExecutePin=@RigVMFunction_Sequence.E)
   &RigUnit_ParentConstraint(
    AdvancedSettings=(InterpolationType=Average,RotationOrderForFilter=XZY)
    Child=(Type=Null,Name="dyn_frontHair_A_01_l_ctrl_null")
    Filter=(TranslationFilter=(bX=true,bY=true,bZ=true),RotationFilter=(bX=true,bY=true,bZ=true),ScaleFilter=(bX=false,bY=false,bZ=false))
    Parents=((Item=(Type=Bone,Name="base_frontHair_A_01_l"),Weight=1))
    Weight=1
    bMaintainOffset=true
    ExecutePin=@RigVMFunction_Sequence.D
   )
   > &RigUnit_ParentConstraint(
    AdvancedSettings=(InterpolationType=Average,RotationOrderForFilter=XZY)
    Child=(Type=Null,Name="dyn_frontHair_A_02_l_ctrl_null")
    Filter=(TranslationFilter=(bX=true,bY=true,bZ=true),RotationFilter=(bX=true,bY=true,bZ=true),ScaleFilter=(bX=false,bY=false,bZ=false))
    Parents=((Item=(Type=Bone,Name="base_frontHair_A_01_l"),Weight=1))
    Weight=1
    bMaintainOffset=true
   )
   > &RigUnit_ParentConstraint(
    AdvancedSettings=(RotationOrderForFilter=XZY)
    Child=(Type=Null,Name="dyn_frontHair_A_03_l_ctrl_null")
    Parents="((Item=(Type=Bone,Name=\"upperarm_twist_01_l\"),Weight=0.0536),(Item=(Type=Bone,Name=\"spine_05\"),Weight=0.0601),(Item=(Type=Bone,Name=\"spine_05\"),Weight=0.1179),(Item=(Type=Bone,Name=\"spine_04\"),Weight=0.1579),(Item=(Type=Bone,Name=\"clavicle_l\"),Weight=0.253 ...(truncated)"
    Weight=1
   )
   > &RigUnit_ParentConstraint(
    AdvancedSettings=(RotationOrderForFilter=XZY)
    Child=(Type=Null,Name="dyn_frontHair_A_04_l_ctrl_null")
    Parents="((Item=(Type=Bone,Name=\"spine_05\"),Weight=0.0516),(Item=(Type=Bone,Name=\"upperarm_twist_01_l\"),Weight=0.0585),(Item=(Type=Bone,Name=\"clavicle_l\"),Weight=0.1232),(Item=(Type=Bone,Name=\"spine_05\"),Weight=0.1653),(Item=(Type=Bone,Name=\"spine_04\"),Weight=0.283 ...(truncated)"
    Weight=1
   )
   > &RigUnit_ParentConstraint(
    AdvancedSettings=(RotationOrderForFilter=XZY)
    Child=(Type=Null,Name="dyn_frontHair_A_05_l_ctrl_null")
    Parents="((Item=(Type=Bone,Name=\"spine_03\"),Weight=0.0209),(Item=(Type=Bone,Name=\"spine_02\"),Weight=0.0215),(Item=(Type=Bone,Name=\"clavicle_l\"),Weight=0.0262),(Item=(Type=Bone,Name=\"spine_05\"),Weight=0.0546),(Item=(Type=Bone,Name=\"spine_05\"),Weight=0.358),(Item=(Ty ...(truncated)"
    Weight=1
   )
   > &RigUnit_ParentConstraint(
    AdvancedSettings=(RotationOrderForFilter=XZY)
    Child=(Type=Null,Name="dyn_frontHair_A_06_l_ctrl_null")
    Parents="((Item=(Type=Bone,Name=\"spine_05\"),Weight=0.002),(Item=(Type=Bone,Name=\"spine_01\"),Weight=0.0151),(Item=(Type=Bone,Name=\"spine_05\"),Weight=0.0522),(Item=(Type=Bone,Name=\"spine_03\"),Weight=0.0756),(Item=(Type=Bone,Name=\"spine_02\"),Weight=0.0923),(Item=(Type ...(truncated)"
    Weight=1
   )
   > &RigUnit_ParentConstraint(
    AdvancedSettings=(RotationOrderForFilter=XZY)
    Child=(Type=Null,Name="dyn_frontHair_A_07_l_ctrl_null")
    Parents=((Item=(Type=Bone,Name="spine_01"),Weight=0.0378),(Item=(Type=Bone,Name="spine_03"),Weight=0.0762),(Item=(Type=Bone,Name="spine_04"),Weight=0.1771),(Item=(Type=Bone,Name="pelvis"),Weight=0.2506),(Item=(Type=Bone,Name="spine_02"),Weight=0.4582))
    Weight=1
   )
   &RigUnit_ParentConstraint(
    AdvancedSettings=(RotationOrderForFilter=XZY)
    Child=(Type=Null,Name="dyn_frontHair_A_01_l_ctrl_null")
    ExecutePin=()
    Parents="((Item=(Type=Bone,Name=\"head\"),Weight=0.0082),(Item=(Type=Bone,Name=\"clavicle_l\"),Weight=0.0221),(Item=(Type=Bone,Name=\"head\"),Weight=0.0318),(Item=(Type=Bone,Name=\"spine_05\"),Weight=0.2521),(Item=(Type=Bone,Name=\"neck_02\"),Weight=0.3402),(Item=(Type=Bone, ...(truncated)"
    Weight=1
   )
   > &RigUnit_ParentConstraint(
    AdvancedSettings=(RotationOrderForFilter=XZY)
    Child=(Type=Null,Name="dyn_frontHair_A_02_l_ctrl_null")
    Parents="((Item=(Type=Bone,Name=\"head\"),Weight=0.0259),(Item=(Type=Bone,Name=\"head\"),Weight=0.0306),(Item=(Type=Bone,Name=\"head\"),Weight=0.0394),(Item=(Type=Bone,Name=\"spine_05\"),Weight=0.2264),(Item=(Type=Bone,Name=\"neck_02\"),Weight=0.3077),(Item=(Type=Bone,Name=\" ...(truncated)"
    Weight=1
   )
  ]}
 CollapseNode_ContainedGraph {nodes [
   RigVMFunctionEntryNode()
   joints_twist_c = RigVMDispatchNode(Value=((Type=Bone,Name="dyn_hair_2_c"),(Type=Bone,Name="dyn_hair_3_c"),(Type=Bone,Name="dyn_hair_4_c"),(Type=Bone,Name="dyn_hair_5_c"),(Type=Bone,Name="dyn_hair_6_c"),(Type=Bone,Name="dyn_hair_7_c"))) {template="DISPATCH_RigVMDispatch_Constant(io Value)"}
   > TwistBones = RigVMFunctionReferenceNode(
    Axis to Twist=(bX=True,bY=true,bZ=true)
    Debug=false
    End=(Type=Bone,Name="neck_01")
    Start=(Type=Bone,Name="spine_02")
    Bones to Twist=@joints_twist_c.Value
   )
   joints_twist_l = RigVMDispatchNode(Value=((Type=Bone,Name="dyn_hair_2_l"),(Type=Bone,Name="dyn_hair_3_l"),(Type=Bone,Name="dyn_hair_4_l"),(Type=Bone,Name="dyn_hair_5_l"),(Type=Bone,Name="dyn_hair_6_l"),(Type=Bone,Name="dyn_hair_7_l"))) {template="DISPATCH_RigVMDispatch_Constant(io Value)"}
   > TwistBones_1 = RigVMFunctionReferenceNode(
    Axis to Twist=(bX=True,bY=true,bZ=true)
    Debug=false
    End=(Type=Bone,Name="neck_01")
    Start=(Type=Bone,Name="spine_02")
    Bones to Twist=@joints_twist_l.Value
   )
   joints_twist_r = RigVMDispatchNode(Value=((Type=Bone,Name="dyn_hair_2_r"),(Type=Bone,Name="dyn_hair_3_r"),(Type=Bone,Name="dyn_hair_4_r"),(Type=Bone,Name="dyn_hair_5_r"),(Type=Bone,Name="dyn_hair_6_r"),(Type=Bone,Name="dyn_hair_7_r"))) {template="DISPATCH_RigVMDispatch_Constant(io Value)"}
   > TwistBones_2 = RigVMFunctionReferenceNode(
    Axis to Twist=(bX=True,bY=true,bZ=true)
    Debug=false
    End=(Type=Bone,Name="neck_01")
    Start=(Type=Bone,Name="spine_02")
    Bones to Twist=@joints_twist_r.Value
   )
   ItemArray = RigVMDispatchNode(Value=((Type=Bone,Name="dyn_frontHair_A_02_l"),(Type=Bone,Name="dyn_frontHair_A_03_l"),(Type=Bone,Name="dyn_frontHair_A_04_l"),(Type=Bone,Name="dyn_frontHair_A_05_l"),(Type=Bone,Name="dyn_frontHair_A_06_l"),(Type=Bone,Name="dyn_frontHair_A_07_l"))) {template="DISPATCH_RigVMDispatch_Constant(io Value)"}
   > TwistBones_3 = RigVMFunctionReferenceNode(
    Axis to Twist=(bX=True,bY=true,bZ=true)
    Debug=false
    End=(Type=Bone,Name="neck_01")
    Start=(Type=Null,Name="spine_02_twist_null")
    Bones to Twist=@ItemArray.Value
   )
   > RigVMFunctionReturnNode()
  ]}
 CollapseNode_ContainedGraph {nodes [
   RigVMFunction_Sequence_2 = RigVMAggregateNode(ExecuteContext=())
   GetAlphaFromQuatDelta_18 = RigVMFunctionReferenceNode(
    Rotation Order=ZYX
    child=(Type=Bone,Name="clavicle_l")
    debugRotation=false
    onlyTwist=false
    outputMinMax=(X=0,Y=1)
    parent=(Type=Null,Name="clavicle_l_null")
    premult=1
    rotationMinMax=(X=0,Y=-40)
    triggerAxis=Z
    twistAxis=(X=1,Y=0,Z=0)
    ExecuteContext=@RigVMFunction_Sequence_2.D
   )
   > &RigUnit_ModifyTransforms(
    ItemToModify="((Item=(Type=Null,Name=\"dyn_hair_3_l_ctrl_null\"),Transform=(Rotation=(X=0,Y=0,Z=0,W=-1),Translation=(X=0,Y=-1,Z=3),Scale3D=(X=1,Y=1,Z=1))),(Item=(Type=Null,Name=\"dyn_hair_4_l_ctrl_null\"),Transform=(Rotation=(X=0,Y=0,Z=0,W=-1),Translation=(X=0,Y=1,Z=4),Scal ...(truncated)"
    WeightMinimum=0
    Weight=@GetAlphaFromQuatDelta_18.alpha
   )
   > GetAlphaFromQuatDelta_19 = RigVMFunctionReferenceNode(
    Rotation Order=ZYX
    child=(Type=Bone,Name="clavicle_r")
    debugRotation=false
    onlyTwist=false
    outputMinMax=(X=0,Y=1)
    parent=(Type=Null,Name="clavicle_r_null")
    premult=1
    rotationMinMax=(X=0,Y=-40)
    triggerAxis=Z
    twistAxis=(X=1,Y=0,Z=0)
   )
   > &RigUnit_ModifyTransforms(
    ItemToModify="((Item=(Type=Null,Name=\"dyn_hair_3_r_ctrl_null\"),Transform=(Translation=(X=-2,Y=-1,Z=3),Scale3D=(X=1,Y=1,Z=1))),(Item=(Type=Null,Name=\"dyn_hair_4_r_ctrl_null\"),Transform=(Rotation=(X=0,Y=0,Z=0,W=1),Translation=(X=0,Y=0,Z=5),Scale3D=(X=1,Y=1,Z=1))),(Item=(T ...(truncated)"
    WeightMinimum=0
    Weight=@GetAlphaFromQuatDelta_19.alpha
   )
   > GetAlphaFromQuatDelta_20 = RigVMFunctionReferenceNode(
    Rotation Order=ZYX
    child=(Type=Bone,Name="clavicle_l")
    debugRotation=false
    onlyTwist=false
    outputMinMax=(X=0,Y=1)
    parent=(Type=Null,Name="clavicle_l_null")
    premult=1
    rotationMinMax=(X=0,Y=60)
    triggerAxis=Z
    twistAxis=(X=1,Y=0,Z=0)
   )
   > &RigUnit_ModifyTransforms(
    ItemToModify="((Item=(Type=Null,Name=\"dyn_hair_3_l_ctrl_null\"),Transform=(Rotation=(X=0,Y=0,Z=0,W=-1),Translation=(X=0,Y=-3,Z=-1),Scale3D=(X=1,Y=1,Z=1))),(Item=(Type=Null,Name=\"dyn_hair_4_l_ctrl_null\"),Transform=(Rotation=(X=0,Y=0,Z=0,W=-1),Translation=(X=0,Y=-5,Z=-1),S ...(truncated)"
    WeightMinimum=0
    Weight=@GetAlphaFromQuatDelta_20.alpha
   )
   > GetAlphaFromQuatDelta_21 = RigVMFunctionReferenceNode(
    Rotation Order=ZYX
    child=(Type=Bone,Name="clavicle_r")
    debugRotation=false
    onlyTwist=false
    outputMinMax=(X=0,Y=1)
    parent=(Type=Null,Name="clavicle_r_null")
    premult=1
    rotationMinMax=(X=0,Y=60)
    triggerAxis=Z
    twistAxis=(X=1,Y=0,Z=0)
   )
   > &RigUnit_ModifyTransforms(
    ItemToModify="((Item=(Type=Null,Name=\"dyn_hair_3_r_ctrl_null\"),Transform=(Rotation=(X=0,Y=0,Z=0,W=-1),Translation=(X=0,Y=3,Z=0),Scale3D=(X=1,Y=1,Z=1))),(Item=(Type=Null,Name=\"dyn_hair_4_r_ctrl_null\"),Transform=(Rotation=(X=0,Y=0,Z=0,W=-1),Translation=(X=0,Y=5,Z=-1),Scal ...(truncated)"
    WeightMinimum=0
    Weight=@GetAlphaFromQuatDelta_21.alpha
   )
   > GetAlphaFromQuatDelta_22 = RigVMFunctionReferenceNode(
    Rotation Order=ZYX
    child=(Type=Bone,Name="clavicle_l")
    debugRotation=false
    onlyTwist=false
    outputMinMax=(X=0,Y=1)
    parent=(Type=Null,Name="clavicle_l_null")
    premult=1
    rotationMinMax=(X=0,Y=-60)
    triggerAxis=Y
    twistAxis=(X=1,Y=0,Z=0)
   )
   > &RigUnit_ModifyTransforms(
    ItemToModify="((Item=(Type=Null,Name=\"dyn_hair_3_l_ctrl_null\"),Transform=(Rotation=(X=0,Y=0,Z=0,W=-1),Translation=(X=0,Y=4,Z=3),Scale3D=(X=1,Y=1,Z=1))),(Item=(Type=Null,Name=\"dyn_hair_4_l_ctrl_null\"),Transform=(Rotation=(X=0,Y=0,Z=0,W=-1),Translation=(X=0,Y=-1,Z=0),Scal ...(truncated)"
    WeightMinimum=0
    Weight=@GetAlphaFromQuatDelta_22.alpha
   )
   > GetAlphaFromQuatDelta_23 = RigVMFunctionReferenceNode(
    Rotation Order=ZYX
    child=(Type=Bone,Name="clavicle_r")
    debugRotation=false
    onlyTwist=false
    outputMinMax=(X=0,Y=1)
    parent=(Type=Null,Name="clavicle_r_null")
    premult=1
    rotationMinMax=(X=0,Y=-60)
    triggerAxis=Y
    twistAxis=(X=1,Y=0,Z=0)
   )
   > &RigUnit_ModifyTransforms(
    ItemToModify="((Item=(Type=Null,Name=\"dyn_hair_3_r_ctrl_null\"),Transform=(Rotation=(X=0,Y=0,Z=0,W=-1),Translation=(X=0,Y=-4,Z=3),Scale3D=(X=1,Y=1,Z=1))),(Item=(Type=Null,Name=\"dyn_hair_4_r_ctrl_null\"),Transform=(Rotation=(X=0,Y=0,Z=0,W=-1),Translation=(X=0,Y=1,Z=0),Scal ...(truncated)"
    WeightMinimum=0
    Weight=@GetAlphaFromQuatDelta_23.alpha
   )
   GetAlphaFromQuatDelta_13 = RigVMFunctionReferenceNode(
    Rotation Order=ZYX
    child=(Type=Bone,Name="head")
    debugRotation=false
    onlyTwist=True
    outputMinMax=(X=0,Y=1)
    parent=(Type=Bone,Name="neck_01")
    premult=1
    rotationMinMax=(X=0,Y=-60)
    triggerAxis=X
    twistAxis=(X=1,Y=0,Z=0)
    ExecuteContext=@RigVMFunction_Sequence_2.B
   )
   > &RigUnit_ModifyTransforms(
    ItemToModify="((Item=(Type=Null,Name=\"dyn_hair_2_r_ctrl_null\"),Transform=(Rotation=(X=0,Y=0,Z=0,W=1),Translation=(X=0,Y=0,Z=3),Scale3D=(X=1,Y=1,Z=1))),(Item=(Type=Null,Name=\"dyn_hair_3_r_ctrl_null\"),Transform=(Rotation=(X=0,Y=0,Z=0,W=-1),Translation=(X=0,Y=0,Z=3),Scale3 ...(truncated)"
    WeightMinimum=0
    Weight=@GetAlphaFromQuatDelta_13.alpha
   )
   > GetAlphaFromQuatDelta_14 = RigVMFunctionReferenceNode(
    Rotation Order=ZYX
    child=(Type=Bone,Name="head")
    onlyTwist=True
    outputMinMax=(X=0,Y=1)
    parent=(Type=Bone,Name="neck_01")
    premult=1
    rotationMinMax=(X=0,Y=60)
    triggerAxis=X
    twistAxis=(X=1,Y=0,Z=0)
   )
   > &RigUnit_ModifyTransforms(
    ItemToModify="((Item=(Type=Null,Name=\"dyn_hair_1_r_ctrl_null\"),Transform=(Rotation=(X=0,Y=0,Z=0,W=1),Translation=(X=-0.2341,Y=-3.6072,Z=-0.9853),Scale3D=(X=1,Y=1,Z=1))),(Item=(Type=Null,Name=\"dyn_hair_2_r_ctrl_null\"),Transform=(Rotation=(X=0,Y=0,Z=0,W=-1),Translation=(X ...(truncated)"
    WeightMinimum=0
    Weight=@GetAlphaFromQuatDelta_14.alpha
   )
   > GetAlphaFromQuatDelta_15 = RigVMFunctionReferenceNode(
    Rotation Order=ZYX
    child=(Type=Bone,Name="head")
    debugRotation=false
    onlyTwist=False
    outputMinMax=(X=0,Y=1)
    parent=(Type=Null,Name="head_null")
    premult=1
    rotationMinMax=(X=-60,Y=-80)
    triggerAxis=Z
    twistAxis=(X=1,Y=0,Z=0)
   )
   > RigVMVariableNode(Value=@GetAlphaFromQuatDelta_15.alpha) {variable=head rz n60n80 alpha}
   > &RigUnit_ModifyTransforms(
    ItemToModify="((Item=(Type=Null,Name=\"dyn_hair_3_r_ctrl_null\"),Transform=(Rotation=(X=0,Y=0,Z=0,W=-1),Translation=(X=2.2209,Y=-1,Z=-1),Scale3D=(X=1,Y=1,Z=1))),(Item=(Type=Null,Name=\"dyn_hair_2_l_ctrl_null\"),Transform=(Rotation=(X=0,Y=0,Z=0,W=-1),Translation=(X=7.6214,Y= ...(truncated)"
    WeightMinimum=0
    Weight=@GetAlphaFromQuatDelta_15.alpha
   )
   > &RigUnit_ModifyTransforms(
    ItemToModify="((Item=(Type=Null,Name=\"dyn_hair_1_r_ctrl_null\"),Transform=(Rotation=(X=0,Y=0,Z=0,W=1),Translation=(X=-5,Y=0,Z=0),Scale3D=(X=1,Y=1,Z=1))),(Item=(Type=Null,Name=\"dyn_hair_2_r_ctrl_null\"),Transform=(Rotation=(X=0,Y=0,Z=0,W=1),Translation=(X=5,Y=0,Z=0),Scale3 ...(truncated)"
    WeightMinimum=0
    Weight=@GetAlphaFromQuatDelta_15.alpha
   )
   > &RigUnit_ModifyTransforms(
    ItemToModify="((Item=(Type=Null,Name=\"dyn_hair_1_r_ctrl_null\"),Transform=(Rotation=(X=0,Y=0,Z=0,W=1),Translation=(X=0,Y=-3,Z=1),Scale3D=(X=1,Y=1,Z=1))),(Item=(Type=Null,Name=\"dyn_hair_2_r_ctrl_null\"),Transform=(Rotation=(X=0,Y=0,Z=0,W=1),Translation=(X=3,Y=-1,Z=0),Scale ...(truncated)"
    WeightMinimum=0
    Weight=&RigVMFunction_MathDoubleRemap(SourceMaximum=0.5, SourceMinimum=0, TargetMaximum=0, TargetMinimum=1, Value=RigVMVariableNode() {variable=headTiltingSideways}.Value).Result
   )
   > &RigUnit_ModifyTransforms(
    ItemToModify="((Item=(Type=Null,Name=\"dyn_hair_1_r_ctrl_null\"),Transform=(Rotation=(X=0,Y=0,Z=0,W=1),Translation=(X=-6,Y=2,Z=0),Scale3D=(X=1,Y=1,Z=1))),(Item=(Type=Null,Name=\"dyn_hair_3_r_ctrl_null\"),Transform=(Rotation=(X=0,Y=0,Z=0,W=1),Translation=(X=0,Y=-1,Z=1),Scale ...(truncated)"
    WeightMinimum=0
    Weight=RigVMVariableNode() {variable=HeadLookingUp_Crouching}.Value
   )
   n1 = &RigVMFunction_MathDoubleRemap(SourceMaximum=0.5, SourceMinimum=0, TargetMaximum=0, TargetMinimum=1, Value=RigVMVariableNode() {variable=headTiltingSideways}.Value)
   n2 = RigVMVariableNode() {variable=headLookingUp}
   n3 = RigVMVariableNode() {variable=Crouching}
   n4 = &RigVMFunction_MathDoubleRemap(
    SourceMaximum=0.5
    SourceMinimum=0
    TargetMaximum=1
    TargetMinimum=0
    Value=&RigVMFunction_MathDoubleMul(
     A=&RigVMFunction_MathDoubleMul(B=@n1.Result, A=@n2.Value).Result
     B=&RigVMFunction_MathDoubleSub(A=1, B=@n3.Value).Result
    ).Result
   )
   > &RigUnit_ModifyTransforms(
    ItemToModify="((Item=(Type=Null,Name=\"dyn_hair_2_r_ctrl_null\"),Transform=(Rotation=(X=0,Y=0,Z=0,W=1),Translation=(X=3,Y=-5,Z=0),Scale3D=(X=1,Y=1,Z=1))),(Item=(Type=Null,Name=\"dyn_hair_3_r_ctrl_null\"),Transform=(Rotation=(X=0,Y=0,Z=0,W=1),Translation=(X=0,Y=-3,Z=0),Scale ...(truncated)"
    WeightMinimum=0
    Weight=@n4.Result
   )
   > &RigUnit_ModifyTransforms(
    ItemToModify="((Item=(Type=Null,Name=\"dyn_hair_2_l_ctrl_null\"),Transform=(Rotation=(X=0,Y=0,Z=0,W=1),Translation=(X=0,Y=-6,Z=0),Scale3D=(X=1,Y=1,Z=1))),(Item=(Type=Null,Name=\"dyn_hair_3_l_ctrl_null\"),Transform=(Rotation=(X=0,Y=0,Z=0,W=1),Translation=(X=0,Y=-3,Z=0),Scale ...(truncated)"
    WeightMinimum=0
    Weight=@n4.Result
   )
   n5 = &RigVMFunction_MathDoubleMul(
    B=&RigVMFunction_MathDoubleRemap(SourceMaximum=0.5, SourceMinimum=0, TargetMaximum=0, TargetMinimum=1, Value=RigVMVariableNode() {variable=headTiltingSideways}.Value).Result
    A=RigVMVariableNode() {variable=headLookingUp}.Value
   )
   n6 = RigVMVariableNode() {variable=Crouching}
   n7 = RigVMVariableNode() {variable=HeadADSBigGun}
   n8 = &RigVMFunction_MathDoubleRemap(
    SourceMaximum=0.5
    SourceMinimum=0
    TargetMaximum=1
    TargetMinimum=0
    Value=&RigVMFunction_MathDoubleMul(
     A=&RigVMFunction_MathDoubleMul(A=@n5.Result, B=@n6.Value).Result
     B=&RigVMFunction_MathDoubleSub(A=1, B=@n7.Value).Result
    ).Result
   )
   > &RigUnit_ModifyTransforms(
    ItemToModify="((Item=(Type=Null,Name=\"dyn_hair_2_r_ctrl_null\"),Transform=(Rotation=(X=0,Y=0,Z=0,W=1),Translation=(X=3,Y=-5,Z=0),Scale3D=(X=1,Y=1,Z=1))),(Item=(Type=Null,Name=\"dyn_hair_3_r_ctrl_null\"),Transform=(Rotation=(X=0,Y=0,Z=0,W=1),Translation=(X=0,Y=-3,Z=0),Scale ...(truncated)"
    WeightMinimum=0
    Weight=@n8.Result
   )
   > &RigUnit_ModifyTransforms(
    ItemToModify="((Item=(Type=Null,Name=\"dyn_hair_2_l_ctrl_null\"),Transform=(Rotation=(X=0,Y=0,Z=0,W=1),Translation=(X=0,Y=-6,Z=0),Scale3D=(X=1,Y=1,Z=1))),(Item=(Type=Null,Name=\"dyn_hair_3_l_ctrl_null\"),Transform=(Rotation=(X=0,Y=0,Z=0,W=1),Translation=(X=0,Y=-3,Z=0),Scale ...(truncated)"
    WeightMinimum=0
    Weight=@n8.Result
   )
   GetAlphaFromQuatDelta_16 = RigVMFunctionReferenceNode(
    Rotation Order=ZYX
    child=(Type=Bone,Name="spine_05")
    onlyTwist=True
    outputMinMax=(X=0,Y=1)
    parent=(Type=Bone,Name="spine_03")
    premult=1
    rotationMinMax=(X=0,Y=35)
    triggerAxis=X
    twistAxis=(X=1,Y=0,Z=0)
    ExecuteContext=@RigVMFunction_Sequence_2.C
   )
   > &RigUnit_ModifyTransforms(
    ItemToModify="((Item=(Type=Null,Name=\"dyn_hair_4_r_ctrl_null\"),Transform=(Rotation=(X=0,Y=0,Z=0,W=1),Translation=(X=0,Y=-2.112,Z=3.5606),Scale3D=(X=1,Y=1,Z=1))),(Item=(Type=Null,Name=\"dyn_hair_5_r_ctrl_null\"),Transform=(Rotation=(X=0,Y=0,Z=0,W=-1),Translation=(X=0,Y=0,Z ...(truncated)"
    WeightMinimum=0
    Weight=@GetAlphaFromQuatDelta_16.alpha
   )
   > &RigUnit_ModifyTransforms(
    ItemToModify="((Item=(Type=Null,Name=\"dyn_hair_3_r_ctrl_null\"),Transform=(Rotation=(X=0,Y=0,Z=0,W=1),Translation=(X=0,Y=0,Z=3),Scale3D=(X=1,Y=1,Z=1))),(Item=(Type=Null,Name=\"dyn_hair_2_r_ctrl_null\"),Transform=(Rotation=(X=0,Y=0,Z=0,W=1),Translation=(X=0,Y=0,Z=2.0512),Sc ...(truncated)"
    WeightMinimum=0
    Weight=@GetAlphaFromQuatDelta_16.alpha
   )
   > GetAlphaFromQuatDelta_17 = RigVMFunctionReferenceNode(
    Rotation Order=ZYX
    child=(Type=Bone,Name="spine_05")
    onlyTwist=True
    outputMinMax=(X=0,Y=1)
    parent=(Type=Bone,Name="spine_03")
    premult=0
    rotationMinMax=(X=0,Y=-35)
    triggerAxis=X
    twistAxis=(X=1,Y=0,Z=0)
   )
   > &RigUnit_ModifyTransforms(
    ItemToModify="((Item=(Type=Null,Name=\"dyn_hair_4_l_ctrl_null\"),Transform=(Rotation=(X=0,Y=0,Z=0,W=1),Translation=(X=0,Y=-2.112,Z=3.5606),Scale3D=(X=1,Y=1,Z=1))),(Item=(Type=Null,Name=\"dyn_hair_5_l_ctrl_null\"),Transform=(Rotation=(X=0,Y=0,Z=0,W=-1),Translation=(X=0,Y=0,Z ...(truncated)"
    WeightMinimum=0
    Weight=@GetAlphaFromQuatDelta_17.alpha
   )
   n9 = &RigUnit_SphericalPoseReader(
    ActiveRegionScaleFactors=(PositiveWidth=0.5,NegativeWidth=0.5,PositiveHeight=0.5,NegativeHeight=0.5)
    ActiveRegionSize=0.1
    Debug=(bDrawDebug=false,bDraw2D=False,bDrawLocalAxes=False,DebugScale=25,DebugSegments=20,DebugThickness=0.25)
    DriverAxis=(X=-1,Y=0,Z=0)
    DriverItem=(Type=Bone,Name="upperarm_r")
    FalloffRegionScaleFactors=(PositiveWidth=1,NegativeWidth=1,PositiveHeight=1,NegativeHeight=1)
    FalloffSize=0.2
    FlipHeightScaling=False
    FlipWidthScaling=False
    OptionalParentItem=(Type=Bone,Name="clavicle_r")
    RotationOffset=(X=120,Y=0,Z=-140)
    ExecutePin=@RigVMFunction_Sequence_2.A
   )
   n10 = &RigVMFunction_AlphaInterp(InterpSpeedDecreasing=8, InterpSpeedIncreasing=14, Scale=1, bClampResult=False, Value=@n9.OutputParam)
   > RigVMVariableNode(Value=@n10.Result) {variable=isPistolASDUp}
   > &RigUnit_ModifyTransforms(
    ItemToModify="((Item=(Type=Null,Name=\"dyn_hair_2_r_ctrl_null\"),Transform=(Rotation=(X=0,Y=0,Z=0,W=1),Translation=(X=0,Y=0,Z=0),Scale3D=(X=1,Y=1,Z=1))),(Item=(Type=Null,Name=\"dyn_hair_2_r_ctrl_null\"),Transform=(Rotation=(X=0,Y=0,Z=0,W=1),Translation=(X=0,Y=0,Z=0),Scale3D ...(truncated)"
    WeightMinimum=0
    Weight=@n10.Result
   )
   > n11 = &RigUnit_SphericalPoseReader(
    ActiveRegionScaleFactors=(PositiveWidth=1,NegativeWidth=1,PositiveHeight=1,NegativeHeight=1)
    ActiveRegionSize=0.05
    Debug=(bDrawDebug=false,bDraw2D=False,bDrawLocalAxes=False,DebugScale=25,DebugSegments=20,DebugThickness=0.25)
    DriverAxis=(X=1,Y=0,Z=0)
    DriverItem=(Type=Bone,Name="head")
    FalloffRegionScaleFactors=(PositiveWidth=1,NegativeWidth=1,PositiveHeight=1,NegativeHeight=1)
    FalloffSize=0.2
    FlipHeightScaling=False
    FlipWidthScaling=False
    OptionalParentItem=(Type=Bone,Name="spine_05")
    RotationOffset=(X=-48,Y=0,Z=35)
   )
   > RigVMVariableNode(Value=@n11.OutputParam) {variable=HeadADSBigGun}
   > &RigUnit_ModifyTransforms(
    ItemToModify="((Item=(Type=Null,Name=\"dyn_hair_1_r_ctrl_null\"),Transform=(Rotation=(X=0,Y=0,Z=0,W=1),Translation=(X=-4,Y=1,Z=0),Scale3D=(X=1,Y=1,Z=1))),(Item=(Type=Null,Name=\"dyn_hair_2_r_ctrl_null\"),Transform=(Rotation=(X=0,Y=0,Z=0,W=1),Translation=(X=0,Y=-3,Z=2),Scale ...(truncated)"
    WeightMinimum=0
    Weight=@n11.OutputParam
   )
   RigVMFunctionEntryNode()
   RigVMFunctionReturnNode()
  ]}
 CollapseNode_ContainedGraph {nodes [
   n1 = RigVMFunctionEntryNode()
   RigVMFunction_Sequence = &RigVMFunction_Sequence(ExecuteContext=@n1.ExecuteContext)
   n2 = &RigVMFunction_AlphaInterp(InterpSpeedDecreasing=2, InterpSpeedIncreasing=10, Scale=1, bClampResult=False, Value=RigVMVariableNode() {variable=Melee Anim}.Value)
   AlterTransform = RigVMFunctionReferenceNode(
    inheritsTransform=false
    transforms="((Item=(Type=Bone,Name=\"dyn_hair_3_r\"),Transform=(Rotation=(X=0,Y=0,Z=0,W=1),Translation=(X=0,Y=2,Z=0),Scale3D=(X=1,Y=1,Z=1))),(Item=(Type=Bone,Name=\"dyn_hair_1_r\"),Transform=(Rotation=(X=0,Y=0,Z=-0.1736,W=0.9848),Translation=(X=0,Y=1,Z=0),Scale3D=(X=1,Y=1 ...(truncated)"
    weights=@n2.Result
    ExecuteContext=@RigVMFunction_Sequence.A
   )
   > GetAlphaFromQuatDelta_13 = RigVMFunctionReferenceNode(
    Rotation Order=ZYX
    child=(Type=Bone,Name="head")
    debugRotation=false
    onlyTwist=true
    outputMinMax=(X=0,Y=1)
    parent=(Type=Null,Name="head_null")
    premult=1
    rotationMinMax=(X=-20,Y=-80)
    triggerAxis=X
    twistAxis=(X=1,Y=0,Z=0)
   )
   > AlterTransform_3 = RigVMFunctionReferenceNode(
    inheritsTransform=false
    transforms="((Item=(Type=Bone,Name=\"dyn_hair_2_r\"),Transform=(Rotation=(X=0,Y=0,Z=0,W=1),Translation=(X=0,Y=0,Z=-1),Scale3D=(X=1,Y=1,Z=1))),(Item=(Type=Bone,Name=\"dyn_hair_1_r\"),Transform=(Rotation=(X=0,Y=0,Z=-0.4226,W=0.9063),Translation=(X=0,Y=2,Z=0),Scale3D=(X=1,Y= ...(truncated)"
    weights=@GetAlphaFromQuatDelta_13.alpha
   )
   > &RigUnit_ModifyTransforms(
    ItemToModify="((Item=(Type=Bone,Name=\"dyn_frontHair_A_03_l\"),Transform=(Rotation=(X=-0.2588,Y=0,Z=0,W=0.9659),Translation=(X=0,Y=0,Z=0),Scale3D=(X=1,Y=1,Z=1))),(Item=(Type=Bone,Name=\"dyn_frontHair_A_05_l\"),Transform=(Rotation=(X=-0.173,Y=0.0858,Z=0.0151,W=0.9811),Transl ...(truncated)"
    WeightMinimum=0
    Weight=RigVMVariableNode() {variable=ClavFrontDelta_l}.Value
   )
   > &RigUnit_ModifyTransforms(
    ItemToModify=((Item=(Type=Bone,Name="dyn_frontHair_A_04_l"),Transform=(Rotation=(X=-0.2588,Y=0,Z=0,W=0.9659),Translation=(X=0,Y=0,Z=0),Scale3D=(X=1,Y=1,Z=1))))
    WeightMinimum=0
    Weight=RigVMVariableNode() {variable=TorsoTwistRight}.Value
   )
   n3 = &RigVMFunction_AlphaInterp(
    InterpSpeedDecreasing=10
    InterpSpeedIncreasing=10
    Scale=1
    bClampResult=False
    Value=&RigVMFunction_MathBoolToFloat(Value=RigVMVariableNode() {variable=pub isFloating}.Value).Result
   )
   > AlterTransform_1 = RigVMFunctionReferenceNode(
    inheritsTransform=false
    transforms="((Item=(Type=Bone,Name=\"dyn_hair_1_l\"),Transform=(Rotation=(X=0,Y=0,Z=0,W=1),Translation=(X=0,Y=0,Z=3),Scale3D=(X=1,Y=1,Z=1))),(Item=(Type=Bone,Name=\"dyn_hair_1_r\"),Transform=(Rotation=(X=0,Y=0,Z=0,W=1),Translation=(X=0,Y=0,Z=3),Scale3D=(X=1,Y=1,Z=1))),(It ...(truncated)"
    weights=@n3.Result
   )
   n4 = &RigVMFunction_AlphaInterp(
    InterpSpeedDecreasing=10
    InterpSpeedIncreasing=10
    Scale=1
    bClampResult=False
    Value=&RigVMFunction_MathBoolToFloat(Value=RigVMVariableNode() {variable=idleFE}.Value).Result
   )
   > AlterTransform_2 = RigVMFunctionReferenceNode(
    inheritsTransform=false
    transforms="((Item=(Type=Bone,Name=\"dyn_hair_2_l\"),Transform=(Rotation=(X=0,Y=0,Z=0,W=1),Translation=(X=0,Y=0,Z=0.5),Scale3D=(X=1,Y=1,Z=1))),(Item=(Type=Bone,Name=\"dyn_hair_3_l\"),Transform=(Rotation=(X=0,Y=0,Z=0,W=1),Translation=(X=0,Y=0,Z=1),Scale3D=(X=1,Y=1,Z=1))),( ...(truncated)"
    weights=@n4.Result
   )
   > AlterTransform_4 = RigVMFunctionReferenceNode(
    inheritsTransform=true
    transforms="((Item=(Type=Bone,Name=\"dyn_hair_3_r\"),Transform=(Rotation=(X=0,Y=0,Z=0.0436,W=0.999),Translation=(X=0,Y=0,Z=0.5),Scale3D=(X=1,Y=1,Z=1))),(Item=(Type=Bone,Name=\"dyn_hair_3_l\"),Transform=(Rotation=(X=0,Y=0,Z=-0.0436,W=0.999),Translation=(X=0,Y=0,Z=0),Scale3 ...(truncated)"
    weights=@n4.Result
   )
   RigVMFunctionReturnNode(ExecuteContext=@RigVMFunction_Sequence.B)
  ]}
 CollapseNode_ContainedGraph {nodes [
   RigVMFunctionEntryNode()
   > RigVMFunction_Sequence = RigVMAggregateNode()
   GetAlphaFromQuatDelta = RigVMFunctionReferenceNode(
    Rotation Order=ZYX
    child=(Type=Bone,Name="head")
    debugRotation=false
    outputMinMax=(X=0,Y=1)
    parent=(Type=Bone,Name="neck_01")
    premult=1
    rotationMinMax=(X=-50,Y=-65)
    triggerAxis=Z
    twistAxis=(X=1,Y=0,Z=0)
    ExecuteContext=@RigVMFunction_Sequence.A
   )
   > RigVMVariableNode(Value=@GetAlphaFromQuatDelta.alpha) {variable=HeadLookingUp_Crouching}
   > GetAlphaFromQuatDelta_1 = RigVMFunctionReferenceNode(
    Rotation Order=ZYX
    child=(Type=Bone,Name="head")
    outputMinMax=(X=0,Y=1)
    parent=(Type=Bone,Name="neck_01")
    premult=1
    rotationMinMax=(X=-20,Y=20)
    triggerAxis=Y
    twistAxis=(X=1,Y=0,Z=0)
   )
   > RigVMVariableNode(Value=@GetAlphaFromQuatDelta_1.alpha) {variable=headTiltingSideways}
   > GetAlphaFromQuatDelta_2 = RigVMFunctionReferenceNode(
    Rotation Order=ZYX
    child=(Type=Bone,Name="head")
    debugRotation=false
    outputMinMax=(X=0,Y=1)
    parent=(Type=Bone,Name="neck_01")
    premult=1
    rotationMinMax=(X=-20,Y=-43)
    triggerAxis=Z
    twistAxis=(X=1,Y=0,Z=0)
   )
   n1 = &RigVMFunction_AnimEvalRichCurve(
    Curve="(EditorCurveData=(Keys=((InterpMode=RCIM_Cubic,TangentMode=RCTM_User,ArriveTangent=0.6623,LeaveTangent=0.6623),(InterpMode=RCIM_Cubic,TangentMode=RCTM_User,Time=0.5,Value=0.4,ArriveTangent=0.8898,LeaveTangent=0.8898),(InterpMode=RCIM_Cubic,TangentMode=RCTM ...(truncated)"
    SourceMinimum=0
    TargetMinimum=0
    Value=@GetAlphaFromQuatDelta_2.alpha
   )
   > RigVMVariableNode(Value=@n1.Result) {variable=headLookingUp}
   RigVMVariableNode(
    Value=&RigVMFunction_MathDoubleRemap(
     SourceMaximum=65
     SourceMinimum=44
     TargetMaximum=0
     TargetMinimum=1
     Value=&RigUnit_GetRelativeTransformForItem(Child=(Type=Bone,Name="pelvis"), Parent=(Type=Bone,Name="root"), bChildInitial=False, bParentInitial=False).RelativeTransform.Translation.Z
    ).Result
    ExecuteContext=@RigVMFunction_Sequence.B
   ) {variable=CrouchingInPlace}
   n2 = &RigVMFunction_AnimEvalRichCurve(
    Curve="(EditorCurveData=(Keys=((InterpMode=RCIM_Cubic,TangentMode=RCTM_Break),(InterpMode=RCIM_Cubic,TangentMode=RCTM_Break,Time=0.45,ArriveTangent=0.0163,LeaveTangent=1.4085),(Time=0.6,Value=1),(Time=0.74,Value=1),(InterpMode=RCIM_Cubic,TangentMode=RCTM_Break,Ti ...(truncated)"
    SourceMinimum=0
    TargetMinimum=0
    Value=&RigVMFunction_MathDoubleRemap(
     SourceMaximum=100
     SourceMinimum=0
     TargetMaximum=1
     TargetMinimum=0
     Value=&RigUnit_GetRelativeTransformForItem(Child=(Type=Bone,Name="pelvis"), Parent=(Type=Bone,Name="root"), bChildInitial=False, bParentInitial=False).RelativeTransform.Translation.Z
    ).Result
   )
   > RigVMVariableNode(Value=@n2.Result) {variable=Crouching}
   RigVMFunctionReturnNode(ExecuteContext=@RigVMFunction_Sequence.C)
  ]}
 CollapseNode_ContainedGraph {nodes [
   RigVMFunctionEntryNode()
   > RigVMFunction_Sequence = RigVMAggregateNode()
   ItemArray = RigVMDispatchNode(Value="((Type=Bone,Name=\"dyn_hair_1_l\"),(Type=Bone,Name=\"dyn_hair_2_l\"),(Type=Bone,Name=\"dyn_hair_3_l\"),(Type=Bone,Name=\"dyn_hair_4_l\"),(Type=Bone,Name=\"dyn_hair_5_l\"),(Type=Bone,Name=\"dyn_hair_6_l\"),(Type=Bone,Name=\"dyn_hair_7_l\"),(Type=Bone,Name=\"dyn_hair_8_l\") ...(truncated)") {template="DISPATCH_RigVMDispatch_Constant(io Value)"}
   For_Each = RigVMDispatchNode(Array=@ItemArray.Value, ExecuteContext=@RigVMFunction_Sequence.A) {template="DISPATCH_RigVMDispatch_ArrayIterator(in Array,out Element,out Index,out Count,out Ratio)"}
   Concat = /Script/RigVM.RigVMFunction_NameConcat(B=_ctrl_null, A=@For_Each.Element.Name) {is=struct RigVMUnitNode, template="Concat::Execute(in A,in B,out Result)"}
   RigUnit_GetTransform = &RigUnit_GetTransform(Item=(Type=Bone,Name="None"), bInitial=true, Item.Name=@For_Each.Element.Name)
   > &RigUnit_HierarchyAddNull(Parent=(Type=Bone,Name="None"), Space=GlobalSpace, Name=@Concat.Result, Transform=@RigUnit_GetTransform.Transform)
   For_Each_1 = RigVMDispatchNode(ExecuteContext=@For_Each.Completed, Array=@ItemArray.Value) {template="DISPATCH_RigVMDispatch_ArrayIterator(in Array,out Element,out Index,out Count,out Ratio)"}
   n1 = &RigVMFunction_NameConcat(B=_ctrl, A=@For_Each_1.Element.Name) {template="Concat::Execute(in A,in B,out Result)"}
   Contains = /Script/RigVM.RigVMFunction_Contains(Search=_r_, Name=@n1.Result) {is=struct RigVMUnitNode, template="Contains::Execute(in Name,in Search,out Result)"}
   n2 = &RigVMFunction_Contains(Search=_c_, Name=@n1.Result) {template="Contains::Execute(in Name,in Search,out Result)"}
   > SpawnControl = /Script/ControlRig.RigUnit_HierarchyAddControlTransform(
    InitialValue=(Rotation=(X=0,Y=0,Z=0,W=1),Translation=(X=0,Y=0,Z=0),Scale3D=(X=1,Y=1,Z=1))
    OffsetSpace=GlobalSpace
    Parent=(Type=Null,Name="None")
    Settings="(InitialSpace=LocalSpace,bUsePreferredRotationOrder=False,PreferredRotationOrder=YZX,Limits=(LimitTranslationX=(bMinimum=False,bMaximum=False),LimitTranslationY=(bMinimum=False,bMaximum=False),LimitTranslationZ=(bMinimum=False,bMaximum=False),LimitPitch=(b ...(truncated)"
    Name=@n1.Result
    Parent.Name=&RigVMFunction_NameConcat(B=_ctrl_null, A=@For_Each_1.Element.Name) {template="Concat::Execute(in A,in B,out Result)"}.Result
    Settings.Shape.Color=RigVMDispatchNode(
     True=(R=1,G=0,B=0,A=1)
     Condition=@Contains.Result
     False=RigVMDispatchNode(False=(R=0,G=0,B=1,A=1), True=(R=1,G=1,B=0,A=1), Condition=@n2.Result) {template="DISPATCH_RigVMDispatch_If(in Condition,in True,in False,out Result)"}.Result
    ) {template="DISPATCH_RigVMDispatch_If(in Condition,in True,in False,out Result)"}.Result
    OffsetTransform=&RigUnit_GetTransform(Item=(Type=Bone,Name="None"), bInitial=true, Item.Name=@For_Each_1.Element.Name).Transform
   ) {
    is=struct RigVMUnitNode
    template="SpawnControl::Execute(in InitialValue,in Settings,in OffsetTransform,in OffsetSpace,in Parent,in Name,out Item)"
   }
   Item Array = RigVMDispatchNode(Value="((Type=Bone,Name=\"thigh_l\"),(Type=Bone,Name=\"calf_l\"),(Type=Bone,Name=\"foot_l\"),(Type=Bone,Name=\"thigh_r\"),(Type=Bone,Name=\"calf_r\"),(Type=Bone,Name=\"foot_r\"),(Type=Bone,Name=\"clavicle_l\"),(Type=Bone,Name=\"upperarm_l\"),(Type=Bone,Name=\"hand_l\"),(Type=Bone, ...(truncated)") {template="DISPATCH_RigVMDispatch_Constant(io Value)"}
   addNullObject = RigVMFunctionReferenceNode(
    Options_Null="(pelvis_1_#1=False,spine_3_#2=true,clavicle_5_#3=false,arm_7_#4=false,leg_9_#5=False,head_11_46DC15EF465CD ...(truncated)"
    Custom_Joints=@"Item Array".Value
    ExecuteContext=@RigVMFunction_Sequence.B
   )
   > &RigUnit_HierarchyAddNull(
    Name=neck_02_null
    Parent=(Type=Bone,Name="spine_05")
    Space=LocalSpace
    Transform=&RigUnit_GetRelativeTransformForItem(Child=(Type=Bone,Name="neck_02"), Parent=(Type=Bone,Name="spine_05"), bChildInitial=true, bParentInitial=true).RelativeTransform
   )
   > RigVMFunctionReturnNode()
   ItemArray_2 = RigVMDispatchNode(Value="((Type=Bone,Name=\"dyn_frontHair_A_01_l\"),(Type=Bone,Name=\"dyn_frontHair_A_02_l\"),(Type=Bone,Name=\"dyn_frontHair_A_03_l\"),(Type=Bone,Name=\"dyn_frontHair_A_04_l\"),(Type=Bone,Name=\"dyn_frontHair_A_05_l\"),(Type=Bone,Name=\"dyn_frontHair_A_06_l\"),(Type=Bone,Name ...(truncated)") {template="DISPATCH_RigVMDispatch_Constant(io Value)"}
   For_Each_2 = RigVMDispatchNode(Array=@ItemArray_2.Value, ExecuteContext=@RigVMFunction_Sequence.C) {template="DISPATCH_RigVMDispatch_ArrayIterator(in Array,out Element,out Index,out Count,out Ratio)"}
   > &RigUnit_HierarchyAddNull(
    Parent=(Type=Bone,Name="None")
    Space=GlobalSpace
    Name=&RigVMFunction_NameConcat(B=_ctrl_null, A=@For_Each_2.Element.Name) {template="Concat::Execute(in A,in B,out Result)"}.Result
    Transform=&RigUnit_GetTransform(Item=(Type=Bone,Name="None"), bInitial=true, Item.Name=@For_Each_2.Element.Name).Transform
   )
   For_Each_3 = RigVMDispatchNode(ExecuteContext=@For_Each_2.Completed, Array=@ItemArray_2.Value) {template="DISPATCH_RigVMDispatch_ArrayIterator(in Array,out Element,out Index,out Count,out Ratio)"}
   n3 = &RigVMFunction_NameConcat(B=_ctrl, A=@For_Each_3.Element.Name) {template="Concat::Execute(in A,in B,out Result)"}
   > &RigUnit_HierarchyAddControlTransform(
    InitialValue=(Rotation=(X=0,Y=0,Z=0,W=1),Translation=(X=0,Y=0,Z=0),Scale3D=(X=1,Y=1,Z=1))
    OffsetSpace=GlobalSpace
    Parent=(Type=Null,Name="None")
    Settings="(InitialSpace=LocalSpace,bUsePreferredRotationOrder=False,PreferredRotationOrder=YZX,Limits=(LimitTranslationX=(bMinimum=False,bMaximum=False),LimitTranslationY=(bMinimum=False,bMaximum=False),LimitTranslationZ=(bMinimum=False,bMaximum=False),LimitPitch=(b ...(truncated)"
    Name=@n3.Result
    Parent.Name=&RigVMFunction_NameConcat(B=_ctrl_null, A=@For_Each_3.Element.Name) {template="Concat::Execute(in A,in B,out Result)"}.Result
    OffsetTransform=&RigUnit_GetTransform(Item=(Type=Bone,Name="None"), bInitial=true, Item.Name=@For_Each_3.Element.Name).Transform
   ) {template="SpawnControl::Execute(in InitialValue,in Settings,in OffsetTransform,in OffsetSpace,in Parent,in Name,out Item)"}
   RigVMDispatchNode(
    True=(R=1,G=0,B=0,A=1)
    Condition=&RigVMFunction_Contains(Search=_r_, Name=@n3.Result) {template="Contains::Execute(in Name,in Search,out Result)"}.Result
    False=RigVMDispatchNode(
     False=(R=0,G=0,B=1,A=1)
     True=(R=1,G=1,B=0,A=1)
     Condition=&RigVMFunction_Contains(Search=_c_, Name=@n3.Result) {template="Contains::Execute(in Name,in Search,out Result)"}.Result
    ) {template="DISPATCH_RigVMDispatch_If(in Condition,in True,in False,out Result)"}.Result
   ) {template="DISPATCH_RigVMDispatch_If(in Condition,in True,in False,out Result)"}
  ]}
 New Function_ContainedGraph {nodes [
   RigVMFunctionEntryNode()
   RigVMFunction_MathBoolToInteger = &RigVMFunction_MathBoolToInteger(Value=RigVMVariableNode() {variable=onlyTwist}.Value)
   RigVMFunction_MathQuaternionSwingTwist = &RigVMFunction_MathQuaternionSwingTwist(
    Input=&RigUnit_GetRelativeTransformForItem(
     bChildInitial=False
     bParentInitial=False
     Child=RigVMVariableNode() {variable=child}.Value
     Parent=RigVMVariableNode() {variable=parent}.Value
    ).RelativeTransform.Rotation
    TwistAxis=RigVMVariableNode() {variable=twistAxis}.Value
   )
   n1 = RigVMDispatchNode(
    Values=((),())
    Index=@RigVMFunction_MathBoolToInteger.Result
    Values.0=@RigVMFunction_MathQuaternionSwingTwist.Swing
    Values.1=@RigVMFunction_MathQuaternionSwingTwist.Twist
   ) {template="DISPATCH_RigVMDispatch_SelectInt32(in Index,in Values,out Result)"}
   > RigVMVariableNode(Value=@n1.Result) {variable=entry_outputQuat}
   RigVMFunction_MathQuaternionToEuler = &RigVMFunction_MathQuaternionToEuler(Value=@n1.Result, RotationOrder=RigVMVariableNode() {variable=Rotation Order}.Value)
   > DISPATCH_RigVMDispatch_Print = RigVMDispatchNode(
    ScreenColor=(R=1,G=1,B=1,A=1)
    ScreenDuration=1.5
    Value=@RigVMFunction_MathQuaternionToEuler.Result
    Enabled=RigVMVariableNode() {variable=debugRotation}.Value
   ) {template="DISPATCH_RigVMDispatch_Print(in Prefix,in Value,in Enabled,in ScreenDuration,in ScreenColor)"}
   Cast = RigVMDispatchNode(Value=RigVMVariableNode() {variable=triggerAxis}.Value) {template="DISPATCH_RigVMDispatch_CastEnumToInt(in Value,out Result)"}
   DISPATCH_RigVMDispatch_SelectInt32 = RigVMDispatchNode(
    Values=(0,0,0)
    Index=@Cast.Result
    Values.0=@RigVMFunction_MathQuaternionToEuler.Result.X
    Values.1=@RigVMFunction_MathQuaternionToEuler.Result.Y
    Values.2=@RigVMFunction_MathQuaternionToEuler.Result.Z
   ) {template="DISPATCH_RigVMDispatch_SelectInt32(in Index,in Values,out Result)"}
   n2 = RigVMVariableNode() {variable=rotationMinMax}
   n3 = RigVMVariableNode() {variable=outputMinMax}
   Remap = &RigVMFunction_MathDoubleRemap(
    bClamp=False
    Value=@DISPATCH_RigVMDispatch_SelectInt32.Result
    SourceMinimum=@n2.Value.X
    SourceMaximum=@n2.Value.Y
    TargetMinimum=@n3.Value.X
    TargetMaximum=@n3.Value.Y
   )
   Absolute = &RigVMFunction_MathDoubleAbs(Value=@Remap.Result)
   n4 = RigVMVariableNode() {variable=absoluteValue}
   Multiply = &RigVMFunction_MathDoubleMul(
    A=&RigVMFunction_MathDoubleClamp(Value=RigVMDispatchNode(True=@Absolute.Result, False=@Remap.Result, Condition=@n4.Value) {template="DISPATCH_RigVMDispatch_If(in Condition,in True,in False,out Result)"}.Result).Result
    B=RigVMVariableNode() {variable=premult}.Value
   )
   > RigVMFunctionReturnNode(alpha=@Multiply.Result, rawQuat=RigVMVariableNode() {variable=entry_outputQuat}.Value)
  ]}
 New Function_ContainedGraph {nodes [
   n1 = RigVMFunctionEntryNode()
   > For_Each = RigVMDispatchNode(Array=@n1.transforms) {template="DISPATCH_RigVMDispatch_ArrayIterator(in Array,out Element,out Index,out Count,out Ratio)"}
   > &RigUnit_OffsetTransformForItem(Item=@For_Each.Element.Item, OffsetTransform=@For_Each.Element.Transform, Weight=@n1.weights, bPropagateToChildren=@n1.inheritsTransform)
   RigVMFunctionReturnNode(ExecuteContext=@For_Each.Completed)
  ]}
 RigVMFunctionLibrary {nodes [
   AlterTransform = RigVMCollapseNode(ExecuteContext=(), inheritsTransform=false, transforms=(), weights=1)
   GetAlphaFromQuatDelta = RigVMCollapseNode(
    ExecuteContext=()
    Rotation Order=ZYX
    child=(Type=Bone,Name="None")
    outputMinMax=(X=0,Y=1)
    parent=(Type=Bone,Name="None")
    premult=1
    rotationMinMax=(X=0,Y=0)
    twistAxis=(X=1,Y=0,Z=0)
   )
  ]}
 RigVMFunction_Sequence_1_ContainedGraph {nodes [
   n1 = RigVMFunctionEntryNode()
   RigVMFunction_Sequence = &RigVMFunction_Sequence(ExecuteContext=@n1.ExecuteContext)
   n2 = &RigVMFunction_Sequence(ExecuteContext=@RigVMFunction_Sequence.B)
   n3 = &RigVMFunction_Sequence(ExecuteContext=@n2.B)
   RigVMFunctionReturnNode(A=@RigVMFunction_Sequence.A, B=@n2.A, C=@n3.A, D=@n3.B)
  ]}
 RigVMFunction_Sequence_1_ContainedGraph {nodes [
   n1 = RigVMFunctionEntryNode()
   RigVMFunction_Sequence = &RigVMFunction_Sequence(ExecuteContext=@n1.ExecuteContext)
   n2 = &RigVMFunction_Sequence(ExecuteContext=@RigVMFunction_Sequence.B)
   n3 = &RigVMFunction_Sequence(ExecuteContext=@n2.B)
   n4 = &RigVMFunction_Sequence(ExecuteContext=@n3.B)
   RigVMFunctionReturnNode(A=@RigVMFunction_Sequence.A, B=@n2.A, C=@n3.A, D=@n4.A, E=@n4.B)
  ]}
 RigVMFunction_Sequence_1_ContainedGraph {nodes [
   n1 = RigVMFunctionEntryNode()
   RigVMFunction_Sequence = &RigVMFunction_Sequence(ExecuteContext=@n1.ExecuteContext)
   n2 = &RigVMFunction_Sequence(ExecuteContext=@RigVMFunction_Sequence.B)
   RigVMFunctionReturnNode(A=@RigVMFunction_Sequence.A, B=@n2.A, C=@n2.B)
  ]}
 RigVMFunction_Sequence_1_ContainedGraph {nodes [
   n1 = RigVMFunctionEntryNode()
   RigVMFunction_Sequence = &RigVMFunction_Sequence(ExecuteContext=@n1.ExecuteContext)
   n2 = &RigVMFunction_Sequence(ExecuteContext=@RigVMFunction_Sequence.B)
   RigVMFunctionReturnNode(A=@RigVMFunction_Sequence.A, B=@n2.A, C=@n2.B)
  ]}
 RigVMFunction_Sequence_2_ContainedGraph {nodes [
   n1 = RigVMFunctionEntryNode()
   n2 = &RigVMFunction_Sequence(ExecuteContext=@n1.ExecuteContext)
   n3 = &RigVMFunction_Sequence(ExecuteContext=@n2.B)
   n4 = &RigVMFunction_Sequence(ExecuteContext=@n3.B)
   n5 = &RigVMFunction_Sequence(ExecuteContext=@n4.B)
   RigVMFunctionReturnNode(A=@n2.A, B=@n3.A, C=@n4.A, D=@n5.A, E=@n5.B)
  ]}
 RigVMFunction_Sequence_2_ContainedGraph {nodes [
   n1 = RigVMFunctionEntryNode()
   n2 = &RigVMFunction_Sequence(ExecuteContext=@n1.ExecuteContext)
   n3 = &RigVMFunction_Sequence(ExecuteContext=@n2.B)
   n4 = &RigVMFunction_Sequence(ExecuteContext=@n3.B)
   RigVMFunctionReturnNode(A=@n2.A, B=@n3.A, C=@n4.A, D=@n4.B)
  ]}
 RigVMModel {nodes [
   RigUnit_BeginExecution = &RigUnit_BeginExecution()
   DefiningVariables = RigVMCollapseNode(ExecuteContext=@RigUnit_BeginExecution.ExecutePin)
   > Hair = RigVMCollapseNode()
  ]}
 RigVMModel Construction Graph {nodes [
   &RigUnit_PrepareForExecution()
   SpawnNulls_st = RigVMCollapseNode(ExecuteContext=())
  ]}
]
hierarchy [
 root {loc=<0 0 0>, rot=<0 0 0 1>}
 pelvis {parent=root, loc=<0 -0.5301 98.0571>, rot=<0.0127 0.707 -0.0127 -0.707>}
 spine_01 {parent=pelvis, loc=<6.1208 -0.9328 0>, rot=<0 0 0.034 -0.9994>}
 spine_02 {parent=spine_01, loc=<6.5997 -0.2095 0>, rot=<0 0 0.0238 -0.9997>}
 spine_03 {parent=spine_02, loc=<6.4914 -0.1237 0>, rot=<0 0 0.0149 -0.9999>}
 spine_04 {parent=spine_03, loc=<6.9598 0.0794 0>, rot=<0 0 -0.0132 -0.9999>}
 spine_05 {parent=spine_04, loc=<6.9987 0 0>, rot=<0 0 -0.0015 -1>}
 neck_01 {parent=spine_05, loc=<11.3138 0.1529 0>, rot=<0 0 -0.2003 -0.9797>}
 neck_02 {parent=neck_01, loc=<4.5627 0 0>, rot=<0 0 0.0504 -0.9987>}
 head {parent=neck_02, loc=<5.3914 0 0>, rot=<0 0 0.115 -0.9934>}
 base_hair_1_c {parent=head, loc=<1.4522 -5.6606 0>, rot=<0.1182 -0.6972 0.6972 -0.1182>}
 dyn_hair_1_c {parent=base_hair_1_c, loc=<0 0 0>, rot=<0 0 0 1>}
 dyn_hair_2_c {parent=dyn_hair_1_c, loc=<7.3205 0 0>, rot=<0 0.0795 0 -0.9968>}
 dyn_hair_3_c {parent=dyn_hair_2_c, loc=<7.3197 0 0>, rot=<0 -0.0726 0 -0.9974>}
 dyn_hair_4_c {parent=dyn_hair_3_c, loc=<7.2884 0 0>, rot=<0 -0.1866 0 -0.9824>}
 dyn_hair_5_c {parent=dyn_hair_4_c, loc=<7.2939 0 0>, rot=<0 -0.0514 0 -0.9987>}
 dyn_hair_6_c {parent=dyn_hair_5_c, loc=<7.3213 0 0>, rot=<0 0.026 0 -0.9997>}
 dyn_hair_7_c {parent=dyn_hair_6_c, loc=<6.213 0 0>, rot=<0 0 0 1>}
 dyn_hair_8_c {parent=dyn_hair_7_c, loc=<8.0705 0 0>, rot=<0 0 0 1>}
 base_hair_1_l {parent=head, loc=<1.2526 -5.0535 -7.9336>, rot=<0.12 -0.6968 0.6946 -0.1325>}
 dyn_hair_1_l {parent=base_hair_1_l, loc=<0 0 0>, rot=<0 0 0 1>}
 dyn_hair_2_l {parent=dyn_hair_1_l, loc=<7.3205 0 0>, rot=<0 0.0754 0 -0.9972>}
 dyn_hair_3_l {parent=dyn_hair_2_l, loc=<7.3197 0 0>, rot=<0 -0.0875 0 -0.9962>}
 dyn_hair_4_l {parent=dyn_hair_3_l, loc=<7.2884 0 0>, rot=<0 -0.1688 0 -0.9857>}
 dyn_hair_5_l {parent=dyn_hair_4_l, loc=<7.0383 0 0>, rot=<0 -0.0424 0 -0.9991>}
 dyn_hair_6_l {parent=dyn_hair_5_l, loc=<7.2603 0 0.0217>, rot=<0 0.026 0 -0.9997>}
 dyn_hair_7_l {parent=dyn_hair_6_l, loc=<6.3654 0 -0.0164>, rot=<0 0 0 1>}
 dyn_hair_8_l {parent=dyn_hair_7_l, loc=<5.7238 0 0>, rot=<0 0 0 1>}
 base_hair_1_r {parent=head, loc=<1.1548 -4.3665 7.0967>, rot=<0.0471 -0.7055 0.7068 -0.0207>}
 dyn_hair_1_r {parent=base_hair_1_r, loc=<0 0 0>, rot=<0 0 0 1>}
 dyn_hair_2_r {parent=dyn_hair_1_r, loc=<7.3206 0 -0.0001>, rot=<0 0.3236 0 -0.9462>}
 dyn_hair_3_r {parent=dyn_hair_2_r, loc=<8.0121 0 0.0002>, rot=<0 -0.1751 0 -0.9845>}
 dyn_hair_4_r {parent=dyn_hair_3_r, loc=<7.8328 0 -0.0001>, rot=<0 -0.1825 0 -0.9832>}
 dyn_hair_5_r {parent=dyn_hair_4_r, loc=<7.4091 0 0.0001>, rot=<-0.003 -0.0959 -0.0308 -0.9949>}
 dyn_hair_6_r {parent=dyn_hair_5_r, loc=<7.5012 0 -0.0001>, rot=<0 0.0268 0 -0.9996>}
 dyn_hair_7_r {parent=dyn_hair_6_r, loc=<6.2129 0 0>, rot=<0 0 0 1>}
 dyn_hair_8_r {parent=dyn_hair_7_r, loc=<5.7238 0 0>, rot=<0 0 0 1>}
 base_hair2Curve_r {parent=dyn_hair_2_r, loc=<-0.3449 0.8565 -1.7151>, rot=<0 -0.2002 0 -0.9797>}
 dyn_hair2Curve_r {parent=base_hair2Curve_r, loc=<0 0 0>, rot=<0 0 0 1>}
 base_frontHair_A_01_l {parent=head, loc=<1.1231 1.8249 -8.4662>, rot=<0.0137 -0.9964 -0.077 0.0341>}
 dyn_frontHair_A_01_l {parent=base_frontHair_A_01_l, loc=<0 0 0>, rot=<0 0 0 1>}
 dyn_frontHair_A_02_l {parent=dyn_frontHair_A_01_l, loc=<7.3205 0 0>, rot=<0.2471 0.1526 -0.1571 -0.9439>}
 dyn_frontHair_A_03_l {parent=dyn_frontHair_A_02_l, loc=<7.3197 0 0>, rot=<0.1303 0.0495 0.038 -0.9895>}
 dyn_frontHair_A_04_l {parent=dyn_frontHair_A_03_l, loc=<7.2884 0 0>, rot=<0 -0.0484 -0.0484 -0.9977>}
 dyn_frontHair_A_05_l {parent=dyn_frontHair_A_04_l, loc=<7.2939 0 0>, rot=<0.1353 -0.1506 0.0424 -0.9784>}
 dyn_frontHair_A_06_l {parent=dyn_frontHair_A_05_l, loc=<7.3213 0 0>, rot=<0 -0.0335 -0.0581 -0.9977>}
 dyn_frontHair_A_07_l {parent=dyn_frontHair_A_06_l, loc=<7.3211 0 0>, rot=<0 0 0 1>}
 base_frontHair_B_01_l {parent=head, loc=<1.0943 5.3134 -8.4662>, rot=<-0.0344 -0.9959 -0.0752 0.0378>}
 dyn_frontHair_B_01_l {parent=base_frontHair_B_01_l, loc=<0 0 0>, rot=<0 0 0 1>}
 dyn_frontHair_B_02_l {parent=dyn_frontHair_B_01_l, loc=<7.3205 0 0>, rot=<0.1114 0.1141 -0.0598 -0.9854>}
 dyn_frontHair_B_03_l {parent=dyn_frontHair_B_02_l, loc=<6.7257 0 0>, rot=<0.1347 0.0872 -0.0481 -0.9859>}
 dyn_frontHair_B_04_l {parent=dyn_frontHair_B_03_l, loc=<6.6944 0 0>, rot=<0 -0.0592 -0.0342 -0.9977>}
 dyn_frontHair_B_05_l {parent=dyn_frontHair_B_04_l, loc=<7.2939 0 0>, rot=<-0.0033 -0.1387 0.021 -0.9901>}
 dyn_frontHair_B_06_l {parent=dyn_frontHair_B_05_l, loc=<7.3213 0 0>, rot=<-0.0074 -0.2844 -0.0833 -0.9551>}
 dyn_frontHair_B_07_l {parent=dyn_frontHair_B_06_l, loc=<4.8999 0 0>, rot=<0 0 0 1>}
 base_frontHair_C_01_l {parent=head, loc=<6.646 8.322 -5.5984>, rot=<0.0493 -0.9895 -0.0757 -0.1128>}
 dyn_frontHair_C_01_l {parent=base_frontHair_C_01_l, loc=<0 0 0>, rot=<0 0 0 1>}
 dyn_frontHair_C_02_l {parent=dyn_frontHair_C_01_l, loc=<5.7642 -0.0438 0.1956>, rot=<-0.003 -0.1435 -0.0945 -0.9851>}
 dyn_frontHair_C_03_l {parent=dyn_frontHair_C_02_l, loc=<7.3205 0 0>, rot=<0.1114 0.1141 -0.0598 -0.9854>}
 dyn_frontHair_C_04_l {parent=dyn_frontHair_C_03_l, loc=<6.7257 0 0>, rot=<0.1347 0.0872 -0.0481 -0.9859>}
 dyn_frontHair_C_05_l {parent=dyn_frontHair_C_04_l, loc=<3.693 0 0>, rot=<0 -0.0592 -0.0342 -0.9977>}
 base_hair_0_c {parent=head, loc=<2.118 -7.5672 0>, rot=<0.1182 -0.6972 0.6972 -0.1182>}
 hair_0_c {parent=base_hair_0_c, loc=<0 0 0>, rot=<0 0 0 1>}
 head_null {parent=spine_05, type=Null, loc=<20.6573 3.5489 0>, rot=<0 0 -0.036 -0.9994>}
 clavicle_l {parent=spine_05, loc=<5.4831 -0.1053 -3.2946>, rot=<0.0465 -0.7289 -0.009 -0.6829>}
 upperarm_l {parent=clavicle_l, loc=<9.8658 -1.2136 -0.4105>, rot=<0.0164 -0.3245 -0.0269 -0.9454>}
 upperarm_twist_01_l {parent=upperarm_l, loc=<7.2135 -0.1134 0>, rot=<0 0 0 1>}
 upperarm_twist_02_l {parent=upperarm_l, loc=<19.7998 -0.1134 0>, rot=<0 0 0 1>}
 upperarm_l_null {parent=clavicle_l, type=Null, loc=<9.8658 -1.2136 -0.4105>, rot=<0.0164 -0.3245 -0.0269 -0.9454>}
 clavicle_r {parent=spine_05, loc=<5.4831 -0.1053 3.2946>, rot=<0.7289 0.0465 -0.6829 0.009>}
 upperarm_r {parent=clavicle_r, loc=<-9.8658 1.2136 0.4105>, rot=<0.0164 -0.3245 -0.0269 -0.9454>}
 upperarm_twist_01_r {parent=upperarm_r, loc=<-7.2135 0.1134 0>, rot=<0 0 0 -1>}
 upperarm_twist_02_r {parent=upperarm_r, loc=<-19.7998 0.1134 0>, rot=<0 0 0 -1>}
 upperarm_r_null {parent=clavicle_r, type=Null, loc=<-9.8658 1.2136 0.4105>, rot=<0.0164 -0.3245 -0.0269 -0.9454>}
 clavicle_r_null {parent=spine_05, type=Null, loc=<5.4831 -0.1053 3.2946>, rot=<0.7289 0.0465 -0.6829 0.009>}
 clavicle_l_null {parent=spine_05, type=Null, loc=<5.4831 -0.1053 -3.2946>, rot=<0.0465 -0.7289 -0.009 -0.6829>}
 neck_02_null {parent=spine_05, type=Null, loc=<15.5105 1.9434 0>, rot=<0 0 0.1506 0.9886>}
 spine_05_null {parent=spine_04, type=Null, loc=<6.9987 0 0>, rot=<0 0 -0.0015 -1>}
 spine_04_null {parent=spine_03, type=Null, loc=<6.9598 0.0794 0>, rot=<0 0 -0.0132 -0.9999>}
 spine_03_null {parent=spine_02, type=Null, loc=<6.4914 -0.1237 0>, rot=<0 0 0.0149 -0.9999>}
 spine_02_twist_null {parent=spine_01, type=Null, loc=<6.5997 -0.2095 0>, rot=<0 0 0.0238 -0.9997>}
 spine_02_null {parent=spine_01, type=Null, loc=<6.5997 -0.2095 0>, rot=<0 0 0.0238 -0.9997>}
 spine_01_null {parent=pelvis, type=Null, loc=<6.1208 -0.9328 0>, rot=<0 0 0.034 -0.9994>}
 dyn_hair_1_l_ctrl_null {type=Null, loc=<7.9336 -6.751 153.245>, rot=<-0.4087 -0.4014 0.5744 -0.5848>}
 dyn_hair_1_l_ctrl {controlType=EULER_TRANSFORM, parent=dyn_hair_1_l_ctrl_null, type=Control, loc=<0 0 0>, rot=<0 0 0 1>}
 dyn_hair_2_l_ctrl_null {type=Null, loc=<8.0653 -9.2667 146.3717>, rot=<0.3642 0.3562 -0.6035 0.6134>}
 dyn_hair_2_l_ctrl {controlType=EULER_TRANSFORM, parent=dyn_hair_2_l_ctrl_null, type=Control, loc=<22.7354 -9.2667 144.5971>, rot=<0.3177 0.4014 -0.6293 0.5848>}
 dyn_hair_3_l_ctrl_null {type=Null, loc=<8.1954 -12.7867 139.9552>, rot=<-0.4156 -0.4085 0.5693 -0.5799>}
 dyn_hair_3_l_ctrl {
  controlType=EULER_TRANSFORM
  parent=dyn_hair_3_l_ctrl_null
  type=Control
  loc=<-23.5313 -12.7867 137.9656>
  rot=<0.4639 0.3562 -0.5308 0.6134>
 }
 dyn_hair_4_l_ctrl_null {type=Null, loc=<8.3265 -15.1243 133.0531>, rot=<-0.5058 -0.5005 0.491 -0.5026>}
 dyn_hair_4_l_ctrl {
  controlType=EULER_TRANSFORM
  parent=dyn_hair_4_l_ctrl_null
  type=Control
  loc=<-43.2895 -15.1243 125.8182>
  rot=<0.5814 0.4085 -0.3986 0.5799>
 }
 dyn_hair_5_l_ctrl_null {type=Null, loc=<8.4448 -15.035 126.0163>, rot=<-0.5261 -0.5213 0.4691 -0.4809>}
 dyn_hair_5_l_ctrl {controlType=EULER_TRANSFORM, parent=dyn_hair_5_l_ctrl_null, type=Control, loc=<-9.2737 -15.035 125.6825>, rot=<0.5455 0.5005 -0.4464 0.5026>}
 dyn_hair_6_l_ctrl_null {type=Null, loc=<8.5626 -14.3498 118.7894>, rot=<0.5137 0.5087 -0.4827 0.4943>}
 dyn_hair_6_l_ctrl {
  controlType=EULER_TRANSFORM
  parent=dyn_hair_6_l_ctrl_null
  type=Control
  loc=<7.4783 -14.3498 118.5392>
  rot=<-0.501 -0.5213 0.4959 -0.4809>
 }
 dyn_hair_7_l_ctrl_null {type=Null, loc=<8.6681 -14.0442 112.4321>, rot=<0.5137 0.5087 -0.4827 0.4943>}
 dyn_hair_7_l_ctrl {controlType=EULER_TRANSFORM, parent=dyn_hair_7_l_ctrl_null, type=Control, loc=<2.3027 -14.0442 112.4485>, rot=<0.5137 0.5087 -0.4827 0.4943>}
 dyn_hair_8_l_ctrl_null {type=Null, loc=<8.7631 -13.7841 106.715>, rot=<0.5137 0.5087 -0.4827 0.4943>}
 dyn_hair_8_l_ctrl {controlType=EULER_TRANSFORM, parent=dyn_hair_8_l_ctrl_null, type=Control, loc=<3.0393 -13.7841 106.715>, rot=<0.5137 0.5087 -0.4827 0.4943>}
 dyn_hair_1_c_ctrl_null {type=Null, loc=<0 -7.3626 153.4313>, rot=<-0.3954 -0.3954 0.5862 -0.5862>}
 dyn_hair_1_c_ctrl {controlType=EULER_TRANSFORM, parent=dyn_hair_1_c_ctrl_null, type=Control, loc=<0 0 0>, rot=<0 0 0 1>}
 dyn_hair_2_c_ctrl_null {type=Null, loc=<0 -9.716 146.5087>, rot=<-0.3647 -0.3647 0.6058 -0.6058>}
 dyn_hair_2_c_ctrl {controlType=EULER_TRANSFORM, parent=dyn_hair_2_c_ctrl_null, type=Control, loc=<16.0006 -9.716 145.8162>, rot=<0.3154 0.4117 -0.6329 0.5749>}
 dyn_hair_3_c_ctrl_null {type=Null, loc=<0 -13.141 140.0398>, rot=<0.4077 0.4077 -0.5777 0.5777>}
 dyn_hair_3_c_ctrl {
  controlType=EULER_TRANSFORM
  parent=dyn_hair_3_c_ctrl_null
  type=Control
  loc=<-27.5196 -13.141 137.5041>
  rot=<0.4486 0.3647 -0.5466 0.6058>
 }
 dyn_hair_4_c_ctrl_null {type=Null, loc=<0 -15.5827 133.1725>, rot=<-0.5084 -0.5084 0.4915 -0.4915>}
 dyn_hair_4_c_ctrl {
  controlType=EULER_TRANSFORM
  parent=dyn_hair_4_c_ctrl_null
  type=Control
  loc=<-55.6145 -15.5827 121.2232>
  rot=<0.5912 0.4077 -0.388 0.5777>
 }
 dyn_hair_5_c_ctrl_null {type=Null, loc=<0 -15.3359 125.8828>, rot=<0.533 0.533 -0.4647 0.4647>}
 dyn_hair_5_c_ctrl {
  controlType=EULER_TRANSFORM
  parent=dyn_hair_5_c_ctrl_null
  type=Control
  loc=<-20.1821 -15.3359 124.4683>
  rot=<0.5562 0.5084 -0.4367 0.4915>
 }
 dyn_hair_6_c_ctrl_null {type=Null, loc=<0 -14.3382 118.6297>, rot=<-0.5207 -0.5207 0.4784 -0.4784>}
 dyn_hair_6_c_ctrl {controlType=EULER_TRANSFORM, parent=dyn_hair_6_c_ctrl_null, type=Control, loc=<-1.1409 -14.3382 118.85>, rot=<-0.5081 -0.533 0.4918 -0.4647>}
 dyn_hair_7_c_ctrl_null {type=Null, loc=<0 -13.8127 112.439>, rot=<-0.5207 -0.5207 0.4784 -0.4784>}
 dyn_hair_7_c_ctrl {
  controlType=EULER_TRANSFORM
  parent=dyn_hair_7_c_ctrl_null
  type=Control
  loc=<-6.213 -13.8127 112.439>
  rot=<-0.5207 -0.5207 0.4784 -0.4784>
 }
 dyn_hair_8_c_ctrl_null {type=Null, loc=<0 -13.1302 104.3974>, rot=<-0.5207 -0.5207 0.4784 -0.4784>}
 dyn_hair_8_c_ctrl {
  controlType=EULER_TRANSFORM
  parent=dyn_hair_8_c_ctrl_null
  type=Control
  loc=<-8.0705 -13.1302 104.3974>
  rot=<-0.5207 -0.5207 0.4784 -0.4784>
 }
 dyn_hair_1_r_ctrl_null {type=Null, loc=<-7.0971 -6.064 153.1442>, rot=<-0.3968 -0.4169 0.5867 -0.5697>}
 dyn_hair_1_r_ctrl {controlType=EULER_TRANSFORM, parent=dyn_hair_1_r_ctrl_null, type=Control, loc=<0 0 0>, rot=<0 0 0 1>}
 dyn_hair_2_r_ctrl_null {type=Null, loc=<-7.3696 -6.7034 145.8654>, rot=<-0.2716 -0.2947 0.6541 -0.6415>}
 dyn_hair_2_r_ctrl {controlType=EULER_TRANSFORM, parent=dyn_hair_2_r_ctrl_null, type=Control, loc=<77.7185 -6.7034 124.3076>, rot=<0.0452 0.4864 -0.7068 0.5116>}
 dyn_hair_3_r_ctrl_null {type=Null, loc=<-7.6057 -12.1455 139.9899>, rot=<0.3819 0.4025 -0.5965 0.58>}
 dyn_hair_3_r_ctrl {
  controlType=EULER_TRANSFORM
  parent=dyn_hair_3_r_ctrl_null
  type=Control
  loc=<-62.932 -12.1455 126.0182>
  rot=<0.4805 0.2947 -0.5204 0.6415>
 }
 dyn_hair_4_r_ctrl_null {type=Null, loc=<-7.8841 -15.1568 132.7644>, rot=<-0.4844 -0.5016 0.5167 -0.4968>}
 dyn_hair_4_r_ctrl {controlType=EULER_TRANSFORM, parent=dyn_hair_4_r_ctrl_null, type=Control, loc=<-62.3223 -15.1568 118.2767>, rot=<0.5705 0.4025 -0.4196 0.58>}
 dyn_hair_5_r_ctrl_null {type=Null, loc=<-8.1598 -15.3607 125.3633>, rot=<0.5483 0.5302 -0.4539 0.4606>}
 dyn_hair_5_r_ctrl {
  controlType=EULER_TRANSFORM
  parent=dyn_hair_5_r_ctrl_null
  type=Control
  loc=<-40.0939 -12.9089 120.0838>
  rot=<0.604 0.4652 -0.4147 0.4968>
 }
 dyn_hair_6_r_ctrl_null {type=Null, loc=<-7.967 -14.1353 117.9654>, rot=<-0.536 -0.5177 0.4684 -0.4747>}
 dyn_hair_6_r_ctrl {
  controlType=EULER_TRANSFORM
  parent=dyn_hair_6_r_ctrl_null
  type=Control
  loc=<-9.1285 -14.1353 118.6246>
  rot=<-0.5232 -0.5302 0.4826 -0.4606>
 }
 dyn_hair_7_r_ctrl_null {type=Null, loc=<-7.8107 -13.4501 111.7924>, rot=<-0.536 -0.5177 0.4684 -0.4747>}
 dyn_hair_7_r_ctrl {
  controlType=EULER_TRANSFORM
  parent=dyn_hair_7_r_ctrl_null
  type=Control
  loc=<-14.0236 -13.4501 111.7925>
  rot=<0.536 0.5177 -0.4684 0.4747>
 }
 dyn_hair_8_r_ctrl_null {type=Null, loc=<-7.6667 -12.8188 106.1054>, rot=<-0.536 -0.5177 0.4684 -0.4747>}
 dyn_hair_8_r_ctrl {
  controlType=EULER_TRANSFORM
  parent=dyn_hair_8_r_ctrl_null
  type=Control
  loc=<-13.3905 -12.8188 106.1054>
  rot=<0.536 0.5177 -0.4684 0.4747>
 }
 dyn_frontHair_A_01_l_ctrl_null {type=Null, loc=<8.4657 0.128 153.1636>, rot=<0.0032 -0.6986 -0.1024 -0.7082>}
 dyn_frontHair_A_01_l_ctrl {controlType=EULER_TRANSFORM, parent=dyn_frontHair_A_01_l_ctrl_null, type=Control, loc=<0 0 0>, rot=<0 0 0 1>}
 dyn_frontHair_A_02_l_ctrl_null {type=Null, loc=<7.9839 -0.0498 145.87>, rot=<0.1043 -0.5815 -0.3414 -0.731>}
 dyn_frontHair_A_02_l_ctrl {
  controlType=EULER_TRANSFORM
  parent=dyn_frontHair_A_02_l_ctrl_null
  type=Control
  loc=<31.2801 -75.2411 120.991>
  rot=<0.2257 0.5925 0.367 0.6807>
 }
 dyn_frontHair_A_03_l_ctrl_null {type=Null, loc=<8.6463 2.7156 139.1253>, rot=<-0.2036 0.4909 0.391 0.7515>}
 dyn_frontHair_A_03_l_ctrl {
  controlType=EULER_TRANSFORM
  parent=dyn_frontHair_A_03_l_ctrl_null
  type=Control
  loc=<16.1411 -32.6136 134.3161>
  rot=<-0.1029 0.4642 0.4894 0.731>
 }
 dyn_frontHair_A_04_l_ctrl_null {type=Null, loc=<10.1941 5.5417 132.5878>, rot=<0.1983 -0.5359 -0.4166 -0.7071>}
 dyn_frontHair_A_04_l_ctrl {
  controlType=EULER_TRANSFORM
  parent=dyn_frontHair_A_04_l_ctrl_null
  type=Control
  loc=<-9.3814 5.8556 132.2739>
  rot=<-0.1921 0.51 0.3718 0.7515>
 }
 dyn_frontHair_A_05_l_ctrl_null {type=Null, loc=<10.767 8.2881 125.855>, rot=<-0.3752 0.566 0.4202 0.6019>}
 dyn_frontHair_A_05_l_ctrl {
  controlType=EULER_TRANSFORM
  parent=dyn_frontHair_A_05_l_ctrl_null
  type=Control
  loc=<-33.3659 -26.8232 118.6906>
  rot=<-0.3729 0.3904 0.4567 0.7071>
 }
 dyn_frontHair_A_06_l_ctrl_null {type=Null, loc=<10.8116 8.8822 118.5579>, rot=<0.3555 -0.6067 -0.4416 -0.5572>}
 dyn_frontHair_A_06_l_ctrl {
  controlType=EULER_TRANSFORM
  parent=dyn_frontHair_A_06_l_ctrl_null
  type=Control
  loc=<-3.4466 8.8796 118.5594>
  rot=<-0.3343 0.6073 0.3964 0.6019>
 }
 dyn_frontHair_A_07_l_ctrl_null {type=Null, loc=<9.8866 9.3266 111.3091>, rot=<0.3555 -0.6067 -0.4416 -0.5572>}
 dyn_frontHair_A_07_l_ctrl {
  controlType=EULER_TRANSFORM
  parent=dyn_frontHair_A_07_l_ctrl_null
  type=Control
  loc=<2.5655 9.3266 111.3091>
  rot=<-0.3555 0.6067 0.4416 0.5572>
 }
]
variables [
 ClavFrontDelta_l {type=double, default=0}
 ClavUpDelta_l {type=double, default=0}
 Crouching {type=float, default=0}
 CrouchingInPlace {type=double, default=0}
 HeadADSBigGun {type=float, default=0}
 HeadLookingUp_Crouching {type=double, default=0}
 Melee Anim {type=double, default=0}
 TorsoTwistRight {type=double, default=0}
 head rz n60n80 alpha {type=double, default=0}
 headLookingUp {type=double, default=0}
 headTiltingSideways {type=double, default=0}
 idleFE {type=bool, default=False}
 isAiming {type=bool, default=False}
 isPistolASDUp {type=float, default=0}
 pub isDiving {type=bool, default=False}
 pub isFloating {type=bool, default=False}
 pub isRigidBodies Enabled {type=bool, default=False}
 pub isWearingBackBling {type=bool, default=False}
]
