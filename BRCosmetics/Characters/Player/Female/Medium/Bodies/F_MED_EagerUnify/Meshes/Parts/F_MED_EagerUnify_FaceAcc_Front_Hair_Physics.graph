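# Physics asset for the front hair of the EagerUnify medium female cosmetic body, covering one left-side strand.
# Six chained dyn_frontHair capsule bodies hang from a base body linked to the head, with limited swing and almost no twist.
# The last three segments add offset capsules to widen their collision.
# Head, neck, spine, clavicle, pec and upper-arm shapes, including a large neck box, form the body collision volumes, joined by free-rotation constraints.
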PhysicsAsset /BRCosmetics/Characters/Player/Female/Medium/Bodies/F_MED_EagerUnify/Meshes/Parts/F_MED_EagerUnify_FaceAcc_Front_Hair_Physics
bodies {
 base_frontHair_A_01_l [{shapeType=Capsule, length=0.505, radius=0.505, shapeName=base_frontHair_A_01_l_capsule, rotation=<rot 0 -90 0>}]
 clavicle_l [{
   shapeType=Capsule
   length=7.8238
   radius=3.8771
   shapeName=clavicle_l_capsule
   center=<5.0016 0.7488 0.5058>
   rotation=<rot -0.0002 -94.7719 101.5585>
  }]
 clavicle_r [{
   shapeType=Capsule
   length=7.3238
   radius=5.426
   shapeName=clavicle_l_capsule
   center=<-6.0597 -0.5752 -1.5348>
   rotation=<rot -0.0988 -96.3072 98.4075>
  }]
 deform_pec_aim_l [{shapeType=Capsule, length=7.8944, radius=4.2789, shapeName=deform_pec_aim_l_capsule, center=<7.7881 0 0>, rotation=<rot 0 -90 90>}]
 deform_pec_aim_r [{shapeType=Capsule, length=7.8944, radius=4.2789, shapeName=deform_pec_aim_r_capsule, center=<-7.7881 0 0>, rotation=<rot 0 -90 90>}]
 dyn_frontHair_A_01_l [{shapeType=Capsule, length=4.3949, radius=2.2866, shapeName=dyn_hair_l_01_capsule, center=<3.7008 0 0>, rotation=<rot 90 0 0>}]
 dyn_frontHair_A_02_l [{shapeType=Capsule, length=4.3949, radius=2.2866, shapeName=dyn_hair_l_01_capsule, center=<3.7008 0.1362 0.2033>, rotation=<rot 90 0 0>}]
 dyn_frontHair_A_03_l [{
   shapeType=Capsule
   length=4.3949
   radius=2.2866
   shapeName=dyn_hair_l_01_capsule
   center=<3.7008 0 0>
   rotation=<rot 85.5185 -179.9999 -179.9999>
  }]
 dyn_frontHair_A_04_l [
  {
   shapeType=Capsule
   length=5.4129
   radius=1.5924
   shapeName=dyn_hair_l_01_capsule
   center=<4.2303 3.4378 0.7251>
   rotation=<rot 77.4975 167.0311 -180>
  }
  {
   shapeType=Capsule
   length=6.0107
   radius=2.5698
   shapeName=dyn_hair_l_01_capsule
   center=<3.6692 -3.018 0.4891>
   rotation=<rot 60.2578 -79.5887 -63.6142>
  }
  {shapeType=Capsule, length=5.2295, radius=2.5962, shapeName=dyn_hair_l_01_capsule, center=<3.7008 0 1.1711>, rotation=<rot 90 0 0>}
 ]
 dyn_frontHair_A_05_l [
  {
   shapeType=Capsule
   length=6.9142
   radius=2.8514
   shapeName=dyn_hair_l_01_capsule
   center=<3.1657 -3.6925 -1.7263>
   rotation=<rot 49.4085 -89.5419 -83.9872>
  }
  {shapeType=Capsule, length=4.3949, radius=1.9184, shapeName=dyn_hair_l_01_capsule, center=<3.7008 0 0.4926>, rotation=<rot 90 0 0>}
 ]
 dyn_frontHair_A_06_l [
  {shapeType=Capsule, length=4.8993, radius=1.7822, shapeName=dyn_hair_l_01_capsule, center=<3.7008 0 0>, rotation=<rot 90 0 0>}
  {
   shapeType=Capsule
   length=6.4903
   radius=2.7425
   shapeName=dyn_hair_l_01_capsule
   center=<1.4866 -3.8007 -2.6209>
   rotation=<rot 54.2007 94.9011 90.015>
  }
 ]
 head [{shapeType=Capsule, length=8.1154, radius=5.701, shapeName=head_capsule, center=<0.7935 4.6689 0.2132>, rotation=<rot -17.7283 -90 -89.4925>}]
 neck_01 [
  {
   shapeType=Capsule
   length=106.2941
   radius=4.0001
   shapeName=neck_01_capsule
   center=<-0.0836 -2.0046 -14.8585>
   rotation=<rot 0 -104.8676 180>
  }
  {
   shapeType=Box
   extentX=67.93
   extentY=7.5
   extentZ=77.4536
   shapeName=neck_01_box
   center=<-5.4948 -3.6971 -18.7946>
   rotation=<rot 0 -13.7407 0>
  }
 ]
 neck_02 [{shapeType=Capsule, length=4, radius=6.7139, shapeName=neck_02_capsule, center=<-0.3674 0 0>, rotation=<rot 0 -90 90>}]
 spine_01 [{shapeType=Capsule, length=21.6802, radius=10.1696, shapeName=spine_05_capsule, center=<-2.4195 -1.8029 0>, rotation=<rot 0 -90 0>}]
 spine_02 [{shapeType=Capsule, length=13.5593, radius=10.7904, shapeName=spine_05_capsule, center=<0.2641 0.1017 0>, rotation=<rot 0 -90 0>}]
 spine_03 [
  {shapeType=Capsule, length=7.8312, radius=10.6239, shapeName=spine_05_capsule, center=<6.2994 4.2898 0>, rotation=<rot 0 -90 0>}
  {shapeType=Capsule, length=14.3178, radius=10.6239, shapeName=spine_05_capsule, center=<3.8652 2.0057 0>, rotation=<rot 0 -90 0>}
 ]
 spine_04 [{shapeType=Capsule, length=13.1917, radius=10.031, shapeName=spine_05_capsule, center=<2.4639 0.7376 0>, rotation=<rot 0 -90 0>}]
 spine_05 [{shapeType=Capsule, length=21.9446, radius=7.4651, shapeName=spine_05_capsule, center=<0.2641 0.1017 0>, rotation=<rot 0 -90 0>}]
 upperarm_l [{shapeType=Capsule, length=46.9456, radius=4.9222, shapeName=upperarm_l_capsule, center=<22.5016 1.3544 0>, rotation=<rot 0 -90 90>}]
 upperarm_r [{shapeType=Capsule, length=42.9253, radius=3.9234, shapeName=upperarm_r_capsule, center=<-15.0717 -1.4224 0>, rotation=<rot 0 -90 90>}]
}
constraints [
 {bone1Name=dyn_frontHair_A_06_l, bone2Name=dyn_frontHair_A_05_l, swing1LimitDegrees=20, swing2LimitDegrees=30, twistLimitDegrees=1}
 {bone1Name=dyn_frontHair_A_01_l, bone2Name=base_frontHair_A_01_l, swing1LimitDegrees=15, swing2LimitDegrees=15, twistLimitDegrees=0}
 {bone1Name=dyn_frontHair_A_02_l, bone2Name=dyn_frontHair_A_01_l, swing1LimitDegrees=20, swing2LimitDegrees=30, twistLimitDegrees=1}
 {bone1Name=dyn_frontHair_A_03_l, bone2Name=dyn_frontHair_A_02_l, swing1LimitDegrees=20, swing2LimitDegrees=25, twistLimitDegrees=1}
 {bone1Name=dyn_frontHair_A_04_l, bone2Name=dyn_frontHair_A_03_l, swing1LimitDegrees=20, swing2LimitDegrees=30, twistLimitDegrees=1}
 {bone1Name=dyn_frontHair_A_05_l, bone2Name=dyn_frontHair_A_04_l, swing1LimitDegrees=20, swing2LimitDegrees=30, twistLimitDegrees=1}
 {
  bone1Name=base_frontHair_A_01_l
  bone2Name=head
  swing1LimitDegrees=45
  swing1Motion=Free
  swing2LimitDegrees=45
  swing2Motion=Free
  twistLimitDegrees=45
  twistMotion=Free
 }
 {
  bone1Name=spine_05
  bone2Name=spine_04
  swing1LimitDegrees=45
  swing1Motion=Free
  swing2LimitDegrees=45
  swing2Motion=Free
  twistLimitDegrees=45
  twistMotion=Free
 }
 {
  bone1Name=neck_01
  bone2Name=spine_04
  swing1LimitDegrees=45
  swing1Motion=Free
  swing2LimitDegrees=45
  swing2Motion=Free
  twistLimitDegrees=45
  twistMotion=Free
 }
 {
  bone1Name=neck_02
  bone2Name=spine_04
  swing1LimitDegrees=45
  swing1Motion=Free
  swing2LimitDegrees=45
  swing2Motion=Free
  twistLimitDegrees=45
  twistMotion=Free
 }
 {
  bone1Name=head
  bone2Name=spine_04
  swing1LimitDegrees=45
  swing1Motion=Free
  swing2LimitDegrees=45
  swing2Motion=Free
  twistLimitDegrees=45
  twistMotion=Free
 }
 {
  bone1Name=clavicle_l
  bone2Name=spine_04
  swing1LimitDegrees=45
  swing1Motion=Free
  swing2LimitDegrees=45
  swing2Motion=Free
  twistLimitDegrees=45
  twistMotion=Free
 }
 {
  bone1Name=upperarm_l
  bone2Name=spine_04
  swing1LimitDegrees=45
  swing1Motion=Free
  swing2LimitDegrees=45
  swing2Motion=Free
  twistLimitDegrees=45
  twistMotion=Free
 }
 {
  bone1Name=clavicle_r
  bone2Name=spine_04
  swing1LimitDegrees=45
  swing1Motion=Free
  swing2LimitDegrees=45
  swing2Motion=Free
  twistLimitDegrees=45
  twistMotion=Free
 }
 {
  bone1Name=upperarm_r
  bone2Name=spine_04
  swing1LimitDegrees=45
  swing1Motion=Free
  swing2LimitDegrees=45
  swing2Motion=Free
  twistLimitDegrees=45
  twistMotion=Free
 }
 {
  bone1Name=deform_pec_aim_l
  bone2Name=spine_04
  swing1LimitDegrees=45
  swing1Motion=Free
  swing2LimitDegrees=45
  swing2Motion=Free
  twistLimitDegrees=45
  twistMotion=Free
 }
 {
  bone1Name=deform_pec_aim_r
  bone2Name=spine_04
  swing1LimitDegrees=45
  swing1Motion=Free
  swing2LimitDegrees=45
  swing2Motion=Free
  twistLimitDegrees=45
  twistMotion=Free
 }
 {
  bone1Name=spine_04
  bone2Name=spine_03
  swing1LimitDegrees=45
  swing1Motion=Free
  swing2LimitDegrees=45
  swing2Motion=Free
  twistLimitDegrees=45
  twistMotion=Free
 }
]
