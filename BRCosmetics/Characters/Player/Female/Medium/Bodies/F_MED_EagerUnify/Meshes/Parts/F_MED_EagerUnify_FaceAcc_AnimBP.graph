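# Animation Blueprint for the EagerUnify female face accessory, adding secondary hair motion to the pose copied from the parent mesh.
# A Control Rig feeds switchable back-hair branches: per-bone AnimDynamics, AnimDynamics plus RigidBody, or bypass.
# The result enters one of two front-hair RigidBody setups (FE or standard), then additive bone offsets.
# EventGraph updates backbling, crouch, melee and front-end idle states, plus gravity overrides from a parameter asset and sliding.
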AnimBlueprint /BRCosmetics/Characters/Player/Female/Medium/Bodies/F_MED_EagerUnify/Meshes/Parts/F_MED_EagerUnify_FaceAcc_AnimBP
target_skeleton=&/F_MED_EagerUnify_FaceAcc_Skeleton
graphs [
 AnimGraph {nodes [
   n1 = "Transform (Modify) Bone - Bone: dyn_hair_4_l"!(
    ComponentPose=Local To Component(LocalPose=Use cached pose 'CRCache_Dyn'())
    bAlphaBoolEnabled:bool=True
    Translation="0,0,1"
    Rotation="2,0,0"
   ) -> (Pose) {
    alphaInputType=Bool
    rotationMode=BMM_Additive
    rotationSpace=BCS_BoneSpace
    translationMode=BMM_Additive
    translationSpace=BCS_BoneSpace
    alphaBoolBlend {blendInTime=2, blendOption=Linear, blendOutTime=2}
    boneToModify {boneName=dyn_hair_4_l}
    rotation {pitch=2}
    translation {z=1}
   }
   n2 = "Transform (Modify) Bone - Bone: dyn_hair_4_c"!(ComponentPose=@n1, bAlphaBoolEnabled:bool=True, Translation="0,0,1", Rotation="2,0,0") -> (Pose) {
    alphaInputType=Bool
    rotationMode=BMM_Additive
    rotationSpace=BCS_BoneSpace
    translationMode=BMM_Additive
    translationSpace=BCS_BoneSpace
    alphaBoolBlend {blendInTime=2, blendOption=Linear, blendOutTime=2}
    boneToModify {boneName=dyn_hair_4_c}
    rotation {pitch=2}
    translation {z=1}
   }
   n3 = "Transform (Modify) Bone - Bone: dyn_hair_4_r"!(ComponentPose=@n2, bAlphaBoolEnabled:bool=True, Translation="0,0,1", Rotation="2,0,0") -> (Pose) {
    alphaInputType=Bool
    rotationMode=BMM_Additive
    rotationSpace=BCS_BoneSpace
    translationMode=BMM_Additive
    translationSpace=BCS_BoneSpace
    alphaBoolBlend {blendInTime=2, blendOption=Linear, blendOutTime=2}
    boneToModify {boneName=dyn_hair_4_r}
    rotation {pitch=2}
    translation {z=1}
   }
   Output Pose(Result=Component To Local(ComponentPose=@n3)) {layerGroup=None}
   n4 = Control Rig!(
    Source=Copy Pose From Mesh() {bUseAttachedParent=true}
    Alpha
    pub isRigidBodies Enabled:bool=false
    pub isWearingBackBling:bool=false
    isAiming:bool=false
    Melee Anim:double=0
    pub isFloating:bool=false
    idleFE:bool=false
   ) -> (Pose) {
    alphaBoolBlend {blendOption=Linear}
    controlRigAssetReference {blueprintRigClass=&/F_MED_EagerUnify_FaceAcc_CtrlRig.F_MED_EagerUnify_FaceAcc_CtrlRig_C}
   }
   Save cached pose 'CRCache_Input'(Pose=@n4) {cacheName=CRCache_Input}
   n5 = RigidBody!(ComponentPose=Local To Component(LocalPose=Use cached pose 'BackHair_dynamics'()), Alpha, OverrideWorldGravity:Vector="0,0,0", ExternalForce) -> (Pose) {
    bClampLinearTranslationLimitToRefPose=true
    bOverrideWorldGravity=true
    overridePhysicsAsset=&/F_MED_EagerUnify_FaceAcc_Front_Hair_FE_Physics
    simulationSpace=BaseBoneSpace
    baseBoneRef {boneName=head}
    componentAppliedLinearAccClamp=<3000 3000 3000>
    overrideWorldGravity {z=-980}
    simSpaceSettings {
     dampingAlpha=0.65
     maxAngularAcceleration=40
     maxAngularVelocity=6
     maxLinearAcceleration=900
     maxLinearVelocity=400
     worldAlpha=1
     externalLinearDragV=<0.2 0.2 0.2>
    }
   }
   n6 = RigidBody(ComponentPose=Local To Component(LocalPose=Use cached pose 'BackHair_dynamics'()), OverrideWorldGravity="0,0,0") {
    bClampLinearTranslationLimitToRefPose=true
    bOverrideWorldGravity=true
    overridePhysicsAsset=&/F_MED_EagerUnify_FaceAcc_Front_Hair_Physics
    simulationSpace=BaseBoneSpace
    baseBoneRef {boneName=spine_05}
    componentAppliedLinearAccClamp=<3000 3000 3000>
    overrideWorldGravity {z=-980}
    simSpaceSettings {
     dampingAlpha=0.65
     maxAngularAcceleration=40
     maxAngularVelocity=6
     maxLinearAcceleration=900
     maxLinearVelocity=400
     worldAlpha=1
     externalLinearDragV=<0.2 0.2 0.2>
    }
   }
   Save cached pose 'CRCache_Dyn'(Pose=Blend Poses by bool(BlendPose_0=Component To Local(ComponentPose=@n5), BlendPose_1=Component To Local(ComponentPose=@n6), bActiveValue=true)) {cacheName=CRCache_Dyn}
   n7 = Anim Dynamics - Bone: dyn_hair_1_r!(ComponentPose=Local To Component(LocalPose=Use cached pose 'CRCache_Input'()), Alpha=1, GravityOverride:Vector="0,0,-980") -> (Pose) {
    angularDampingOverride=1
    angularSpringConstant=10000
    bAngularSpring=true
    bOverrideAngularDamping=true
    bUseGravityOverride=true
    gravityScale=0.35
    simulationSpace=RootRelative
    boundBone {boneName=dyn_hair_1_r}
    componentAppliedLinearAccClamp=<10000 10000 10000>
    componentLinearAccScale=<1 1 1>
    componentLinearVelScale=<1 1 1>
    gravityOverride {z=-980}
    physicsBodyDefinitions [{
      boundBone {boneName=dyn_hair_1_r}
      boxExtents=<20 30 30>
      constraintSetup {angularLimitsMax=<0 0 3>, angularLimitsMin=<0 -15 -3>, angularTarget=<1 0 0>}
      localJointOffset=<4 0 0>
     }]
    simSpaceSettings {maxAngularAcceleration=25, maxAngularVelocity=5, simSpaceAngularAlpha=1}
   }
   n8 = Anim Dynamics - Bone: dyn_hair_1_l!(ComponentPose=@n7, Alpha=1, GravityOverride:Vector="0,0,-980") -> (Pose) {
    angularDampingOverride=1
    angularSpringConstant=10000
    bAngularSpring=true
    bOverrideAngularDamping=true
    bUseGravityOverride=true
    gravityScale=0.35
    simulationSpace=RootRelative
    boundBone {boneName=dyn_hair_1_l}
    componentAppliedLinearAccClamp=<10000 10000 10000>
    componentLinearAccScale=<1 1 1>
    componentLinearVelScale=<1 1 1>
    gravityOverride {z=-980}
    physicsBodyDefinitions [{
      boundBone {boneName=dyn_hair_1_l}
      boxExtents=<20 30 30>
      constraintSetup {angularLimitsMax=<0 0 3>, angularLimitsMin=<0 -15 -3>, angularTarget=<1 0 0>}
      localJointOffset=<4 0 0>
     }]
    simSpaceSettings {maxAngularAcceleration=25, maxAngularVelocity=5, simSpaceAngularAlpha=1}
   }
   n9 = Anim Dynamics - Bone: dyn_hair_1_c!(ComponentPose=@n8, Alpha=1, GravityOverride:Vector="0,0,-980") -> (Pose) {
    angularDampingOverride=1
    angularSpringConstant=10000
    bAngularSpring=true
    bOverrideAngularDamping=true
    bUseGravityOverride=true
    gravityScale=0.35
    simulationSpace=RootRelative
    boundBone {boneName=dyn_hair_1_c}
    componentAppliedLinearAccClamp=<10000 10000 10000>
    componentLinearAccScale=<1 1 1>
    componentLinearVelScale=<1 1 1>
    gravityOverride {z=-980}
    physicsBodyDefinitions [{
      boundBone {boneName=dyn_hair_1_c}
      boxExtents=<20 30 30>
      constraintSetup {angularLimitsMax=<0 0 3>, angularLimitsMin=<0 -15 -3>, angularTarget=<1 0 0>}
      localJointOffset=<4 0 0>
     }]
    simSpaceSettings {maxAngularAcceleration=25, maxAngularVelocity=5, simSpaceAngularAlpha=1}
   }
   n10 = Anim Dynamics - Bone: dyn_hair_2_r!(ComponentPose=@n9, Alpha=1, GravityOverride:Vector="0,0,-980") -> (Pose) {
    angularDampingOverride=1
    angularSpringConstant=10000
    bAngularSpring=true
    bOverrideAngularDamping=true
    bUseGravityOverride=true
    gravityScale=0.35
    simulationSpace=RootRelative
    boundBone {boneName=dyn_hair_2_r}
    componentAppliedLinearAccClamp=<10000 10000 10000>
    componentLinearAccScale=<1 1 1>
    componentLinearVelScale=<1 1 1>
    gravityOverride {z=-980}
    physicsBodyDefinitions [{
      boundBone {boneName=dyn_hair_2_r}
      boxExtents=<20 30 30>
      constraintSetup {angularLimitsMax=<0 0 3>, angularLimitsMin=<0 -25 -3>, angularTarget=<1 0 0>}
      localJointOffset=<4 0 0>
     }]
    simSpaceSettings {maxAngularAcceleration=25, maxAngularVelocity=5, simSpaceAngularAlpha=1}
   }
   n11 = Anim Dynamics - Bone: dyn_hair_2_l!(ComponentPose=@n10, Alpha=1, GravityOverride:Vector="0,0,-980") -> (Pose) {
    angularDampingOverride=1
    angularSpringConstant=10000
    bAngularSpring=true
    bOverrideAngularDamping=true
    bUseGravityOverride=true
    gravityScale=0.35
    simulationSpace=RootRelative
    boundBone {boneName=dyn_hair_2_l}
    componentAppliedLinearAccClamp=<10000 10000 10000>
    componentLinearAccScale=<1 1 1>
    componentLinearVelScale=<1 1 1>
    gravityOverride {z=-980}
    physicsBodyDefinitions [{
      boundBone {boneName=dyn_hair_2_l}
      boxExtents=<20 30 30>
      constraintSetup {angularLimitsMax=<0 0 3>, angularLimitsMin=<0 -25 -3>, angularTarget=<1 0 0>}
      localJointOffset=<4 0 0>
     }]
    simSpaceSettings {maxAngularAcceleration=25, maxAngularVelocity=5, simSpaceAngularAlpha=1}
   }
   n12 = Anim Dynamics - Bone: dyn_hair_2_c!(ComponentPose=@n11, Alpha=1, GravityOverride:Vector="0,0,-980") -> (Pose) {
    angularDampingOverride=1
    angularSpringConstant=10000
    bAngularSpring=true
    bOverrideAngularDamping=true
    bUseGravityOverride=true
    gravityScale=0.35
    simulationSpace=RootRelative
    boundBone {boneName=dyn_hair_2_c}
    componentAppliedLinearAccClamp=<10000 10000 10000>
    componentLinearAccScale=<1 1 1>
    componentLinearVelScale=<1 1 1>
    gravityOverride {z=-980}
    physicsBodyDefinitions [{
      boundBone {boneName=dyn_hair_2_c}
      boxExtents=<20 30 30>
      constraintSetup {angularLimitsMax=<0 0 3>, angularLimitsMin=<0 -25 -3>, angularTarget=<1 0 0>}
      localJointOffset=<4 0 0>
     }]
    simSpaceSettings {maxAngularAcceleration=25, maxAngularVelocity=5, simSpaceAngularAlpha=1}
   }
   n13 = RigidBody!(ComponentPose=@n12, Alpha, OverrideWorldGravity="0,0,0") -> (Pose) {
    bOverrideWorldGravity=true
    overridePhysicsAsset=&/F_MED_EagerUnify_FaceAcc_Physics
    simulationSpace=BaseBoneSpace
    alphaBoolBlend {blendOption=Linear}
    alphaScaleBiasClamp {interpSpeedDecreasing=3, interpSpeedIncreasing=3}
    baseBoneRef {boneName=spine_05}
    componentAppliedLinearAccClamp=<3000 3000 3000>
    simSpaceSettings {
     dampingAlpha=0.5
     maxAngularAcceleration=100
     maxAngularVelocity=3
     maxLinearAcceleration=800
     maxLinearVelocity=400
     velocityScaleZ=0.75
     worldAlpha=1
     externalLinearDragV=<1 1 1>
    }
   }
   n14 = Component To Local(ComponentPose=@n13)
   n15 = Anim Dynamics - Bone: dyn_hair_3_l!(ComponentPose=Local To Component(LocalPose=Use cached pose 'CRCache_Input'()), Alpha=0.85, GravityOverride:Vector="-980,0,0") -> (Pose) {
    alpha=0.85
    angularDampingOverride=1
    angularSpringConstant=100
    bAngularSpring=true
    bOverrideAngularDamping=true
    bUseGravityOverride=true
    gravityScale=0.35
    simulationSpace=RootRelative
    boundBone {boneName=dyn_hair_3_l}
    componentAppliedLinearAccClamp=<6000 6000 6000>
    componentLinearAccScale=<0.5 0.5 0.5>
    componentLinearVelScale=<0.6 0.6 0.6>
    gravityOverride {x=-980}
    physicsBodyDefinitions [{
      boundBone {boneName=dyn_hair_3_l}
      boxExtents=<10 10 10>
      constraintSetup {angularLimitsMax=<0 0 2.5>, angularLimitsMin=<0 -10 -2.5>, angularTarget=<1 0 0>}
      localJointOffset=<2 0 0>
     }]
    relativeSpaceBone {boneName=neck_01}
    simSpaceSettings {maxAngularAcceleration=15, maxAngularVelocity=7, simSpaceAngularAlpha=0.75}
   }
   n16 = Anim Dynamics - Bone: dyn_hair_4_l!(ComponentPose=@n15, Alpha=0.85, GravityOverride:Vector="-980,0,0") -> (Pose) {
    alpha=0.85
    angularDampingOverride=1
    angularSpringConstant=100
    bAngularSpring=true
    bOverrideAngularDamping=true
    bUseGravityOverride=true
    gravityScale=0.35
    simulationSpace=RootRelative
    boundBone {boneName=dyn_hair_4_l}
    componentAppliedLinearAccClamp=<6000 6000 6000>
    componentLinearAccScale=<0.5 0.5 0.5>
    componentLinearVelScale=<0.6 0.6 0.6>
    gravityOverride {x=-980}
    physicsBodyDefinitions [{
      boundBone {boneName=dyn_hair_4_l}
      boxExtents=<10 10 10>
      constraintSetup {angularLimitsMax=<0 2 2.5>, angularLimitsMin=<0 -10 -2.5>, angularTarget=<1 0 0>}
      localJointOffset=<2 0 0>
     }]
    relativeSpaceBone {boneName=neck_01}
    simSpaceSettings {maxAngularAcceleration=15, maxAngularVelocity=7, simSpaceAngularAlpha=0.75}
   }
   n17 = Anim Dynamics - Bone: dyn_hair_5_l!(ComponentPose=@n16, Alpha=1, GravityOverride:Vector="-980,0,0") -> (Pose) {
    angularDampingOverride=1
    angularSpringConstant=100
    bAngularSpring=true
    bOverrideAngularDamping=true
    bUseGravityOverride=true
    gravityScale=0.35
    simulationSpace=RootRelative
    boundBone {boneName=dyn_hair_5_l}
    componentAppliedLinearAccClamp=<6000 6000 6000>
    componentLinearAccScale=<0.5 0.5 0.5>
    componentLinearVelScale=<0.6 0.6 0.6>
    gravityOverride {x=-980}
    physicsBodyDefinitions [{
      boundBone {boneName=dyn_hair_5_l}
      boxExtents=<10 10 10>
      constraintSetup {angularLimitsMax=<0 2 2.5>, angularLimitsMin=<0 -10 -2.5>, angularTarget=<1 0 0>}
      localJointOffset=<2 0 0>
     }]
    relativeSpaceBone {boneName=neck_01}
    simSpaceSettings {maxAngularAcceleration=15, maxAngularVelocity=7, simSpaceAngularAlpha=0.75}
   }
   n18 = Anim Dynamics - Bone: dyn_hair_6_l!(ComponentPose=@n17, Alpha=0.85, GravityOverride:Vector="-980,0,0") -> (Pose) {
    alpha=0.85
    angularDampingOverride=1
    angularSpringConstant=100
    bAngularSpring=true
    bOverrideAngularDamping=true
    bUseGravityOverride=true
    gravityScale=0.35
    simulationSpace=RootRelative
    boundBone {boneName=dyn_hair_6_l}
    componentAppliedLinearAccClamp=<6000 6000 6000>
    componentLinearAccScale=<0.5 0.5 0.5>
    componentLinearVelScale=<0.6 0.6 0.6>
    gravityOverride {x=-980}
    physicsBodyDefinitions [{
      boundBone {boneName=dyn_hair_6_l}
      boxExtents=<10 10 10>
      constraintSetup {angularLimitsMax=<0 0 5>, angularLimitsMin=<0 -10 -5>, angularTarget=<1 0 0>}
      localJointOffset=<2 0 0>
     }]
    relativeSpaceBone {boneName=neck_01}
    simSpaceSettings {maxAngularAcceleration=15, maxAngularVelocity=7, simSpaceAngularAlpha=0.75}
   }
   n19 = Anim Dynamics - Bone: dyn_hair_7_l!(ComponentPose=@n18, Alpha=0.85, GravityOverride:Vector="-980,0,0") -> (Pose) {
    alpha=0.85
    angularDampingOverride=1
    angularSpringConstant=100
    bAngularSpring=true
    bOverrideAngularDamping=true
    bUseGravityOverride=true
    gravityScale=0.35
    simulationSpace=RootRelative
    boundBone {boneName=dyn_hair_7_l}
    componentAppliedLinearAccClamp=<6000 6000 6000>
    componentLinearAccScale=<0.5 0.5 0.5>
    componentLinearVelScale=<0.6 0.6 0.6>
    gravityOverride {x=-980}
    physicsBodyDefinitions [{
      boundBone {boneName=dyn_hair_7_l}
      boxExtents=<10 10 10>
      constraintSetup {angularLimitsMax=<0 0 5>, angularLimitsMin=<0 -10 -5>, angularTarget=<1 0 0>}
      localJointOffset=<2 0 0>
     }]
    relativeSpaceBone {boneName=neck_01}
    simSpaceSettings {maxAngularAcceleration=15, maxAngularVelocity=7, simSpaceAngularAlpha=0.75}
   }
   n20 = Anim Dynamics - Bone: dyn_hair_8_l!(ComponentPose=@n19, Alpha=0.85, GravityOverride:Vector="-980,0,0") -> (Pose) {
    alpha=0.85
    angularDampingOverride=1
    angularSpringConstant=100
    bAngularSpring=true
    bOverrideAngularDamping=true
    bUseGravityOverride=true
    gravityScale=0.35
    simulationSpace=RootRelative
    boundBone {boneName=dyn_hair_8_l}
    componentAppliedLinearAccClamp=<6000 6000 6000>
    componentLinearAccScale=<0.5 0.5 0.5>
    componentLinearVelScale=<0.6 0.6 0.6>
    gravityOverride {x=-980}
    physicsBodyDefinitions [{
      boundBone {boneName=dyn_hair_8_l}
      boxExtents=<10 10 10>
      constraintSetup {angularLimitsMax=<0 0 5>, angularLimitsMin=<0 -10 -5>, angularTarget=<1 0 0>}
      localJointOffset=<2 0 0>
     }]
    relativeSpaceBone {boneName=neck_01}
    simSpaceSettings {maxAngularAcceleration=15, maxAngularVelocity=7, simSpaceAngularAlpha=0.75}
   }
   n21 = Anim Dynamics - Bone: dyn_hair_3_c!(ComponentPose=@n20, Alpha=0.85, GravityOverride:Vector="-980,0,0") -> (Pose) {
    alpha=0.85
    angularDampingOverride=1
    angularSpringConstant=100
    bAngularSpring=true
    bOverrideAngularDamping=true
    bUseGravityOverride=true
    gravityScale=0.35
    simulationSpace=RootRelative
    boundBone {boneName=dyn_hair_3_c}
    componentAppliedLinearAccClamp=<6000 6000 6000>
    componentLinearAccScale=<0.5 0.5 0.5>
    componentLinearVelScale=<0.6 0.6 0.6>
    gravityOverride {x=-980}
    physicsBodyDefinitions [{
      boundBone {boneName=dyn_hair_3_c}
      boxExtents=<10 10 10>
      constraintSetup {angularLimitsMax=<0 2 2.5>, angularLimitsMin=<0 -10 -2.5>, angularTarget=<1 0 0>}
      localJointOffset=<2 0 0>
     }]
    relativeSpaceBone {boneName=neck_01}
    simSpaceSettings {maxAngularAcceleration=15, maxAngularVelocity=7, simSpaceAngularAlpha=0.75}
   }
   n22 = Anim Dynamics - Bone: dyn_hair_4_c!(ComponentPose=@n21, Alpha=0.85, GravityOverride:Vector="-980,0,0") -> (Pose) {
    alpha=0.85
    angularDampingOverride=1
    angularSpringConstant=100
    bAngularSpring=true
    bOverrideAngularDamping=true
    bUseGravityOverride=true
    gravityScale=0.35
    simulationSpace=RootRelative
    boundBone {boneName=dyn_hair_4_c}
    componentAppliedLinearAccClamp=<6000 6000 6000>
    componentLinearAccScale=<0.5 0.5 0.5>
    componentLinearVelScale=<0.6 0.6 0.6>
    gravityOverride {x=-980}
    physicsBodyDefinitions [{
      boundBone {boneName=dyn_hair_4_c}
      boxExtents=<10 10 10>
      constraintSetup {angularLimitsMax=<0 2 2.5>, angularLimitsMin=<0 -10 -2.5>, angularTarget=<1 0 0>}
      localJointOffset=<2 0 0>
     }]
    relativeSpaceBone {boneName=neck_01}
    simSpaceSettings {maxAngularAcceleration=15, maxAngularVelocity=7, simSpaceAngularAlpha=0.75}
   }
   n23 = Anim Dynamics - Bone: dyn_hair_5_c!(ComponentPose=@n22, Alpha=1, GravityOverride:Vector="-980,0,0") -> (Pose) {
    angularDampingOverride=1
    angularSpringConstant=100
    bAngularSpring=true
    bOverrideAngularDamping=true
    bUseGravityOverride=true
    gravityScale=0.35
    simulationSpace=RootRelative
    boundBone {boneName=dyn_hair_5_c}
    componentAppliedLinearAccClamp=<6000 6000 6000>
    componentLinearAccScale=<0.5 0.5 0.5>
    componentLinearVelScale=<0.6 0.6 0.6>
    gravityOverride {x=-980}
    physicsBodyDefinitions [{
      boundBone {boneName=dyn_hair_5_c}
      boxExtents=<10 10 10>
      constraintSetup {angularLimitsMax=<0 2 2.5>, angularLimitsMin=<0 -10 -2.5>, angularTarget=<1 0 0>}
      localJointOffset=<2 0 0>
     }]
    relativeSpaceBone {boneName=neck_01}
    simSpaceSettings {maxAngularAcceleration=15, maxAngularVelocity=7, simSpaceAngularAlpha=0.75}
   }
   n24 = Anim Dynamics - Bone: dyn_hair_6_c!(ComponentPose=@n23, Alpha=0.85, GravityOverride:Vector="-980,0,0") -> (Pose) {
    alpha=0.85
    angularDampingOverride=1
    angularSpringConstant=100
    bAngularSpring=true
    bOverrideAngularDamping=true
    bUseGravityOverride=true
    gravityScale=0.35
    simulationSpace=RootRelative
    boundBone {boneName=dyn_hair_6_c}
    componentAppliedLinearAccClamp=<6000 6000 6000>
    componentLinearAccScale=<0.5 0.5 0.5>
    componentLinearVelScale=<0.6 0.6 0.6>
    gravityOverride {x=-980}
    physicsBodyDefinitions [{
      boundBone {boneName=dyn_hair_6_c}
      boxExtents=<10 10 10>
      constraintSetup {angularLimitsMax=<0 0 5>, angularLimitsMin=<0 -10 -5>, angularTarget=<1 0 0>}
      localJointOffset=<2 0 0>
     }]
    relativeSpaceBone {boneName=neck_01}
    simSpaceSettings {maxAngularAcceleration=15, maxAngularVelocity=7, simSpaceAngularAlpha=0.75}
   }
   n25 = Anim Dynamics - Bone: dyn_hair_7_c!(ComponentPose=@n24, Alpha=0.85, GravityOverride:Vector="-980,0,0") -> (Pose) {
    alpha=0.85
    angularDampingOverride=1
    angularSpringConstant=100
    bAngularSpring=true
    bOverrideAngularDamping=true
    bUseGravityOverride=true
    gravityScale=0.35
    simulationSpace=RootRelative
    boundBone {boneName=dyn_hair_7_c}
    componentAppliedLinearAccClamp=<6000 6000 6000>
    componentLinearAccScale=<0.5 0.5 0.5>
    componentLinearVelScale=<0.6 0.6 0.6>
    gravityOverride {x=-980}
    physicsBodyDefinitions [{
      boundBone {boneName=dyn_hair_7_c}
      boxExtents=<10 10 10>
      constraintSetup {angularLimitsMax=<0 0 5>, angularLimitsMin=<0 -10 -5>, angularTarget=<1 0 0>}
      localJointOffset=<2 0 0>
     }]
    relativeSpaceBone {boneName=neck_01}
    simSpaceSettings {maxAngularAcceleration=15, maxAngularVelocity=7, simSpaceAngularAlpha=0.75}
   }
   n26 = Anim Dynamics - Bone: dyn_hair_8_c!(ComponentPose=@n25, Alpha=0.85, GravityOverride:Vector="-980,0,0") -> (Pose) {
    alpha=0.85
    angularDampingOverride=1
    angularSpringConstant=100
    bAngularSpring=true
    bOverrideAngularDamping=true
    bUseGravityOverride=true
    gravityScale=0.35
    simulationSpace=RootRelative
    boundBone {boneName=dyn_hair_8_c}
    componentAppliedLinearAccClamp=<6000 6000 6000>
    componentLinearAccScale=<0.5 0.5 0.5>
    componentLinearVelScale=<0.6 0.6 0.6>
    gravityOverride {x=-980}
    physicsBodyDefinitions [{
      boundBone {boneName=dyn_hair_8_c}
      boxExtents=<10 10 10>
      constraintSetup {angularLimitsMax=<0 0 5>, angularLimitsMin=<0 -10 -5>, angularTarget=<1 0 0>}
      localJointOffset=<2 0 0>
     }]
    relativeSpaceBone {boneName=neck_01}
    simSpaceSettings {maxAngularAcceleration=15, maxAngularVelocity=7, simSpaceAngularAlpha=0.75}
   }
   n27 = Anim Dynamics - Bone: dyn_hair_3_r!(ComponentPose=@n26, Alpha=0.85, GravityOverride:Vector="-980,0,0") -> (Pose) {
    alpha=0.85
    angularDampingOverride=1
    angularSpringConstant=100
    bAngularSpring=true
    bOverrideAngularDamping=true
    bUseGravityOverride=true
    gravityScale=0.35
    simulationSpace=RootRelative
    boundBone {boneName=dyn_hair_3_r}
    componentAppliedLinearAccClamp=<6000 6000 6000>
    componentLinearAccScale=<0.5 0.5 0.5>
    componentLinearVelScale=<0.6 0.6 0.6>
    gravityOverride {x=-980}
    physicsBodyDefinitions [{
      boundBone {boneName=dyn_hair_3_r}
      boxExtents=<10 10 10>
      constraintSetup {angularLimitsMax=<0 0 2.5>, angularLimitsMin=<0 -10 -2.5>, angularTarget=<1 0 0>}
      localJointOffset=<2 0 0>
     }]
    relativeSpaceBone {boneName=neck_01}
    simSpaceSettings {maxAngularAcceleration=15, maxAngularVelocity=7, simSpaceAngularAlpha=0.75}
   }
   n28 = Anim Dynamics - Bone: dyn_hair_4_r!(ComponentPose=@n27, Alpha=0.85, GravityOverride:Vector="-980,0,0") -> (Pose) {
    alpha=0.85
    angularDampingOverride=1
    angularSpringConstant=100
    bAngularSpring=true
    bOverrideAngularDamping=true
    bUseGravityOverride=true
    gravityScale=0.35
    simulationSpace=RootRelative
    boundBone {boneName=dyn_hair_4_r}
    componentAppliedLinearAccClamp=<6000 6000 6000>
    componentLinearAccScale=<0.5 0.5 0.5>
    componentLinearVelScale=<0.6 0.6 0.6>
    gravityOverride {x=-980}
    physicsBodyDefinitions [{
      boundBone {boneName=dyn_hair_4_r}
      boxExtents=<10 10 10>
      constraintSetup {angularLimitsMax=<0 2 2.5>, angularLimitsMin=<0 -10 -2.5>, angularTarget=<1 0 0>}
      localJointOffset=<2 0 0>
     }]
    relativeSpaceBone {boneName=neck_01}
    simSpaceSettings {maxAngularAcceleration=15, maxAngularVelocity=7, simSpaceAngularAlpha=0.75}
   }
   n29 = Anim Dynamics - Bone: dyn_hair_5_r!(ComponentPose=@n28, Alpha=0.85, GravityOverride:Vector="-980,0,0") -> (Pose) {
    alpha=0.85
    angularDampingOverride=1
    angularSpringConstant=100
    bAngularSpring=true
    bOverrideAngularDamping=true
    bUseGravityOverride=true
    gravityScale=0.35
    simulationSpace=RootRelative
    boundBone {boneName=dyn_hair_5_r}
    componentAppliedLinearAccClamp=<6000 6000 6000>
    componentLinearAccScale=<0.5 0.5 0.5>
    componentLinearVelScale=<0.6 0.6 0.6>
    gravityOverride {x=-980}
    physicsBodyDefinitions [{
      boundBone {boneName=dyn_hair_5_r}
      boxExtents=<10 10 10>
      constraintSetup {angularLimitsMax=<0 2 2.5>, angularLimitsMin=<0 -10 -2.5>, angularTarget=<1 0 0>}
      localJointOffset=<2 0 0>
     }]
    relativeSpaceBone {boneName=neck_01}
    simSpaceSettings {maxAngularAcceleration=15, maxAngularVelocity=7, simSpaceAngularAlpha=0.75}
   }
   n30 = Anim Dynamics - Bone: dyn_hair_6_r!(ComponentPose=@n29, Alpha=0.85, GravityOverride:Vector="-980,0,0") -> (Pose) {
    alpha=0.85
    angularDampingOverride=1
    angularSpringConstant=100
    bAngularSpring=true
    bOverrideAngularDamping=true
    bUseGravityOverride=true
    gravityScale=0.35
    simulationSpace=RootRelative
    boundBone {boneName=dyn_hair_6_r}
    componentAppliedLinearAccClamp=<6000 6000 6000>
    componentLinearAccScale=<0.5 0.5 0.5>
    componentLinearVelScale=<0.6 0.6 0.6>
    gravityOverride {x=-980}
    physicsBodyDefinitions [{
      boundBone {boneName=dyn_hair_6_r}
      boxExtents=<10 10 10>
      constraintSetup {angularLimitsMax=<0 0 5>, angularLimitsMin=<0 -10 -5>, angularTarget=<1 0 0>}
      localJointOffset=<2 0 0>
     }]
    relativeSpaceBone {boneName=neck_01}
    simSpaceSettings {maxAngularAcceleration=15, maxAngularVelocity=7, simSpaceAngularAlpha=0.75}
   }
   n31 = Anim Dynamics - Bone: dyn_hair_7_r!(ComponentPose=@n30, Alpha=0.85, GravityOverride:Vector="-980,0,0") -> (Pose) {
    alpha=0.85
    angularDampingOverride=1
    angularSpringConstant=100
    bAngularSpring=true
    bOverrideAngularDamping=true
    bUseGravityOverride=true
    gravityScale=0.35
    simulationSpace=RootRelative
    boundBone {boneName=dyn_hair_7_r}
    componentAppliedLinearAccClamp=<6000 6000 6000>
    componentLinearAccScale=<0.5 0.5 0.5>
    componentLinearVelScale=<0.6 0.6 0.6>
    gravityOverride {x=-980}
    physicsBodyDefinitions [{
      boundBone {boneName=dyn_hair_7_r}
      boxExtents=<10 10 10>
      constraintSetup {angularLimitsMax=<0 0 5>, angularLimitsMin=<0 -10 -5>, angularTarget=<1 0 0>}
      localJointOffset=<2 0 0>
     }]
    relativeSpaceBone {boneName=neck_01}
    simSpaceSettings {maxAngularAcceleration=15, maxAngularVelocity=7, simSpaceAngularAlpha=0.75}
   }
   n32 = Anim Dynamics - Bone: dyn_hair_8_r!(ComponentPose=@n31, Alpha=0.85, GravityOverride:Vector="-980,0,0") -> (Pose) {
    alpha=0.85
    angularDampingOverride=1
    angularSpringConstant=100
    bAngularSpring=true
    bOverrideAngularDamping=true
    bUseGravityOverride=true
    gravityScale=0.35
    simulationSpace=RootRelative
    boundBone {boneName=dyn_hair_8_r}
    componentAppliedLinearAccClamp=<6000 6000 6000>
    componentLinearAccScale=<0.5 0.5 0.5>
    componentLinearVelScale=<0.6 0.6 0.6>
    gravityOverride {x=-980}
    physicsBodyDefinitions [{
      boundBone {boneName=dyn_hair_8_r}
      boxExtents=<10 10 10>
      constraintSetup {angularLimitsMax=<0 0 5>, angularLimitsMin=<0 -10 -5>, angularTarget=<1 0 0>}
      localJointOffset=<2 0 0>
     }]
    relativeSpaceBone {boneName=neck_01}
    simSpaceSettings {maxAngularAcceleration=15, maxAngularVelocity=7, simSpaceAngularAlpha=0.75}
   }
   n33 = Component To Local(ComponentPose=@n32)
   Save cached pose 'BackHair_dynamics'(Pose=Blend Poses by bool(BlendPose_0=Use cached pose 'CRCache_Input'(), BlendPose_1=Blend Poses by bool(BlendPose_0=@n33, BlendPose_1=@n14, bActiveValue=true))) {cacheName=BackHair_dynamics}
  ]}
 EventGraph {nodes [
   n1 = Event BlueprintUpdateAnimation()
   > K2Node_Composite_0 = StateBased_Graph!(execute, DeltaTime:double=@n1.DeltaTimeX) -> (then:Exec) {is=title K2Node_Composite}
   > n2 = "Gravity Override State Function (Parameter Asset) (Nativized)"(
    CharacterPartAnimInstance=Self-Reference()
    GameTimeInSeconds=GetGameTimeInSeconds()
    Parameters=&/F_MED_EagerUnify_FaceAcc_Parameters
   )
   > n3 = Set Gravity Override(Gravity Override=@n2)
   > n4 = Sequence()
   > n5 = Branch(Condition=Get bIsSliding())
   > Set BackRBAN_GravityOverride(BackRBAN_GravityOverride:Vector="0,0,-800") -> (Output_Get:Vector)
   Set BackRBAN_GravityOverride(execute=@n5.else, BackRBAN_GravityOverride=Get Gravity Override())
   Branch(
    execute=@n4.then_1
    Condition="Equal (Vector)"!(A:Vector=@n3, B:Vector, ErrorTolerance:float) -> (ReturnValue:bool=~) {is=title K2Node_PromotableOperator}
   )
   > Set Gravity Override(Gravity Override="0,0,-400")
   Event OnCharacterCustomizationLoaded()
   n6 = CheckForBackbling()
   > Set isWearingBlackbling(isWearingBlackbling:bool=OR Boolean!(A=@n6, B=@n6.bIsBackblingACape, C:bool=@n6.bIsBackblingLong) -> (ReturnValue)) -> (Output_Get:bool)
  ]}
 StateBased_Graph {
  path=EventGraph.K2Node_Composite_0.StateBased_Graph
  nodes [
   n1 = Inputs!() -> (execute, DeltaTime:double=~)
   n2 = Get bIsPlayingMeleeAnim()
   > Set MeleeSwing(MeleeSwing:double=FInterpTo(Current=Get MeleeSwing() -> (MeleeSwing:double), Target="To Float (Boolean)"(InBool=@n2), DeltaTime=@n1, InterpSpeed=25)) -> (Output_Get:double)
   > Set CrouchingRBAN(CrouchingRBAN:bool=AND Boolean(A=Get bIsCrouching(), B=Get bRigidBodyEnabled())) -> (Output_Get:bool)
   n3 = Get bRigidBodyEnabled()
   > Set BackBlingOrRB_Off(BackBlingOrRB_Off:bool=OR Boolean!(A=Get isWearingBlackbling() -> (isWearingBlackbling:bool), B=NOT Boolean(A=@n3)) -> (ReturnValue)) -> (Output_Get:bool)
   n4 = Get bIsPlayingEmote()
   > Set isIdleFrontEnd(isIdleFrontEnd:bool=AND Boolean(A=Get bIsInFrontEnd(), B=InvertBool!(bValue:bool=@n4) -> (ReturnValue:bool))) -> (Output_Get:bool)
   > Outputs!(then:Exec)
  ]
 }
]
variables [
 BackBlingOrRB_Off {type=boolean}
 BackRBAN_GravityOverride {type=Vector}
 CrouchingRBAN {type=boolean}
 Gravity Override {type=Vector}
 isIdleFrontEnd {type=boolean}
 isWearingBlackbling {type=boolean}
 MeleeSwing {type=number}
]
